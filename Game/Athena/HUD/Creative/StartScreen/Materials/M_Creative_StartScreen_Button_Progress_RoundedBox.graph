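# Translucent UI material for a Creative start screen button: a rounded box with fill, outline and a horizontal progress cutoff.
# A shared keep-aspect-ratio rounded box function produces SDF and fill from size, corner radius, outline and glow parameters.
# The SDF is turned into an inner outline, subtracted from the fill so translucent colors do not overlap.
# Both layers are tinted, summed and masked by a hard step of UV against the Progress parameter.
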Material /Game/Athena/HUD/Creative/StartScreen/Materials/M_Creative_StartScreen_Button_Progress_RoundedBox
graph {
 properties {blendMode=BLEND_Translucent, materialDomain=MD_UI, pixelDepthOffsetMode=PDOM_Legacy, usedShadingModels=Default Lit}
 nodes [
  n1 = VectorParameter() {group=Colors, parameterName=Box Color, sortPriority=0, defaultValue=<rgba 1 1 1 1>}
  n2 = MaterialFunctionCall(X=@n1.R, Y=@n1.R, Z=@n1.R, A=@n1.R) {
   desc="because float4 \r\nbecome float3\r\nwhen it's transformed \r\nin a parameter"
   materialFunction=&MakeFloat4
  }
  n3 = Multiply(A=@n2.Result, B=NamedRerouteUsage())
  n4 = VectorParameter() {group=Colors, parameterName=Outline Color, sortPriority=1, defaultValue=<rgba 1 0 0.4404 0.5>}
  n5 = Multiply(
   A=NamedRerouteUsage()
   B=MaterialFunctionCall(X=@n4.R, Y=@n4.R, Z=@n4.R, A=@n4.R) {materialFunction=&MakeFloat4}.Result
  )
  ComponentMask(None=Multiply(A=Add(A=@n3, B=@n5), B=NamedRerouteUsage())) {a=true}
  n6 = Subtract(A=NamedRerouteUsage(), B=NamedRerouteUsage()) {desc="By default the outline is centered. In our case, the outline inner the shape. \r\nSo if you use it with the fill and change the color opacity, \r\nyou'll see the outline overlapping the fill color. \r\nWe avoiding that by reducing the fill by the outline."}
  NamedRerouteDeclaration(None=Saturate(None=@n6)) {name=Shape, nodeColor=<rgba 0 0 0 1>}
  n7 = Abs(None=NamedRerouteUsage())
  n8 = Multiply(A=NamedRerouteUsage()) {constB=2, desc="outline was compute from the center\r\nso half of the size needed here "}
  n9 = OneMinus(None=SmoothStep(Min=NamedRerouteUsage(), Max=NamedRerouteUsage(), Value=Subtract(A=@n7, B=@n8) {constB=0.1}) {constMax=0})
  n10 = OneMinus(None=NamedRerouteUsage())
  NamedRerouteDeclaration(None=Saturate(None=Subtract(A=@n9, B=@n10))) {name=Inner Outline, nodeColor=<rgba 0 0 0 1>}
  n11 = MaterialFunctionCall() {materialFunction=&GetUserInterfaceUV}
  NamedRerouteDeclaration(None=@n11.Pixel Size) {name=Pixel Size, nodeColor=<rgba 0 0 0 1>}
  n12 = NamedRerouteUsage()
  n13 = Ceil(None=Subtract(
    A=ComponentMask(None=@n12) {r=true}
    B=ScalarParameter() {defaultValue=1, group=Progress, parameterName=Progress, sliderMax=1, sortPriority=0}
   ))
  NamedRerouteDeclaration(None=Subtract(A=Constant() {desc="Maximum \r\nProgress Value", r=1}, B=Max(A=@n13) {constB=0})) {name=Progress, nodeColor=<rgba 0 0 0 1>}
  NamedRerouteDeclaration(None=@n11.Normalized UV) {name=UVs Normalized, nodeColor=<rgba 0 0 0 1>}
  NamedRerouteDeclaration(None=MaterialFunctionCall(
    X=ScalarParameter() {defaultValue=1, group=Box Settings, parameterName=Box Width, sliderMax=1, sortPriority=0}
    Y=ScalarParameter() {defaultValue=1, group=Box Settings, parameterName=Box Height, sliderMax=1, sortPriority=1}
   ) {materialFunction=&MakeFloat2}.Result) {name=Box Size, nodeColor=<rgba 0 0 0 1>}
  n14 = MaterialFunctionCall!(
   UVs
   Pixel Size=NamedRerouteUsage()
   Box Size=NamedRerouteUsage()
   Corner Radiuses=NamedRerouteUsage()
   Glow Min=NamedRerouteUsage()
   Glow Max=NamedRerouteUsage()
   Outline Thickness=NamedRerouteUsage()
  ) {materialFunction=/Game/Athena/HUD/Creative/MaterialFuntions/MF_UI_Creative_RoundedBox_KeepAspectRatio}
  NamedRerouteDeclaration(None=@n14.SDF) {name=SDF, nodeColor=<rgba 0 0 0 1>}
  NamedRerouteDeclaration(None=@n14.Fill) {name=Fill, nodeColor=<rgba 0 0 0 1>}
  NamedRerouteDeclaration(None=ScalarParameter() {group=Box Settings, parameterName=Outline Thickness, sliderMax=0.5, sortPriority=3}) {name=Outline Thickness, nodeColor=<rgba 0 0 0 1>}
  NamedRerouteDeclaration(None=ScalarParameter() {group=Box Settings, parameterName=Glow Max, sortPriority=5}) {name=Glow Max, nodeColor=<rgba 0 0 0 1>}
  NamedRerouteDeclaration(None=ScalarParameter() {group=Box Settings, parameterName=Glow Min, sortPriority=4}) {name=Glow Min, nodeColor=<rgba 0 0 0 1>}
  n15 = VectorParameter() {group=Box Settings, parameterName=Box Corners Radius, sortPriority=2}
  NamedRerouteDeclaration(None=MaterialFunctionCall(X=@n15.R, Y=@n15.R, Z=@n15.R, A=@n15.R) {materialFunction=&MakeFloat4}.Result) {name=Box Corners Radius, nodeColor=<rgba 0 0 0 1>}
  Reroute(None=MaterialFunctionCall(UV=NamedRerouteUsage()) {materialFunction=&MF_UI_UVScaling}.Scaled to Fit) {desc="to patterns displayed inside \r\nthe shape eventually"}
 ]
}
