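# Team title card button for the Creative start-screen credits, built on FortCreativeCreditsTeamTitleCard.
# It lays out an avatar, creator title, name and description over a material-backed border, with a key-bind action widget at the left edge.
# Hover and UnHover animations fade text between white and dark navy while driving the border material's Select parameter.
# The event graph plays these on hover changes, and plays Pressed, a slight scale-down, forward on press and reversed on release.
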WidgetBlueprint /Game/Athena/HUD/Creative/StartScreen/WBP_Creative_Credits_TeamTitleCard_Button
animations [
 UnHover {
  binding_count=4
  length_seconds=0.2
  object=&self:UnHover
  sequence {bindings [
    Text_CreatorDescription {
     id=#1
     possessed_class=CommonTextBlock
     tracks [{
       type=MovieSceneColorTrack
       display_name=Color and Opacity
       sections [{
         type=MovieSceneColorSection
         channels [
          Color.R {type=MovieSceneScriptingFloatChannel, keys [[0,0.0012],[4,1]]}
          Color.G {type=MovieSceneScriptingFloatChannel, keys [[0,0.004],[4,1]]}
          Color.B {type=MovieSceneScriptingFloatChannel, keys [[0,0.0467],[4,1]]}
          Color.A {type=MovieSceneScriptingFloatChannel, const [1,2]}
         ]
        }]
      }]
    }
    Text_CreatorName {
     id=#2
     possessed_class=CommonTextBlock
     tracks [{
       type=MovieSceneColorTrack
       display_name=Color and Opacity
       sections [{
         type=MovieSceneColorSection
         channels [
          Color.R {type=MovieSceneScriptingFloatChannel, keys [[0,0.0012],[4,1]]}
          Color.G {type=MovieSceneScriptingFloatChannel, keys [[0,0.004],[4,1]]}
          Color.B {type=MovieSceneScriptingFloatChannel, keys [[0,0.0467],[4,1]]}
          Color.A {type=MovieSceneScriptingFloatChannel, const [1,2]}
         ]
        }]
      }]
    }
    Border_Background {
     id=#3
     possessed_class=Border
     tracks [{
       type=MovieSceneWidgetMaterialTrack
       sections [{type=MovieSceneParameterSection, channels [Select {type=MovieSceneScriptingFloatChannel, keys [[0,1],[4,0]]}]}]
      }]
    }
    Text_CreatorTitle {
     id=#4
     possessed_class=CommonTextBlock
     tracks [{
       type=MovieSceneColorTrack
       display_name=Color and Opacity
       sections [{
         type=MovieSceneColorSection
         channels [
          Color.R {type=MovieSceneScriptingFloatChannel, keys [[0,0.0012],[4,1]]}
          Color.G {type=MovieSceneScriptingFloatChannel, keys [[0,0.004],[4,1]]}
          Color.B {type=MovieSceneScriptingFloatChannel, keys [[0,0.0467],[4,1]]}
          Color.A {type=MovieSceneScriptingFloatChannel, const [1,2]}
         ]
        }]
      }]
    }
   ]}
 }
 Hover {
  binding_count=4
  length_seconds=0.2003
  object=&self:Hover
  sequence {bindings [
    Text_CreatorTitle {
     id=#5
     possessed_class=CommonTextBlock
     tracks [{
       type=MovieSceneColorTrack
       display_name=Color and Opacity
       sections [{
         type=MovieSceneColorSection
         channels [
          Color.R {type=MovieSceneScriptingFloatChannel, keys [[0,1],[4.0057,0.0012]]}
          Color.G {type=MovieSceneScriptingFloatChannel, keys [[0,1],[4.0057,0.004]]}
          Color.B {type=MovieSceneScriptingFloatChannel, keys [[0,1],[4.0057,0.0467]]}
          Color.A {type=MovieSceneScriptingFloatChannel, const [1,2]}
         ]
        }]
      }]
    }
    Text_CreatorName {
     id=#2
     possessed_class=CommonTextBlock
     tracks [{
       type=MovieSceneColorTrack
       display_name=Color and Opacity
       sections [{
         type=MovieSceneColorSection
         channels [
          Color.R {type=MovieSceneScriptingFloatChannel, keys [[0,1],[4.0057,0.0012]]}
          Color.G {type=MovieSceneScriptingFloatChannel, keys [[0,1],[4.0057,0.004]]}
          Color.B {type=MovieSceneScriptingFloatChannel, keys [[0,1],[4.0057,0.0467]]}
          Color.A {type=MovieSceneScriptingFloatChannel, const [1,2]}
         ]
        }]
      }]
    }
    Border_Background {
     id=#3
     possessed_class=Border
     tracks [{
       type=MovieSceneWidgetMaterialTrack
       sections [{type=MovieSceneParameterSection, channels [Select {type=MovieSceneScriptingFloatChannel, keys [[0,0],[4.0057,1]]}]}]
      }]
    }
    Text_CreatorDescription {
     id=#6
     possessed_class=CommonTextBlock
     tracks [{
       type=MovieSceneColorTrack
       display_name=Color and Opacity
       sections [{
         type=MovieSceneColorSection
         channels [
          Color.R {type=MovieSceneScriptingFloatChannel, keys [[0,1],[4.0057,0.0012]]}
          Color.G {type=MovieSceneScriptingFloatChannel, keys [[0,1],[4.0057,0.004]]}
          Color.B {type=MovieSceneScriptingFloatChannel, keys [[0,1],[4.0057,0.0467]]}
          Color.A {type=MovieSceneScriptingFloatChannel, const [1,2]}
         ]
        }]
      }]
    }
   ]}
 }
 Pressed {
  binding_count=5
  length_seconds=0.1
  object=&self:Pressed
  sequence {bindings [
    Overlay_1 {
     id=#7
     possessed_class=Overlay
     tracks [{
       type=MovieScene2DTransformTrack
       sections [{
         type=MovieScene2DTransformSection
         channels [
          Scale.X {type=MovieSceneScriptingFloatChannel, keys [[0,1],[2,1],[2,0.975]]}
          Scale.Y {type=MovieSceneScriptingFloatChannel, keys [[0,1],[2,0.975]]}
         ]
        }]
      }]
    }
    Image_TeamAvatar {
     id=#8
     possessed_class=Image
     tracks [{
       type=MovieScene2DTransformTrack
       sections [{
         type=MovieScene2DTransformSection
         channels [
          Scale.X {type=MovieSceneScriptingFloatChannel, keys [[0,1],[2,0.95]]}
          Scale.Y {type=MovieSceneScriptingFloatChannel, keys [[0,1],[2,0.975]]}
         ]
        }]
      }]
    }
    Text_CreatorTitle {
     id=#9
     possessed_class=CommonTextBlock
     tracks [{
       type=MovieScene2DTransformTrack
       sections [{
         type=MovieScene2DTransformSection
         channels [
          Scale.X {type=MovieSceneScriptingFloatChannel, keys [[0,1],[2,0.95]]}
          Scale.Y {type=MovieSceneScriptingFloatChannel, keys [[0,1],[2,0.975]]}
         ]
        }]
      }]
    }
    Text_CreatorName {
     id=#10
     possessed_class=CommonTextBlock
     tracks [{
       type=MovieScene2DTransformTrack
       sections [{
         type=MovieScene2DTransformSection
         channels [
          Scale.X {type=MovieSceneScriptingFloatChannel, keys [[0,1],[2,0.95]]}
          Scale.Y {type=MovieSceneScriptingFloatChannel, keys [[0,1],[2,0.975]]}
         ]
        }]
      }]
    }
    Text_CreatorDescription {
     id=#11
     possessed_class=CommonTextBlock
     tracks [{
       type=MovieScene2DTransformTrack
       sections [{
         type=MovieScene2DTransformSection
         channels [
          Scale.X {type=MovieSceneScriptingFloatChannel, keys [[0,1],[2,0.95]]}
          Scale.Y {type=MovieSceneScriptingFloatChannel, keys [[0,1],[2,0.975]]}
         ]
        }]
      }]
    }
   ]}
 }
]
graphs [EventGraph {nodes [
   Event On Pressed() {is=title K2Node_Event}
   n1 = Get Pressed() -> (Pressed:Widget Animation Object Reference)
   > PlayAnimation(InAnimation=@n1)
   Event On Released() {is=title K2Node_Event}
   > PlayAnimation(InAnimation=@n1, PlayMode=Reverse)
   Event On Unhovered()
   > StopAnimation(InAnimation=Get Hover() -> (Hover:Widget Animation Object Reference))
   > PlayAnimation(InAnimation=Get UnHover() -> (UnHover:Widget Animation Object Reference))
   Event On Hovered()
   > PlayAnimation(InAnimation=Get Hover())
  ]}]
variables [
 ButtonText {type=string, default="{Text}"}
 HoveredButtonText {type=string, default="{Text}"}
]
widgets {
 info {
  inheritedWidgetCount=4
  parentClass=/Script/FortniteUI.FortCreativeCreditsTeamTitleCard
  rootWidgetClass=&VerticalBox
  widgetCount=16
 }
 widgets [
  VerticalBox_201 {widgetClassPath=&VerticalBox, widget=VerticalBox_201}
  Overlay_1 {widgetClassPath=&Overlay, parent=VerticalBox_201, widget=Overlay_1, slot {class=VerticalBoxSlot}}
  VerticalBox_92 {
   widgetClassPath=&VerticalBox
   parent=Overlay_1
   widget=VerticalBox_92
   slot {class=OverlaySlot, properties {horizontalAlignment=HAlign_Fill, verticalAlignment=VAlign_Fill}}
  }
  Overlay_112 {widgetClassPath=&Overlay, parent=VerticalBox_92, widget=Overlay_112, slot {class=VerticalBoxSlot}}
  Border_Background {
   widgetClassPath=&Border
   bIsVariable=true
   parent=Overlay_112
   widget=Border_Background
   properties {
    horizontalAlignment=HAlign_Center
    verticalAlignment=VAlign_Top
    visibility=SelfHitTestInvisible
    background {
     bIsDynamicallyLoaded=false
     drawAs=Box
     imageType=NoImage
     mirroring=NoMirror
     resourceName=None
     resourceObject=/Game/Athena/HUD/MiniMap/Art/M_UI_Overview_Container
     tiling=NoTile
     imageSize=<64 64>
     margin=<ltrb 0.5 0.5 0.5 0.5>
     outlineSettings {
      bUseBrushTransparency=false
      roundingType=HalfHeightRadius
      width=0
      color {colorUseRule=UseColor_Specified, specifiedColor=<rgba 0 0 0 0>}
      cornerRadii=<0 0 0 0>
     }
     tintColor {colorUseRule=UseColor_Specified, specifiedColor=<rgba 1 1 1 1>}
     uVRegion {bIsValid=false, max=<0 0>, min=<0 0>}
    }
    padding=<ltrb 0 0 0 0>
   }
   slot {class=OverlaySlot, properties {horizontalAlignment=HAlign_Fill, verticalAlignment=VAlign_Fill}}
  }
  VerticalBox_199 {
   widgetClassPath=&VerticalBox
   parent=Overlay_112
   widget=VerticalBox_199
   slot {class=OverlaySlot, properties {horizontalAlignment=HAlign_Fill, verticalAlignment=VAlign_Fill, padding=<ltrb 30 25 20 25>}}
  }
  HorizontalBox_150 {widgetClassPath=&HorizontalBox, parent=VerticalBox_199, widget=HorizontalBox_150, slot {class=VerticalBoxSlot}}
  Image_TeamAvatar {
   widgetClassPath=&Image
   bIsVariable=true
   parent=HorizontalBox_150
   widget=Image_TeamAvatar
   properties {brush {
     bIsDynamicallyLoaded=false
     drawAs=Image
     imageType=NoImage
     mirroring=NoMirror
     resourceName=None
     resourceObject=/Game/UI/Foundation/Textures/Icons/Emoji/Season_3/T-Item-Emoji-Thief
     tiling=NoTile
     imageSize=<128 128>
     margin=<ltrb 0 0 0 0>
     outlineSettings {
      bUseBrushTransparency=false
      roundingType=HalfHeightRadius
      width=0
      color {colorUseRule=UseColor_Specified, specifiedColor=<rgba 0 0 0 0>}
      cornerRadii=<0 0 0 0>
     }
     tintColor {colorUseRule=UseColor_Specified, specifiedColor=<rgba 1 1 1 1>}
     uVRegion {bIsValid=false, max=<0 0>, min=<0 0>}
    }}
   slot {class=HorizontalBoxSlot}
  }
  VerticalBox_0 {
   widgetClassPath=&VerticalBox
   parent=HorizontalBox_150
   widget=VerticalBox_0
   slot {class=HorizontalBoxSlot, properties {padding=<ltrb 48 32 0 32>, size {sizeRule=Fill, value=1}}}
  }
  Text_CreatorTitle {
   widgetClassPath=&CommonTextBlock
   bInherited=true
   bIsVariable=true
   parent=VerticalBox_0
   widget=Text_CreatorTitle
   properties {
    autoWrapText=true
    bAutoCollapseWithEmptyText=true
    mobileFontSizeMultiplier=1.3
    text=CREATOR/CREATOR TEAM
    font {
     bForceMonospaced=false
     bMaterialIsStencil=false
     fontMaterial=null
     fontObject=&BurbankSmall
     letterSpacing=0
     monospacedWidth=1
     size=18
     skewAmount=0
     typefaceFontName=Bold
     outlineSettings {
      bApplyOutlineToDropShadows=false
      bMiteredCorners=false
      bSeparateFillAlpha=false
      outlineMaterial=null
      outlineSize=0
      outlineColor=<rgba 0 0 0 1>
     }
    }
    shadowOffset=<0 0>
   }
   slot {class=VerticalBoxSlot, properties {verticalAlignment=VAlign_Center}}
  }
  Text_CreatorName {
   widgetClassPath=&CommonTextBlock
   bInherited=true
   bIsVariable=true
   parent=VerticalBox_0
   widget=Text_CreatorName
   properties {
    autoWrapText=true
    mobileFontSizeMultiplier=1.3
    text=CREATOR/CREATOR TEAM NAME
    font {
     bForceMonospaced=false
     bMaterialIsStencil=false
     fontMaterial=null
     fontObject=&BurbankSmall
     letterSpacing=0
     monospacedWidth=1
     size=24
     skewAmount=0
     typefaceFontName=Black
     outlineSettings {
      bApplyOutlineToDropShadows=false
      bMiteredCorners=false
      bSeparateFillAlpha=false
      outlineMaterial=null
      outlineSize=0
      outlineColor=<rgba 0 0 0 1>
     }
    }
    shadowOffset=<0 0>
   }
   slot {class=VerticalBoxSlot, properties {verticalAlignment=VAlign_Center, padding=<ltrb 0 8 0 0>}}
  }
  Text_CreatorDescription {
   widgetClassPath=&CommonTextBlock
   bInherited=true
   parent=VerticalBox_199
   widget=Text_CreatorDescription
   properties {
    autoWrapText=true
    bAutoCollapseWithEmptyText=true
    text="Lorem ipsum dolor sit amet, consectetur adipiscing elit. Nullam velit neque, mattis a diam eu, faucibus sagittis quam. Nunc eget gravida diam. Etiam faucibus, ex nec finibus mattis, velit mi fermentum ipsum, eu sagittis dui arcu in ex."
    font {
     bForceMonospaced=false
     bMaterialIsStencil=false
     fontMaterial=null
     fontObject=&BurbankSmall
     letterSpacing=0
     monospacedWidth=1
     size=24
     skewAmount=0
     typefaceFontName=Black
     outlineSettings {
      bApplyOutlineToDropShadows=false
      bMiteredCorners=false
      bSeparateFillAlpha=false
      outlineMaterial=null
      outlineSize=0
      outlineColor=<rgba 0 0 0 1>
     }
    }
    shadowOffset=<0 0>
   }
   slot {class=VerticalBoxSlot, properties {padding=<ltrb 0 5 0 5>}}
  }
  KeyBind {
   widgetClassPath=&Overlay
   bIsVariable=true
   parent=Overlay_1
   widget=KeyBind
   properties {visibility=HitTestInvisible}
   slot {class=OverlaySlot, properties {verticalAlignment=VAlign_Center, padding=<ltrb -64 0 0 0>}}
  }
  Spacer_70 {widgetClassPath=&Spacer, parent=KeyBind, widget=Spacer_70, properties {size=<128 64>}, slot {class=OverlaySlot}}
  Shadow {
   widgetClassPath=&Image
   bIsVariable=true
   parent=KeyBind
   widget=Shadow
   properties {
    visibility=Collapsed
    brush {
     bIsDynamicallyLoaded=false
     drawAs=Image
     imageType=NoImage
     mirroring=NoMirror
     resourceName=None
     resourceObject=/FortUILibrary/Foundation/Art_ToolKit/Tools/M_UI_XP_Everywhere_Backing_Darken
     tiling=NoTile
     imageSize=<32 32>
     margin=<ltrb 0 0 0 0>
     outlineSettings {
      bUseBrushTransparency=false
      roundingType=HalfHeightRadius
      width=0
      color {colorUseRule=UseColor_Specified, specifiedColor=<rgba 0 0 0 0>}
      cornerRadii=<0 0 0 0>
     }
     tintColor {colorUseRule=UseColor_Specified, specifiedColor=<rgba 0 0 0 1>}
     uVRegion {bIsValid=false, max=<0 0>, min=<0 0>}
    }
    colorAndOpacity=<rgba 1 1 1 0.5>
    renderTransform {angle=0, scale=<1.5 1.5>, shear=<0 0>, translation=<0 0>}
   }
   slot {class=OverlaySlot, properties {horizontalAlignment=HAlign_Center, verticalAlignment=VAlign_Center}}
  }
  InputActionWidget {
   widgetClassPath=&CommonActionWidget
   bInherited=true
   bIsVariable=true
   parent=KeyBind
   widget=InputActionWidget
   properties {
    progressMaterialParam=Percentage
    progressMaterialBrush {
     bIsDynamicallyLoaded=false
     drawAs=Image
     imageType=NoImage
     mirroring=NoMirror
     resourceName=None
     resourceObject=/Game/UI/Materials/HollowCircularProgressBar
     tiling=NoTile
     imageSize=<48 48>
     margin=<ltrb 0 0 0 0>
     outlineSettings {
      bUseBrushTransparency=false
      roundingType=HalfHeightRadius
      width=0
      color {colorUseRule=UseColor_Specified, specifiedColor=<rgba 0 0 0 0>}
      cornerRadii=<0 0 0 0>
     }
     tintColor {colorUseRule=UseColor_Specified, specifiedColor=<rgba 0.7913 0.0086 0.1845 1>}
     uVRegion {bIsValid=false, max=<0 0>, min=<0 0>}
    }
   }
   slot {class=OverlaySlot, properties {horizontalAlignment=HAlign_Center, verticalAlignment=VAlign_Center}}
  }
 ]
}
