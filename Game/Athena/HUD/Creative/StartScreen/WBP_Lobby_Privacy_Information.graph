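# Creative start screen lobby widget that labels the current game as public or private, with a matching unlocked or locked icon.
# On Construct it casts the owning player to Athena_PlayerController and collapses itself if the cast fails.
# IsInPublicMatchmadeCreativeMinigame drives Select nodes that set the text and icon visibilities once, with no later refresh in this graph.
# SetColor tints the text and both icons with a supplied color.
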WidgetBlueprint /Game/Athena/HUD/Creative/StartScreen/WBP_Lobby_Privacy_Information
graphs [
 EventGraph {nodes [
   Event Construct()
   > n1 = Cast To Athena_PlayerController!(execute, Object=GetOwningPlayer()) -> (then, CastFailed, AsAthena Player Controller:Athena Player Controller Object Reference) {is=title K2Node_DynamicCast}
   > SetVisibility(InVisibility=SelfHitTestInvisible)
   n2 = IsInPublicMatchmadeCreativeMinigame!(self:Fort Player Controller Athena Object Reference=@n1) -> (ReturnValue:bool)
   > Set Text(
    Text=Select(Option 0:Text="LOCTEXT(\"PRIVATE GAME\")", Option 1:Text="LOCTEXT(\"PUBLIC GAME\")", Index=@n2) -> (ReturnValue:Text)
    self=Get Text_PrivatePublic() -> (Text_PrivatePublic:Common Text Object Reference)
   )
   > SetVisibility(
    self=Get Image_PublicIcon() -> (Image_PublicIcon:Image Object Reference)
    InVisibility=Select(Option 0:ESlateVisibility Enum=Collapsed, Option 1:ESlateVisibility Enum=SelfHitTestInvisible, Index=@n2) -> (ReturnValue:ESlateVisibility Enum)
   )
   > SetVisibility(
    self=Get Image_PrivateIcon() -> (Image_PrivateIcon:Image Object Reference)
    InVisibility=Select(Option 0=SelfHitTestInvisible, Option 1=Collapsed, Index=@n2)
   )
   SetVisibility(execute=@n1.CastFailed, InVisibility=Collapsed)
  ]}
 SetColor {nodes [
   n1 = SetColor!() -> (then, Color:Slate Color Structure) {is=title K2Node_FunctionEntry}
   > Set ColorAndOpacity!(execute, ColorAndOpacity:Slate Color Structure=@n1, self:Text Object Reference=Get Text_PrivatePublic()) -> (then, Output_Get:Slate Color Structure)
   > SetBrushTintColor!(execute, self:Image Object Reference=[Get Image_PublicIcon(), Get Image_PrivateIcon()], TintColor:Slate Color Structure=@n1) -> (then)
  ]}
]
variables []
widgets {
 info {inheritedWidgetCount=0, parentClass=&CommonUserWidget, rootWidgetClass=&HorizontalBox, widgetCount=5}
 widgets [
  PrivatePublicGameBox {widgetClassPath=&HorizontalBox, bIsVariable=true, widget=PrivatePublicGameBox}
  Image_PrivateIcon {
   widgetClassPath=&Image
   bIsVariable=true
   parent=PrivatePublicGameBox
   widget=Image_PrivateIcon
   properties {
    visibility=Collapsed
    brush {
     bIsDynamicallyLoaded=false
     drawAs=Image
     imageType=NoImage
     mirroring=NoMirror
     resourceName=None
     resourceObject=/Game/UI/Foundation/Textures/Icons/Locks/T-Icon-Lock-64
     tiling=NoTile
     imageSize=<22 22>
     margin=<ltrb 0 0 0 0>
     outlineSettings {
      bUseBrushTransparency=false
      roundingType=HalfHeightRadius
      width=0
      color {colorUseRule=UseColor_Specified, specifiedColor=<rgba 0 0 0 0>}
      cornerRadii=<0 0 0 0>
     }
     tintColor {colorUseRule=UseColor_Specified, specifiedColor=<rgba 1 1 1 1>}
     uVRegion {bIsValid=false, max=<0 0>, min=<0 0>}
    }
   }
   slot {class=HorizontalBoxSlot, properties {horizontalAlignment=HAlign_Center, verticalAlignment=VAlign_Center, padding=<ltrb 0 0 0 3>}}
  }
  Image_PublicIcon {
   widgetClassPath=&Image
   bIsVariable=true
   parent=PrivatePublicGameBox
   widget=Image_PublicIcon
   properties {
    visibility=Collapsed
    brush {
     bIsDynamicallyLoaded=false
     drawAs=Image
     imageType=NoImage
     mirroring=NoMirror
     resourceName=None
     resourceObject=/Game/UI/Foundation/Textures/Icons/Locks/T-Icon-Unlocked-64
     tiling=NoTile
     imageSize=<24 24>
     margin=<ltrb 0 0 0 0>
     outlineSettings {
      bUseBrushTransparency=false
      roundingType=HalfHeightRadius
      width=0
      color {colorUseRule=UseColor_Specified, specifiedColor=<rgba 0 0 0 0>}
      cornerRadii=<0 0 0 0>
     }
     tintColor {colorUseRule=UseColor_Specified, specifiedColor=<rgba 1 1 1 1>}
     uVRegion {bIsValid=false, max=<0 0>, min=<0 0>}
    }
   }
   slot {class=HorizontalBoxSlot, properties {horizontalAlignment=HAlign_Center, verticalAlignment=VAlign_Center, padding=<ltrb 0 0 0 3>}}
  }
  Spacer_1 {widgetClassPath=&Spacer, parent=PrivatePublicGameBox, widget=Spacer_1, properties {size=<8 1>}, slot {class=HorizontalBoxSlot}}
  Text_PrivatePublic {
   widgetClassPath=&CommonTextBlock
   bIsVariable=true
   parent=PrivatePublicGameBox
   widget=Text_PrivatePublic
   properties {
    text=PRIVATE GAME
    wrapTextAt=500
    font {
     bForceMonospaced=false
     bMaterialIsStencil=false
     fontMaterial=null
     fontObject=&BurbankSmall
     letterSpacing=0
     monospacedWidth=1
     size=14
     skewAmount=0
     typefaceFontName=Black
     outlineSettings {
      bApplyOutlineToDropShadows=false
      bMiteredCorners=false
      bSeparateFillAlpha=false
      outlineMaterial=null
      outlineSize=0
      outlineColor=<rgba 0 0 0 1>
     }
    }
    shadowOffset=<0 0>
   }
   slot {class=HorizontalBoxSlot, properties {verticalAlignment=VAlign_Center}}
  }
 ]
}
