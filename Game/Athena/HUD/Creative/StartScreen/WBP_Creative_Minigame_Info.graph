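# Creative start-screen info widget, derived from FortCreativeMiniGameInfo, that shows island title, creator, island code and optional tags.
# Change events fill the texts and toggle island code visibility, hiding it for competitive playlists or when publishing display is Off.
# Helper functions collapse empty boxes and the creator row, apply configurable colors and play a slide/fade intro.
# Once playlist data is ready, non-competitive play-only experiences asynchronously add a tinted lobby privacy widget to that row.
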WidgetBlueprint /Game/Athena/HUD/Creative/StartScreen/WBP_Creative_Minigame_Info
animations [Intro {
  binding_count=1
  length_seconds=0.25
  object=&self:Intro
  sequence {bindings [CreatorBox {
     id=#1
     display_name=CreatorInfo
     possessed_class=HorizontalBox
     tracks [
      {
       type=MovieScene2DTransformTrack
       sections [{type=MovieScene2DTransformSection, channels [Translation.X {type=MovieSceneScriptingFloatChannel, keys [[2,-128],[7,0]]}]}]
      }
      {
       type=MovieSceneFloatTrack
       display_name=Render Opacity
       sections [{type=MovieSceneFloatSection, channels [None {type=MovieSceneScriptingFloatChannel, keys [[2,0],[7,1]]}]}]
      }
     ]
    }]}
 }]
graphs [
 DisplayTags {nodes [
   n1 = DisplayTags!() -> (then, InDisplayTags:bool=~) {is=title K2Node_FunctionEntry}
   > Set Visibility(
    Visibility=Select(Option 0=Collapsed, Option 1=HitTestInvisible, Index=@n1)
    self=Get EntryBox_MiniGameTags() -> (EntryBox_MiniGameTags:Dynamic Entry Box Object Reference)
   )
  ]}
 EventGraph {nodes [
   Event PreConstruct()
   > DisplayTags!(execute, self, InDisplayTags:bool=Get bDisplayTags() -> (bDisplayTags:bool)) -> (then) {is=title K2Node_CallFunction}
   > Set Text Color!(
    execute
    self
    InTextColor:Linear Color Structure=Get TextColor() -> (TextColor:Linear Color Structure)
    InSpecificIslandCodeColor:Linear Color Structure=Get SpecificIslandCodeColor() -> (SpecificIslandCodeColor:Linear Color Structure)
   ) -> (then) {is=title K2Node_CallFunction}
   Event OnTextSet() {is=title K2Node_Event}
   > HideIfTextEmpty!(
    execute
    self
    TextToCheck:Text Object Reference=Get Text_IslandCode() -> (Text_IslandCode:Common Text Object Reference)
    ElementToHide:Widget Object Reference=Get IslandCodeBox() -> (IslandCodeBox:Horizontal Box Object Reference)
   ) -> (then)
   > HideIfTextEmpty(
    TextToCheck=Get Text_CreatorGameTag() -> (Text_CreatorGameTag:Common Text Object Reference)
    ElementToHide=Get CreatorBox() -> (CreatorBox:Horizontal Box Object Reference)
   )
   > UpdateCreatorInfoVisibility()
   Event Construct()
   > n1 = PlaylistDataReadyAsync!(execute:Exec) -> (
    then:Exec
    OnReady:Exec
    GameState:Fort Game State Athena Object Reference
    Playlist:Fort Playlist Object Reference
    PlaylistContextTags:"Gameplay Tag Container Structure (by ref)"
   ) {is=title K2Node_AsyncAction}
   n2 = IsCompetitivePlaylist!() -> (ReturnValue:bool)
   Branch(
    execute=@n1.OnReady
    Condition=AND Boolean(A=NOT Boolean(A=@n2), B=IsExperiencePlayOnly!(self:Fort Game State Athena Object Reference=@n1) -> (ReturnValue:bool))
   )
   > Do Once()
   > n3 = CreateWidgetAsync!(
    execute:Exec
    UserWidgetSoftClass:User Widget Soft Class Reference=&/WBP_Lobby_Privacy_Information.WBP_Lobby_Privacy_Information_C
    OwningPlayer:Player Controller Object Reference
    bSuspendInputUntilComplete:bool=true
   ) -> (then:Exec, OnComplete:Exec, OnFailure:Exec, UserWidget:User Widget Object Reference) {is=title K2Node_AsyncAction}
   n4 = Cast To WBP_Lobby_Privacy_Information!(execute=@n3.OnComplete, Object=@n3) -> (then, CastFailed, AsWBP Lobby Privacy Information:WBP Lobby Privacy Information Object Reference) {is=title K2Node_DynamicCast}
   > SetColor!(
    execute
    self:WBP Lobby Privacy Information Object Reference=@n4
    Color:Slate Color Structure=Make Slate Color(SpecifiedColor=Get TextColor())
   ) -> (then)
   > AddChild(self=Get CreatorInfo() -> (CreatorInfo:Horizontal Box Object Reference), Content=@n3)
   > Set Lobby_Privacy_Information_Widget(Lobby_Privacy_Information_Widget:User Widget Object Reference=@n3) -> (Output_Get:User Widget Object Reference)
   n5 = Event OnIslandCodeChanged!() -> (OutputDelegate, then, SupportCode:"Text (by ref)") {is=title K2Node_Event}
   > n6 = Branch(Condition=IsCompetitivePlaylist())
   > SetVisibility(self=[Get Text_IslandCode(), Get Text_IslandCodeLabel() -> (Text_IslandCodeLabel:Common Text Object Reference)], InVisibility=Collapsed)
   SetVisibility(execute=@n6.else, self=[Get Text_IslandCodeLabel(), Get Text_IslandCode()], InVisibility=SelfHitTestInvisible)
   > "SetText (Text)"(self=Get Text_IslandCode(), InText=@n5.SupportCode)
   n7 = Event OnCreatorNameChanged!() -> (OutputDelegate, then, CreatorName:"Text (by ref)") {is=title K2Node_Event}
   > "SetText (Text)"(self=Get Text_CreatorGameTag(), InText=@n7.CreatorName)
   n8 = Event OnPublishedIslandCodeDisplayChanged!() -> (OutputDelegate, then, PublishedIslandCodeDisplay:ECreativePublishedIslandCodeDisplay Enum=Off) {is=title K2Node_Event}
   > SetVisibility(
    self=Get HBox_IslandCode_Outer() -> (HBox_IslandCode_Outer:Horizontal Box Object Reference)
    InVisibility=Select(
     Option 0=SelfHitTestInvisible
     Option 1=Collapsed
     Index="Equal (Enum)"(A:ECreativePublishedIslandCodeDisplay Enum=@n8.PublishedIslandCodeDisplay, B:ECreativePublishedIslandCodeDisplay Enum=Off)
    )
   )
   > UpdateCreatorInfoVisibility()
   n9 = Event OnIslandTitleChanged!() -> (OutputDelegate, then, IslandName:"Text (by ref)") {is=title K2Node_Event}
   > "SetText (Text)"(self=Get Text_IslandTitle() -> (Text_IslandTitle:Common Text Object Reference), InText=@n9.IslandName)
  ]}
 GetPlayerControllerAthena {nodes [
   Inputs()
   > n1 = Cast To FortPlayerControllerAthena(Object=GetOwningPlayer())
   > Outputs!(then:Exec, CastFailed:Exec=@n1.CastFailed, AsFortPlayerController:Fort Player Controller Athena Object Reference=@n1)
  ]}
 HideIfTextEmpty {nodes [
   n1 = HideIfTextEmpty!() -> (then, TextToCheck:Text Object Reference, ElementToHide:Widget Object Reference) {is=title K2Node_FunctionEntry}
   n2 = "GetText (Text)"!(self:Text Object Reference=@n1) -> (ReturnValue:Text)
   > SetVisibility(self=@n1.ElementToHide, InVisibility=Select(Option 0=SelfHitTestInvisible, Option 1=Collapsed, Index=TextIsEmpty(InText=@n2)))
  ]}
 OnStartingGame {nodes [OnStartingGame() {is=title K2Node_FunctionEntry}]}
 PlayIntro {nodes [
   PlayIntro() {is=title K2Node_FunctionEntry}
   > PlayAnimation(InAnimation=Get Intro())
  ]}
 Set Text Color {nodes [
   n1 = Set Text Color!() -> (then, InTextColor:Linear Color Structure, InSpecificIslandCodeColor:Linear Color Structure) {is=title K2Node_FunctionEntry}
   > SetColorAndOpacity!(
    execute
    self=[Get Text_CreatorLabel() -> (Text_CreatorLabel:Common Text Object Reference), Get Text_IslandCodeLabel(), Get Text_CreatorGameTag()]
    InColorAndOpacity_SpecifiedColor:Linear Color Structure=@n1
    InColorAndOpacity_ColorUseRule:ESlateColorStylingMode Enum
   ) -> (then)
   > SetColorAndOpacity(
    self=Get Text_IslandCode()
    InColorAndOpacity_SpecifiedColor=Get InSpecificIslandCodeColor() -> (InSpecificIslandCodeColor:Linear Color Structure)
   )
  ]}
 UpdateCreatorInfoVisibility {nodes [
   UpdateCreatorInfoVisibility() {is=title K2Node_FunctionEntry}
   > n1 = Get!(execute:Exec) -> (then:Exec, else:Exec, Lobby_Privacy_Information_Widget:User Widget Object Reference)
   n2 = IsVisible(self=Get Lobby_Privacy_Information_Widget() -> (Lobby_Privacy_Information_Widget:User Widget Object Reference))
   n3 = IsVisible(self=Get CreatorBox())
   n4 = AND Boolean(A=IsVisible(self=Get HBox_IslandCode_Outer()), B=IsVisible(self=Get IslandCodeBox()))
   > SetVisibility(
    self=Get CreatorInfo()
    InVisibility=Select(Option 0=Collapsed, Option 1=SelfHitTestInvisible, Index=OR Boolean!(A=@n4, B=@n3, C:bool=@n2) -> (ReturnValue))
   )
   n5 = IsVisible(self=Get CreatorBox())
   n6 = AND Boolean(A=IsVisible(self=Get HBox_IslandCode_Outer()), B=IsVisible(self=Get IslandCodeBox()))
   SetVisibility(
    execute=@n1.else
    self=Get CreatorInfo()
    InVisibility=Select(Option 0=Collapsed, Option 1=SelfHitTestInvisible, Index=OR Boolean!(A=@n6, B=@n5) -> (ReturnValue))
   )
  ]}
]
variables [
 bDisplayTags {type=boolean, default=true}
 Lobby_Privacy_Information_Widget {type=UserWidget, default=null}
 OnStartingGame {}
 SpecificIslandCodeColor {type=LinearColor, default=<rgba 0.2159 0.6584 1 1>}
 TextColor {type=LinearColor, default=<rgba 0.007 0.314 1 1>}
]
widgets {
 info {inheritedWidgetCount=7, parentClass=/Script/FortniteUI.FortCreativeMiniGameInfo, rootWidgetClass=&HorizontalBox, widgetCount=20}
 widgets [
  Root {widgetClassPath=&HorizontalBox, widget=Root}
  Content {
   widgetClassPath=&VerticalBox
   bIsVariable=true
   parent=Root
   widget=Content
   slot {class=HorizontalBoxSlot, properties {size {sizeRule=Fill, value=1}}}
  }
  CreatorInfo {
   widgetClassPath=&HorizontalBox
   bIsVariable=true
   parent=Content
   widget=CreatorInfo
   slot {class=VerticalBoxSlot, properties {horizontalAlignment=HAlign_Left, padding=<ltrb 0 0 0 16>}}
  }
  CreatorBox {
   widgetClassPath=&HorizontalBox
   bIsVariable=true
   parent=CreatorInfo
   widget=CreatorBox
   slot {class=HorizontalBoxSlot, properties {padding=<ltrb 0 0 24 0>}}
  }
  Text_CreatorLabel {
   widgetClassPath=&CommonTextBlock
   bIsVariable=true
   parent=CreatorBox
   widget=Text_CreatorLabel
   properties {
    text=Creator:
    textTransformPolicy=ToUpper
    colorAndOpacity {colorUseRule=UseColor_Specified, specifiedColor=<rgba 1 1 1 0.6>}
    font {
     bForceMonospaced=false
     bMaterialIsStencil=false
     fontMaterial=null
     fontObject=&BurbankSmall
     letterSpacing=0
     monospacedWidth=1
     size=14
     skewAmount=0
     typefaceFontName=Black
     outlineSettings {
      bApplyOutlineToDropShadows=false
      bMiteredCorners=false
      bSeparateFillAlpha=false
      outlineMaterial=null
      outlineSize=1
      outlineColor=<rgba 0 0 0 0.1>
     }
    }
    shadowOffset=<0 0>
   }
   slot {class=HorizontalBoxSlot, properties {verticalAlignment=VAlign_Center}}
  }
  Spacer_205 {widgetClassPath=&Spacer, parent=CreatorBox, widget=Spacer_205, properties {size=<8 1>}, slot {class=HorizontalBoxSlot}}
  Text_CreatorGameTag {
   widgetClassPath=&CommonTextBlock
   bIsVariable=true
   parent=CreatorBox
   widget=Text_CreatorGameTag
   properties {
    text=Creator Name
    textTransformPolicy=ToUpper
    colorAndOpacity {colorUseRule=UseColor_Specified, specifiedColor=<rgba 1 1 1 0.6>}
    font {
     bForceMonospaced=false
     bMaterialIsStencil=false
     fontMaterial=null
     fontObject=&BurbankSmall
     letterSpacing=0
     monospacedWidth=1
     size=14
     skewAmount=0
     typefaceFontName=Black
     outlineSettings {
      bApplyOutlineToDropShadows=false
      bMiteredCorners=false
      bSeparateFillAlpha=false
      outlineMaterial=null
      outlineSize=1
      outlineColor=<rgba 0 0 0 0.1>
     }
    }
    shadowOffset=<0 0>
   }
   slot {class=HorizontalBoxSlot, properties {verticalAlignment=VAlign_Center}}
  }
  HBox_IslandCode_Outer {
   widgetClassPath=&HorizontalBox
   bIsVariable=true
   parent=CreatorInfo
   widget=HBox_IslandCode_Outer
   slot {class=HorizontalBoxSlot, properties {horizontalAlignment=HAlign_Left, verticalAlignment=VAlign_Center, padding=<ltrb 0 0 24 0>}}
  }
  IslandCodeBox {
   widgetClassPath=&HorizontalBox
   bIsVariable=true
   parent=HBox_IslandCode_Outer
   widget=IslandCodeBox
   slot {class=HorizontalBoxSlot}
  }
  Text_IslandCodeLabel {
   widgetClassPath=&CommonTextBlock
   bIsVariable=true
   parent=IslandCodeBox
   widget=Text_IslandCodeLabel
   properties {
    text=Island Code:
    textTransformPolicy=ToUpper
    wrapTextAt=500
    colorAndOpacity {colorUseRule=UseColor_Specified, specifiedColor=<rgba 1 1 1 0.6>}
    font {
     bForceMonospaced=false
     bMaterialIsStencil=false
     fontMaterial=null
     fontObject=&BurbankSmall
     letterSpacing=0
     monospacedWidth=1
     size=14
     skewAmount=0
     typefaceFontName=Black
     outlineSettings {
      bApplyOutlineToDropShadows=false
      bMiteredCorners=false
      bSeparateFillAlpha=false
      outlineMaterial=null
      outlineSize=0
      outlineColor=<rgba 0 0 0 1>
     }
    }
    shadowOffset=<0 0>
   }
   slot {class=HorizontalBoxSlot}
  }
  Spacer {widgetClassPath=&Spacer, parent=IslandCodeBox, widget=Spacer, properties {size=<8 1>}, slot {class=HorizontalBoxSlot}}
  Text_IslandCode {
   widgetClassPath=&CommonTextBlock
   bIsVariable=true
   parent=IslandCodeBox
   widget=Text_IslandCode
   properties {
    text="[PH] 1234-1234-0348"
    wrapTextAt=500
    colorAndOpacity {colorUseRule=UseColor_Specified, specifiedColor=<rgba 1 1 1 0.6>}
    font {
     bForceMonospaced=false
     bMaterialIsStencil=false
     fontMaterial=null
     fontObject=&BurbankSmall
     letterSpacing=0
     monospacedWidth=1
     size=14
     skewAmount=0
     typefaceFontName=Black
     outlineSettings {
      bApplyOutlineToDropShadows=false
      bMiteredCorners=false
      bSeparateFillAlpha=false
      outlineMaterial=null
      outlineSize=0
      outlineColor=<rgba 0 0 0 1>
     }
    }
    shadowOffset=<0 0>
   }
   slot {class=HorizontalBoxSlot}
  }
  Text_IslandTitle {
   widgetClassPath=&CommonTextBlock
   bIsVariable=true
   parent=Content
   widget=Text_IslandTitle
   properties {
    mobileFontSizeMultiplier=0.7
    text="[PH] Grand Theft Apples"
    textTransformPolicy=ToUpper
    font {
     bForceMonospaced=false
     bMaterialIsStencil=false
     fontMaterial=null
     fontObject=&Burbank
     letterSpacing=0
     monospacedWidth=1
     size=48
     skewAmount=0
     typefaceFontName=BurbankRegular-Black
     outlineSettings {
      bApplyOutlineToDropShadows=false
      bMiteredCorners=false
      bSeparateFillAlpha=false
      outlineMaterial=null
      outlineSize=1
      outlineColor=<rgba 0 0 0 0.1>
     }
    }
    renderTransform {angle=0, scale=<1 1>, shear=<-8 0>, translation=<0 0>}
    shadowOffset=<0 0>
   }
   slot {class=VerticalBoxSlot, properties {horizontalAlignment=HAlign_Left}}
  }
  EntryBox_MiniGameTags {
   widgetClassPath=&DynamicEntryBox
   bInherited=true
   bIsVariable=true
   parent=Content
   widget=EntryBox_MiniGameTags
   properties {
    entryHorizontalAlignment=HAlign_Left
    entryVerticalAlignment=VAlign_Top
    entryWidgetClass=/Game/UI/DiscoveryShared/ActivityBrowserTag.ActivityBrowserTag_C
    numDesignerPreviewEntries=7
    entrySpacing=<2 8>
   }
   slot {class=VerticalBoxSlot, properties {horizontalAlignment=HAlign_Left, verticalAlignment=VAlign_Top, padding=<ltrb 0 16 0 0>}}
  }
  Text_Name {widgetClassPath=&CommonTextBlock, bInherited=true, widget=null}
  Text_CreatorName {widgetClassPath=&CommonTextBlock, bInherited=true, widget=null}
  Text_SupportCode {widgetClassPath=&CommonTextBlock, bInherited=true, widget=null}
  Text_Bullet1 {widgetClassPath=&CommonTextBlock, bInherited=true, widget=null}
  Text_Bullet2 {widgetClassPath=&CommonTextBlock, bInherited=true, widget=null}
  Text_Bullet3 {widgetClassPath=&CommonTextBlock, bInherited=true, widget=null}
 ]
}
