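# Countdown widget in the Creative start screen HUD folder, showing a timer icon beside a formatted value and unit.
# UpdateCountdown writes the formatted text to a main text block and an overlaid echo copy.
# At or below a threshold value it plays TimerPulse, a one-second bump of the main text with a scaling, fading echo ghost.
# PreConstruct feeds a test value at design time for editor preview.
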WidgetBlueprint /Game/Athena/HUD/Creative/StartScreen/WBP_Countdown
animations [TimerPulse {
  binding_count=2
  length_seconds=1
  object=&self:TimerPulse
  sequence {
   bindings [
    Text_Countdown_Echo {
     id=#1
     possessed_class=CommonTextBlock
     tracks [
      {
       type=MovieSceneFloatTrack
       display_name=Render Opacity
       sections [{
         type=MovieSceneFloatSection
         channels [None {type=MovieSceneScriptingFloatChannel, keys [[26.6667,0],[40.6983,0.3],[46.7117,0.3],[66.7567,0]]}]
        }]
      }
      {
       type=MovieScene2DTransformTrack
       sections [{
         type=MovieScene2DTransformSection
         end_frame=66
         start_frame=6
         channels [
          Translation.X {type=MovieSceneScriptingFloatChannel, keys [[26.6667,0],[40.6983,4.85],[46.7117,4.8],[66.505,0]]}
          Translation.Y {type=MovieSceneScriptingFloatChannel, const [0,1]}
          Angle {type=MovieSceneScriptingFloatChannel, const [0,1]}
          Scale.X {type=MovieSceneScriptingFloatChannel, keys [[26.6667,1],[46.9617,2],[66.505,1.5001]]}
          Scale.Y {type=MovieSceneScriptingFloatChannel, keys [[26.6667,1],[46.9617,2],[66.505,1.5001]]}
          Shear.X {type=MovieSceneScriptingFloatChannel, const [0,1]}
          Shear.Y {type=MovieSceneScriptingFloatChannel, const [0,1]}
         ]
        }]
      }
     ]
    }
    Text_Countdown {
     id=#2
     possessed_class=CommonTextBlock
     tracks [{
       type=MovieScene2DTransformTrack
       sections [{
         type=MovieScene2DTransformSection
         channels [
          Translation.X {type=MovieSceneScriptingFloatChannel, keys [[0,0],[26.6667,7.8399],[53.3333,0],[100,0]]}
          Translation.Y {type=MovieSceneScriptingFloatChannel}
          Angle {type=MovieSceneScriptingFloatChannel}
          Scale.X {type=MovieSceneScriptingFloatChannel, keys [[0,1],[26.6667,1.2],[53.3333,1],[100,1]]}
          Scale.Y {type=MovieSceneScriptingFloatChannel, keys [[0,1],[26.6667,1.2],[53.3333,1],[100,1]]}
         ]
        }]
      }]
    }
   ]
   display_rate {numerator=100}
  }
 }]
graphs [
 EventGraph {nodes [
   n1 = Event PreConstruct()
   > Branch(Condition=@n1.IsDesignTime)
   > UpdateCountdown!(execute, self, InCountdownValue:int=Get Test Countdown Value() -> (Test Countdown Value:int=0), InUnit:Text="LOCTEXT(\"s\")") -> (then)
   Event Construct()
   Event Tick()
  ]}
 UpdateCountdown {nodes [
   n1 = UpdateCountdown!() -> (then, InCountdownValue:int, InUnit:Text) {is=title K2Node_FunctionEntry}
   n2 = Format Text!(
    Format=Get FormatCountdown() -> (FormatCountdown:Text)
    CountdownValue:int=Get InCountdownValue() -> (InCountdownValue:int=0)
    Unit:Text=Get InUnit() -> (InUnit:Text)
   ) -> (Result)
   > "SetText (Text)"(self=Get Text_Countdown() -> (Text_Countdown:Common Text Object Reference), InText=@n2)
   > "SetText (Text)"(self=Get Text_Countdown_Echo() -> (Text_Countdown_Echo:Common Text Object Reference), InText=@n2)
   > Branch(Condition="integer <= integer"(A=@n1, B=Get Countdown Value Animation Threshold() -> (Countdown Value Animation Threshold:int=0)))
   > PlayAnimation(InAnimation=Get TimerPulse() -> (TimerPulse:Widget Animation Object Reference))
  ]}
]
variables [
 Countdown Value Animation Threshold {type=integer, default=15}
 FormatCountdown {type=string, default="{CountdownValue}{Unit}"}
 Test Countdown Value {type=integer, default=99}
]
widgets {
 info {inheritedWidgetCount=0, parentClass=&UserWidget, rootWidgetClass=&HorizontalBox, widgetCount=6}
 widgets [
  Root {widgetClassPath=&HorizontalBox, widget=Root}
  Image_Icon_Countdown {
   widgetClassPath=&Image
   parent=Root
   widget=Image_Icon_Countdown
   properties {brush {
     bIsDynamicallyLoaded=false
     drawAs=Image
     imageType=NoImage
     mirroring=NoMirror
     resourceName=None
     resourceObject=/Game/Athena/UI/Frontend/ActivityBrowser/Art/M_UI_Timer_Icon_ShopOffer
     tiling=NoTile
     imageSize=<68 68>
     margin=<ltrb 0 0 0 0>
     outlineSettings {
      bUseBrushTransparency=false
      roundingType=HalfHeightRadius
      width=0
      color {colorUseRule=UseColor_Specified, specifiedColor=<rgba 0 0 0 0>}
      cornerRadii=<0 0 0 0>
     }
     tintColor {colorUseRule=UseColor_Specified, specifiedColor=<rgba 1 1 1 1>}
     uVRegion {bIsValid=false, max=<0 0>, min=<0 0>}
    }}
   slot {class=HorizontalBoxSlot, properties {horizontalAlignment=HAlign_Left, verticalAlignment=VAlign_Bottom}}
  }
  Overlay_1 {
   widgetClassPath=&Overlay
   parent=Root
   widget=Overlay_1
   slot {class=HorizontalBoxSlot, properties {horizontalAlignment=HAlign_Left, verticalAlignment=VAlign_Bottom, padding=<ltrb 4 6 0 7>}}
  }
  Spacer_93 {widgetClassPath=&Spacer, parent=Overlay_1, widget=Spacer_93, properties {size=<60 0>}, slot {class=OverlaySlot}}
  Text_Countdown {
   widgetClassPath=&CommonTextBlock
   bIsVariable=true
   parent=Overlay_1
   widget=Text_Countdown
   properties {
    justification=Center
    text=99s
    textTransformPolicy=ToLower
    font {
     bForceMonospaced=false
     bMaterialIsStencil=false
     fontMaterial=null
     fontObject=&BurbankSmall
     letterSpacing=0
     monospacedWidth=1
     size=24
     skewAmount=0
     typefaceFontName=Black
     outlineSettings {
      bApplyOutlineToDropShadows=false
      bMiteredCorners=false
      bSeparateFillAlpha=false
      outlineMaterial=null
      outlineSize=0
      outlineColor=<rgba 0 0 0 1>
     }
    }
    shadowOffset=<0 0>
   }
   slot {class=OverlaySlot, properties {verticalAlignment=VAlign_Bottom}}
  }
  Text_Countdown_Echo {
   widgetClassPath=&CommonTextBlock
   bIsVariable=true
   parent=Overlay_1
   widget=Text_Countdown_Echo
   properties {
    justification=Center
    text=99s
    textTransformPolicy=ToLower
    font {
     bForceMonospaced=false
     bMaterialIsStencil=false
     fontMaterial=null
     fontObject=&BurbankSmall
     letterSpacing=0
     monospacedWidth=1
     size=24
     skewAmount=0
     typefaceFontName=Black
     outlineSettings {
      bApplyOutlineToDropShadows=false
      bMiteredCorners=false
      bSeparateFillAlpha=false
      outlineMaterial=null
      outlineSize=0
      outlineColor=<rgba 0 0 0 1>
     }
    }
    shadowOffset=<0 0>
   }
   slot {class=OverlaySlot, properties {verticalAlignment=VAlign_Bottom}}
  }
 ]
}
