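# Creative minigame pregame HUD widget showing an auto-start countdown and joined/maximum player count.
# OnDisplayMinigamePregameCountdownUI starts or clears a one-second looping timer, toggling the "Waiting for Players" throbber for public matchmade games.
# Each tick reads CurrentMinigame's time and players, hiding the count when max is one, then plays final-seconds tick and end cues.
# PreConstruct tints text and throbber via dynamic material and fills designer preview values.
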WidgetBlueprint /Game/Athena/HUD/Creative/StartScreen/WBP_Creative_Minigame_Joining
graphs [
 EventGraph {nodes [
   n1 = Update Message With Delay() {is=title K2Node_CustomEvent}
   > UpdateMessage()
   n2 = Event OnDisplayMinigamePregameCountdownUI!() -> (OutputDelegate, then, bDisplay:bool) {is=title K2Node_Event}
   > n3 = Branch(Condition=@n2.bDisplay)
   > n4 = Sequence()
   > n5 = Set Timer by Event(Delegate=@n1, Time=1, bLooping=true)
   > Set TimerHandle(TimerHandle:Timer Handle Structure=@n5) -> (Output_Get:Timer Handle Structure)
   > Update Message With Delay() {is=title K2Node_CallFunction}
   Branch(execute=@n4.then_1)
   > n6 = Cast To Athena_PlayerController!(execute, Object=GetOwningPlayer()) -> (then, CastFailed, AsAthena Player Controller:Athena Player Controller Object Reference) {is=title K2Node_DynamicCast}
   > SetVisibility(
    self=[Get Throbber() -> (Throbber:Throbber Object Reference), Get Text_PlayerInfoPhrase() -> (Text_PlayerInfoPhrase:Common Text Object Reference)]
    InVisibility=Select(
     Option 0=Collapsed
     Option 1=SelfHitTestInvisible
     Index=IsInPublicMatchmadeCreativeMinigame!(self:Fort Player Controller Athena Object Reference=@n6) -> (ReturnValue:bool)
    )
   )
   Clear and Invalidate Timer by Handle(execute=@n3.else, Handle=Get TimerHandle() -> (TimerHandle:Timer Handle Structure))
   > SetVisibility(
    self=[
     Get SizeBox_Timer() -> (SizeBox_Timer:Size Box Object Reference)
     Get HorizontalBox_PlayersInfo() -> (HorizontalBox_PlayersInfo:Horizontal Box Object Reference)
    ]
    InVisibility=Collapsed
   )
   > Set bWasVisible(bWasVisible:bool=false) -> (Output_Get:bool)
   n7 = Event PreConstruct()
   n8 = Get PlayerInfoTextColor() -> (PlayerInfoTextColor:Linear Color Structure)
   > SetTextColor!(execute, self, InPlayerInfoTextColor:Linear Color Structure=@n8) -> (then)
   > Set Throbber Parameters!(
    execute
    self
    InThrobberColor:Linear Color Structure=@n8
    InNumberOfPieces:int=Get ThrobberNumberOfPieces() -> (ThrobberNumberOfPieces:int=0)
   ) -> (then) {is=title K2Node_CallFunction}
   > Branch(Condition=@n7.IsDesignTime)
   > SetDesignerTextPreview!(execute, self, InTimeToDisplay:int=36, InTotalPlayers:int=2, InMaxPlayers:int=8) -> (then)
  ]}
 Set Throbber Parameters {nodes [
   Set Throbber Parameters!() -> (then, InThrobberColor:Linear Color Structure, InNumberOfPieces:int) {is=title K2Node_FunctionEntry}
   n1 = Get Throbber()
   n2 = Get!(self:Throbber Object Reference=@n1) -> (
    Image_TintColor:Slate Color Structure
    Image_DrawAs:ESlateBrushDrawType Enum
    Image_Tiling:ESlateBrushTileType Enum
    Image_Mirroring:ESlateBrushMirrorType Enum
    Image_ImageSize:Vector 2D Structure
    Image_Margin:Margin Structure
    Image_ResourceObject:Object Reference
    Image_OutlineSettings:Slate Brush Outline Settings Structure
   )
   > n3 = Set!(
    execute
    Image_TintColor:Slate Color Structure=@n2
    Image_DrawAs:ESlateBrushDrawType Enum=@n2.Image_DrawAs
    Image_Tiling:ESlateBrushTileType Enum=@n2.Image_Tiling
    Image_Mirroring:ESlateBrushMirrorType Enum=@n2.Image_Mirroring
    Image_ImageSize:Vector 2D Structure=@n2.Image_ImageSize
    Image_Margin:Margin Structure=@n2.Image_Margin
    Image_ResourceObject:Object Reference=Get ThrobberMaterial() -> (ThrobberMaterial:Object Reference)
    Image_OutlineSettings:Slate Brush Outline Settings Structure=@n2.Image_OutlineSettings
    self:Throbber Object Reference=@n1
   ) -> (then, Output_Get:Slate Brush Structure)
   > SetVectorParameterValue(self=GetDynamicMaterial!(Brush:"Slate Brush Structure (by ref)"=@n3) -> (ReturnValue), ParameterName=Color, Value=Get PlayerInfoTextColor())
   > SetNumberOfPieces!(execute, self:Throbber Object Reference=Get Throbber(), InNumberOfPieces:int=Get InNumberOfPieces() -> (InNumberOfPieces:int=0)) -> (then)
  ]}
 SetDesignerTextPreview {nodes [
   SetDesignerTextPreview!() -> (then, InTimeToDisplay:int, InTotalPlayers:int, InMaxPlayers:int) {is=title K2Node_FunctionEntry}
   > n1 = Set TimeToDisplay(TimeToDisplay:int=Get InTimeToDisplay() -> (InTimeToDisplay:int=0)) -> (Output_Get:int=0)
   > UpdateCountdown!(execute, self, InCountdownValue:int=@n1) -> (then)
   > "SetText (Text)"(
    self=Get Text_MatchmakingPlayers() -> (Text_MatchmakingPlayers:Common Text Object Reference)
    InText=Format Text!(
     Format=Get MatchmakingFormat() -> (MatchmakingFormat:Text)
     TotalPlayers:int=Get InTotalPlayers() -> (InTotalPlayers:int=0)
     MaxPlayers:int=Get InMaxPlayers() -> (InMaxPlayers:int=0)
    ) -> (Result)
   )
  ]}
 SetTextColor {nodes [
   n1 = SetTextColor!() -> (then, InPlayerInfoTextColor:Linear Color Structure) {is=title K2Node_FunctionEntry}
   > SetColorAndOpacity!(
    execute
    self=Get Text_PlayerInfoPhrase()
    InColorAndOpacity_SpecifiedColor:Linear Color Structure=@n1
    InColorAndOpacity_ColorUseRule:ESlateColorStylingMode Enum
   ) -> (then)
  ]}
 SetTexts {nodes [
   SetTexts() {is=title K2Node_FunctionEntry}
   > SetVisibility(self=Get Root() -> (Root:Horizontal Box Object Reference), InVisibility=SelfHitTestInvisible)
   n1 = Get CurrentMinigame() -> (CurrentMinigame:Fort Minigame Object Reference)
   > n2 = Branch(Condition="float > float"(A=GetTimeRemainingUntilAutoStart!(self:Fort Minigame Object Reference=@n1) -> (ReturnValue:float), B=0) {is=title K2Node_CallFunction})
   n3 = Get SizeBox_Timer()
   > SetVisibility(self=@n3, InVisibility=SelfHitTestInvisible)
   > n4 = Sequence!(execute) -> (then_0, then_1, then_2:Exec)
   > n5 = Cast To FortPlayerControllerAthena(Object=GetOwningPlayer())
   n6 = GetCreativeIslandMatchmakingMaxPlayers!(PlayerControllerAthena:Fort Player Controller Athena Object Reference=@n5) -> (ReturnValue:int=0)
   > n7 = Branch(Condition="Equal (Integer)"(A=@n6, B=1))
   > SetVisibility(self=Get HorizontalBox_PlayersInfo(), InVisibility=Collapsed)
   n8 = Get CurrentMinigame()
   "SetText (Text)"(
    execute=@n7.else
    self=Get Text_MatchmakingPlayers()
    InText=Format Text(
     Format=Get MatchmakingFormat()
     TotalPlayers=GetPlayerCount!(self:Fort Minigame Object Reference=@n8, bIncludeSpectators:bool=true) -> (ReturnValue:int=0)
     MaxPlayers=@n6
    )
   )
   > SetVisibility(self=Get HorizontalBox_PlayersInfo(), InVisibility=SelfHitTestInvisible)
   n9 = GetTimeRemainingUntilAutoStart(self=Get CurrentMinigame())
   n10 = Set TimeToDisplay(execute=@n4.then_1, TimeToDisplay=Round(A=@n9) {is=title K2Node_CallFunction}) -> (Output_Get=0)
   > Branch(Condition="integer >= integer"!(A:int=@n10, B:int=1) -> (ReturnValue:bool) {is=title K2Node_CallFunction})
   > UpdateCountdown(InCountdownValue=@n10)
   SetVisibility(execute=@n2.else, self=@n3, InVisibility=Collapsed)
   Floor(A=@n9) {is=title K2Node_CallFunction}
  ]}
 UpdateCountdown {nodes [
   UpdateCountdown!() -> (then, InCountdownValue:int) {is=title K2Node_FunctionEntry}
   > UpdateCountdown!(
    execute
    self:WBP Countdown Object Reference=Get WBP_UIkit_Countdown() -> (WBP_UIkit_Countdown:WBP Countdown Object Reference)
    InCountdownValue=Get InCountdownValue() -> (InCountdownValue:int=0)
    InUnit:Text=Get CountdownUnit() -> (CountdownUnit:Text)
   ) -> (then) {is=title K2Node_CallFunction}
  ]}
 UpdateMessage {nodes [
   UpdateMessage() {is=title K2Node_FunctionEntry}
   > SetTexts() {is=title K2Node_CallFunction}
   > n1 = Branch(Condition="integer >= integer"(A=Get TimeToDisplay() -> (TimeToDisplay:int=0), B=1))
   > n2 = Set bWasVisible(bWasVisible=true)
   n3 = Decrement Int(" "=[@n1.else, @n2.then], Value=Get TimeToDisplay() -> (TimeToDisplay=0))
   > UpdateCountdown!(execute, self:Self Object Reference, InCountdownValue="Max (Integer)"(A=@n3, B=0)) -> (then)
   > n4 = Branch(Condition="integer <= integer"!(A:int=Get TimeToDisplay() -> (TimeToDisplay=0), B:int=0) -> (ReturnValue:bool) {is=title K2Node_CallFunction})
   > Branch(Condition=Get bWasVisible() -> (bWasVisible:bool))
   > PlaySound2D(Sound=/Game/Sounds/Creative/Toys/Floor_Siren/Timer_UI_Tick_Minigame_End_Cue)
   > Clear and Invalidate Timer by Handle(Handle=Get TimerHandle())
   n5 = Branch(execute=@n4.else, Condition="integer > integer"(A=Get TimeToDisplay() -> (TimeToDisplay=0), B=20) {is=title K2Node_CallFunction})
   n6 = Get TimeToDisplay() -> (TimeToDisplay=0)
   PlaySound2D(
    execute=@n5.else
    Sound=/Game/Sounds/Creative/Toys/Floor_Siren/Timer_UI_Tick_Minigame_Cue
    VolumeMultiplier=Select(Option 0:double=1, Option 1:double=0.5, Index="integer > integer"(A=@n6, B=5)) -> (ReturnValue:double)
   )
  ]}
]
variables [
 bWasVisible {type=boolean}
 CountdownUnit {type=string, default=s}
 MatchmakingFormat {type=string, default="{TotalPlayers}/{MaxPlayers}"}
 PlayerInfoTextColor {type=LinearColor, default=<rgba 0.0044 0.3712 0.8308 1>}
 ThrobberMaterial {type=Object, default=/Game/Athena/HUD/Creative/MaterialInstances/MI_UI_Creative_ThrobberShapeAndColor}
 ThrobberNumberOfPieces {type=integer, default=3}
 TimerHandle {type=TimerHandle}
 TimeToDisplay {type=integer}
]
widgets {
 info {inheritedWidgetCount=1, parentClass=/Script/FortniteUI.MinigameWidgetBase, rootWidgetClass=&HorizontalBox, widgetCount=11}
 widgets [
  Root {widgetClassPath=&HorizontalBox, bIsVariable=true, widget=Root}
  SizeBox_Timer {
   widgetClassPath=&SizeBox
   bIsVariable=true
   parent=Root
   widget=SizeBox_Timer
   properties {bOverride_MinDesiredWidth=true, minDesiredWidth=150}
   slot {class=HorizontalBoxSlot, properties {horizontalAlignment=HAlign_Left}}
  }
  WBP_UIkit_Countdown {
   widgetClassPath=&/WBP_Countdown
   bIsVariable=true
   parent=SizeBox_Timer
   widget=WBP_UIkit_Countdown
   slot {class=SizeBoxSlot}
  }
  HorizontalBox_PlayersInfo {
   widgetClassPath=&HorizontalBox
   bIsVariable=true
   parent=Root
   widget=HorizontalBox_PlayersInfo
   slot {class=HorizontalBoxSlot, properties {padding=<ltrb 20 0 10 0>}}
  }
  Image_Icon_Players {
   widgetClassPath=&Image
   parent=HorizontalBox_PlayersInfo
   widget=Image_Icon_Players
   properties {brush {
     bIsDynamicallyLoaded=false
     drawAs=Image
     imageType=NoImage
     mirroring=NoMirror
     resourceName=None
     resourceObject=/Game/UI/Sidebar/Textures/T_UI_Icon_Sidebar_PartyUp
     tiling=NoTile
     imageSize=<64 64>
     margin=<ltrb 0 0 0 0>
     outlineSettings {
      bUseBrushTransparency=false
      roundingType=HalfHeightRadius
      width=0
      color {colorUseRule=UseColor_Specified, specifiedColor=<rgba 0 0 0 0>}
      cornerRadii=<0 0 0 0>
     }
     tintColor {colorUseRule=UseColor_Specified, specifiedColor=<rgba 1 1 1 1>}
     uVRegion {bIsValid=false, max=<0 0>, min=<0 0>}
    }}
   slot {class=HorizontalBoxSlot, properties {horizontalAlignment=HAlign_Left, verticalAlignment=VAlign_Bottom, padding=<ltrb 0 0 0 -6>}}
  }
  Overlay_2 {
   widgetClassPath=&Overlay
   parent=HorizontalBox_PlayersInfo
   widget=Overlay_2
   slot {class=HorizontalBoxSlot, properties {horizontalAlignment=HAlign_Left, verticalAlignment=VAlign_Bottom, padding=<ltrb 0 0 0 7>}}
  }
  HorizontalBox_1 {
   widgetClassPath=&HorizontalBox
   parent=Overlay_2
   widget=HorizontalBox_1
   slot {class=OverlaySlot, properties {horizontalAlignment=HAlign_Fill, verticalAlignment=VAlign_Bottom, padding=<ltrb 12 0 0 0>}}
  }
  Text_MatchmakingPlayers {
   widgetClassPath=&CommonTextBlock
   bIsVariable=true
   parent=HorizontalBox_1
   widget=Text_MatchmakingPlayers
   properties {
    justification=Center
    text=2/8
    font {
     bForceMonospaced=false
     bMaterialIsStencil=false
     fontMaterial=null
     fontObject=&BurbankSmall
     letterSpacing=0
     monospacedWidth=1
     size=24
     skewAmount=0
     typefaceFontName=Black
     outlineSettings {
      bApplyOutlineToDropShadows=false
      bMiteredCorners=false
      bSeparateFillAlpha=false
      outlineMaterial=null
      outlineSize=0
      outlineColor=<rgba 0 0 0 1>
     }
    }
    shadowOffset=<0 0>
   }
   slot {class=HorizontalBoxSlot, properties {horizontalAlignment=HAlign_Left, verticalAlignment=VAlign_Bottom}}
  }
  Text_PlayerInfoPhrase {
   widgetClassPath=&CommonTextBlock
   bIsVariable=true
   parent=HorizontalBox_1
   widget=Text_PlayerInfoPhrase
   properties {
    text=Waiting for Players
    textTransformPolicy=ToUpper
    colorAndOpacity {colorUseRule=UseColor_Specified, specifiedColor=<rgba 0.0044 0.3712 0.8308 1>}
    font {
     bForceMonospaced=false
     bMaterialIsStencil=false
     fontMaterial=null
     fontObject=&BurbankSmall
     letterSpacing=0
     monospacedWidth=1
     size=16
     skewAmount=0
     typefaceFontName=Black
     outlineSettings {
      bApplyOutlineToDropShadows=false
      bMiteredCorners=false
      bSeparateFillAlpha=false
      outlineMaterial=null
      outlineSize=0
      outlineColor=<rgba 0 0 0 0>
     }
    }
    shadowOffset=<0 0>
   }
   slot {class=HorizontalBoxSlot, properties {horizontalAlignment=HAlign_Left, verticalAlignment=VAlign_Bottom, padding=<ltrb 15 0 0 3>}}
  }
  Throbber {
   widgetClassPath=/Script/UMG.Throbber
   bIsVariable=true
   parent=HorizontalBox_1
   widget=Throbber
   properties {
    bAnimateOpacity=false
    image {
     bIsDynamicallyLoaded=false
     drawAs=Image
     imageType=FullColor
     mirroring=NoMirror
     resourceName=../../../Engine/Content/Slate/Common/Throbber_Piece.png
     resourceObject=&MI_UI_Creative_ThrobberShapeAndColor
     tiling=NoTile
     imageSize=<16 16>
     margin=<ltrb 0 0 0 0>
     outlineSettings {
      bUseBrushTransparency=false
      roundingType=HalfHeightRadius
      width=0
      color {colorUseRule=UseColor_Specified, specifiedColor=<rgba 0 0 0 0>}
      cornerRadii=<0 0 0 0>
     }
     tintColor {colorUseRule=UseColor_Specified, specifiedColor=<rgba 1 1 1 1>}
     uVRegion {bIsValid=false, max=<0 0>, min=<0 0>}
    }
   }
   slot {class=HorizontalBoxSlot, properties {horizontalAlignment=HAlign_Left, verticalAlignment=VAlign_Bottom, padding=<ltrb 0 0 0 6>}}
  }
  TutorialHighlight {widgetClassPath=&TutorialHighlight}
 ]
}
