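# Hold-to-confirm button widget for Creative HUD screens, built on FortHoldableButton, with title, subtitle, input action icon and optional right icon.
# PreConstruct applies exposed materials, font sizes, text, size and icon visibility.
# Hold events play or reverse a material-driven progress animation scaled to HoldTime, feeding hold percentage into the fill sound's FillProgress parameter.
# Completion fades that sound and plays a completion sound; hover animates the subtitle font material.
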WidgetBlueprint /Game/Athena/HUD/Creative/Widgets/WBP_Creative_Button_Hold
animations [
 Hover {
  binding_count=1
  length_seconds=0.15
  object=&self:Hover
  sequence {bindings [Text_subtitle {
     id=#1
     possessed_class=CommonTextBlock
     tracks [{
       type=MovieSceneWidgetMaterialTrack
       display_name=Font.FontMaterial
       sections [{type=MovieSceneParameterSection, end_frame=3, start_frame=0, channels [Hover {type=MovieSceneScriptingFloatChannel, keys [[0,0],[3,1]]}]}]
      }]
    }]}
 }
 HoldProgress {
  binding_count=1
  length_seconds=2
  object=&self:HoldProgress
  sequence {bindings [Image_Progress {
     id=#2
     possessed_class=Image
     tracks [{
       type=MovieSceneWidgetMaterialTrack
       sections [{
         type=MovieSceneParameterSection
         end_frame=40
         start_frame=0
         channels [Progress {type=MovieSceneScriptingFloatChannel, keys [[0,0],[40,1]]}]
        }]
      }]
    }]}
 }
 HoldCompleted {
  binding_count=1
  length_seconds=0.05
  object=&self:HoldCompleted
  sequence {bindings [Image_Progress {
     id=#2
     possessed_class=Image
     tracks [{
       type=MovieSceneWidgetMaterialTrack
       sections [{type=MovieSceneParameterSection, end_frame=1, start_frame=0, channels [Progress {type=MovieSceneScriptingFloatChannel, const [0,1]}]}]
      }]
    }]}
 }
]
graphs [
 EventGraph {nodes [
   PlayFillSound() {is=title K2Node_CustomEvent}
   > n1 = Get!(execute:Exec) -> (then:Exec, else:Exec, FillingSoundAudioComponent:Audio Component Object Reference)
   n2 = SpawnSound2D!(
    execute=@n1.else
    Sound:Sound Base Object Reference=Get FillingSoundOverride() -> (FillingSoundOverride:Sound Base Object Reference)
    VolumeMultiplier:float=1
    PitchMultiplier:float=1
    StartTime:float=0
    ConcurrencySettings:Sound Concurrency Object Reference
    bPersistAcrossLevelTransition:bool=false
    bAutoDestroy:bool=true
   ) -> (then, ReturnValue:Audio Component Object Reference)
   > Set FillingSoundAudioComponent(FillingSoundAudioComponent:Audio Component Object Reference=@n2) -> (Output_Get:Audio Component Object Reference)
   FillComplete() {is=title K2Node_CustomEvent}
   > n3 = Get()
   > FadeOut!(
    execute
    self:Audio Component Object Reference=@n3
    FadeOutDuration:float=0.5
    FadeVolumeLevel:float=0
    FadeCurve:EAudioFaderCurve Enum=Linear
   ) -> (then)
   > PlaySound2D(Sound=Get FillCompleteSoundOverride() -> (FillCompleteSoundOverride:Sound Base Object Reference))
   StopFillSound() {is=title K2Node_CustomEvent}
   > n4 = Get()
   > FadeOut(self=@n4, FadeOutDuration=0.25, FadeVolumeLevel=0, FadeCurve=Linear)
   n5 = Get Image_Progress() -> (Image_Progress:Image Object Reference)
   n6 = Cast To MaterialInstance!(
    execute
    Object=Break Slate Brush!(SlateBrush:"Slate Brush Structure (by ref)"=Get Brush!(self:Image Object Reference=@n5) -> (Brush:Slate Brush Structure)) -> (
     TintColor:Slate Color Structure
     DrawAs:ESlateBrushDrawType Enum
     Tiling:ESlateBrushTileType Enum
     Mirroring:ESlateBrushMirrorType Enum
     ImageSize:Vector 2D Structure
     Margin:Margin Structure
     ResourceObject:Object Reference
     OutlineSettings:Slate Brush Outline Settings Structure
    ) {is=title K2Node_BreakStruct}.ResourceObject
   ) -> (then, CastFailed, AsMaterial Instance:Material Instance Object Reference) {is=title K2Node_DynamicCast}
   > n7 = CreateDynamicMaterialInstance!(execute, Parent=@n6, OptionalName=None, CreationFlags=None) -> (then, ReturnValue)
   > Set HoldProgressDMI(HoldProgressDMI:Material Instance Dynamic Object Reference=@n7) -> (Output_Get:Material Instance Dynamic Object Reference)
   n8 = Event PreConstruct()
   > n9 = Sequence()
   > SetMaterials()
   > OverrideFontSlate!(
    execute
    self
    InText:Text Object Reference=Get Text_Title() -> (Text_Title:Common Text Object Reference)
    inSize:int=Get TitleFontSize() -> (TitleFontSize:int=0)
    InFontMaterial:Object Reference=Get TitleMaterial() -> (TitleMaterial:Object Reference)
   ) -> (then)
   > OverrideFontSlate(
    InText=Get Text_subtitle() -> (Text_subtitle:Common Text Object Reference)
    inSize=Get SubtitleFontSize() -> (SubtitleFontSize:int=0)
    InFontMaterial=Get SubtitleMaterial() -> (SubtitleMaterial:Object Reference)
   )
   > "SetText (Text)"(self=Get Text_subtitle(), InText=Get In SubText() -> (In SubText:Text))
   n10 = Get Root() -> (Root:Size Box Object Reference)
   > SetMinDesiredWidth!(execute, self:Size Box Object Reference=@n10, InMinDesiredWidth:float=Get In Width Override() -> (In Width Override:double)) -> (then)
   > SetHeightOverride(self=@n10, InHeightOverride=Get In Height Override() -> (In Height Override:double))
   n11 = Get In Height Override()
   > SetDesiredSizeOverride!(
    execute
    self:Image Object Reference=[Get Image_Background() -> (Image_Background:Image Object Reference), Get Image_Progress()]
    DesiredSize:Vector 2D Structure=MakeVector2D(X=@n11, Y=@n11)
   ) -> (then)
   > SetVisibility(
    self=[Get Spacer_Icon() -> (Spacer_Icon:Spacer Object Reference), Get Image_Icon() -> (Image_Icon:Image Object Reference)]
    InVisibility=Select(Option 0=Collapsed, Option 1=HitTestInvisible, Index=Get Show Right icon() -> (Show Right icon:bool))
   )
   > Set Justification!(
    execute
    Justification:ETextJustify Enum=Get Justification() -> (Justification:ETextJustify Enum=Left)
    self:Text Layout Widget Object Reference=Get Text_Title()
   ) -> (then, Output_Get:ETextJustify Enum=Left)
   "SetText (Text)"(execute=@n9.then_1, self=Get Text_Title(), InText=Get In Text() -> (In Text:Text))
   > "SetText (Text)"(self=Get Text_subtitle(), InText=Get In SubText())
   n12 = Event BP_OnHoldStarted!() -> (OutputDelegate, then, HoldPercentage:float, bInUseSecondaryHoldAnimation:bool) {is=title K2Node_Event}
   n13 = Get HoldProgress() -> (HoldProgress:Widget Animation Object Reference)
   n14 = GetEndTime(self=@n13)
   > PlayAnimation(
    InAnimation=@n13
    StartAtTime=float * float(A=@n14, B=@n12.HoldPercentage)
    PlaybackSpeed=float / float(A=@n14, B=Get HoldTime() -> (HoldTime:float)) {is=title K2Node_CallFunction}
   )
   n15 = Event BP_OnHoldEnded!() -> (OutputDelegate, then, HoldPercentage:float, bInUseSecondaryHoldAnimation:bool) {is=title K2Node_Event}
   > Branch(Condition="float > float"(A=@n15.HoldPercentage, B=0) {is=title K2Node_CallFunction})
   n16 = Get HoldProgress()
   > PlayAnimationReverse(InAnimation=@n16, PlaybackSpeed=float / float(A=GetEndTime(self=@n16), B=Get HoldTime()))
   > StopFillSound() {is=title K2Node_CallFunction}
   Event BP_OnHoldCompleted() {is=title K2Node_Event}
   > StopAllAnimations(self:User Widget Object Reference)
   > PlayAnimation(InAnimation=Get HoldCompleted() -> (HoldCompleted:Widget Animation Object Reference))
   > FillComplete() {is=title K2Node_CallFunction}
   n17 = Event BP_OnHoldIncreased!() -> (OutputDelegate, then, HoldPercentage:float) {is=title K2Node_Event}
   > n18 = Get()
   > Set Float Parameter!(execute, self:Audio Parameter Controller Interface Interface=@n18, InName:Name=FillProgress, InFloat:float=@n17.HoldPercentage) -> (then) {is=title K2Node_CallFunction}
   n19 = Event BP_OnHoldDecreased!() -> (OutputDelegate, then, HoldPercentage:float) {is=title K2Node_Event}
   > n20 = Get()
   > Set Float Parameter(self=@n20, InName=FillProgress, InFloat=@n19.HoldPercentage)
   Event BP_OnHoldReset() {is=title K2Node_Event}
   > StopAllAnimations()
   > PlayAnimation(InAnimation=Get HoldCompleted())
   > n21 = Get()
   > Stop(self:Audio Component Object Reference=@n21)
   Event On Hovered()
   n22 = Get Hover() -> (Hover:Widget Animation Object Reference)
   > PlayAnimationForward(InAnimation=@n22)
   Event On Unhovered()
   > n23 = Branch(Condition=Get bHoldCancelledFromMovement() -> (bHoldCancelledFromMovement:bool))
   PlayAnimationReverse(execute=@n23.else, InAnimation=@n22)
   Event Destruct()
   n24 = Create Event()
   n25 = Get HoldProgress()
   > UnbindFromAnimationStarted!(execute, self:User Widget Object Reference, Animation:Widget Animation Object Reference=@n25, Delegate:Delegate=@n24) -> (then)
   Event Construct()
   > BindToAnimationStarted!(execute, self:User Widget Object Reference, Animation:Widget Animation Object Reference=@n25, Delegate:Delegate=@n24) -> (then)
   n26 = Get bIsHoldableButton() -> (bIsHoldableButton:bool)
   Branch(Condition=@n26)
   > SetVisibility(self=Get Image_Progress(), InVisibility=Select(Option 0=Collapsed, Option 1=SelfHitTestInvisible, Index=@n26))
   Event On Pressed() {is=title K2Node_Event}
   Reroute Node(InputPin:bool=@n8.IsDesignTime) -> (OutputPin:bool)
  ]}
 OverrideFontSlate {nodes [
   n1 = OverrideFontSlate!() -> (then, InText:Text Object Reference, inSize:int, InFontMaterial:Object Reference) {is=title K2Node_FunctionEntry}
   n2 = Break Slate Font Info!(SlateFontInfo:"Slate Font Info Structure (by ref)"=Get Font!(self:Text Object Reference=Get InText() -> (InText:Text Object Reference)) -> (Font:Slate Font Info Structure)) -> (
    FontObject:Object Reference
    FontMaterial:Object Reference
    OutlineSettings:Font Outline Settings Structure
    TypefaceFontName:Name
    Size:float=~
    LetterSpacing:int
    SkewAmount:float=~
    bForceMonospaced:bool=~
    bMaterialIsStencil:bool=~
    MonospacedWidth:float=~
   ) {is=title K2Node_BreakStruct}
   n3 = Get inSize() -> (inSize:int=0)
   > SetFont(
    self=@n1
    InFontInfo=MakeSlateFontInfo!(
     FontObject:Object Reference=@n2
     FontMaterial:Object Reference=Get InFontMaterial() -> (InFontMaterial:Object Reference)
     OutlineSettings:Font Outline Settings Structure=@n2.OutlineSettings
     TypefaceFontName:Name=@n2.TypefaceFontName
     Size:float="To Float (Integer)"(InInt=@n3)
     LetterSpacing:int=@n2.LetterSpacing
     SkewAmount:float=0
     bForceMonospaced:bool=false
     bMaterialIsStencil:bool=false
     MonospacedWidth:float=1
    ) -> (ReturnValue:Slate Font Info Structure)
   )
  ]}
 SetMaterials {nodes [
   SetMaterials() {is=title K2Node_FunctionEntry}
   > SetBrushFromMaterial(self=Get Image_Background(), Material=Get BackgroundMaterial() -> (BackgroundMaterial:Material Interface Object Reference))
   > SetBrushFromMaterial(self=Get Image_Progress(), Material=Get ProgressMaterial() -> (ProgressMaterial:Material Interface Object Reference))
  ]}
]
variables [
 BackgroundMaterial {type=MaterialInterface, default=/Game/Athena/HUD/Creative/StartScreen/MaterialInstances/MI_Creative_StartScreen_Button_Hold_Shape_Progress}
 FillCompleteSoundOverride {type=SoundBase, default=null}
 FillingSoundAudioComponent {type=AudioComponent, default=null}
 FillingSoundOverride {type=SoundBase, default=null}
 HoldProgressDMI {type=MaterialInstanceDynamic}
 In Height Override {type=number, default=75}
 In SubText {type=string, default=Subtext}
 In Text {type=string, default=button text}
 In Width Override {type=number, default=450}
 Justification {type=ETextJustify, default=Center}
 ProgressMaterial {type=MaterialInterface, default=/Game/Athena/HUD/Creative/StartScreen/MaterialInstances/MI_Creative_StartScreen_Button_Hold_Progress}
 Show Right icon {type=boolean}
 SubtitleFontSize {type=integer, default=14}
 SubtitleMaterial {type=Object, default=/Game/Athena/HUD/Creative/StartScreen/MaterialInstances/MI_Creative_StartScreen_Button_Hold_StateColors_Subtitle}
 TitleFontSize {type=integer, default=22}
 TitleMaterial {type=Object, default=/Game/Athena/HUD/Creative/StartScreen/MaterialInstances/MI_Creative_StartScreen_Button_Hold_StateColors_Title}
]
widgets {
 info {inheritedWidgetCount=1, parentClass=/Script/FortniteUI.FortHoldableButton, rootWidgetClass=&SizeBox, widgetCount=12}
 widgets [
  Root {widgetClassPath=&SizeBox, bIsVariable=true, widget=Root, properties {bOverride_HeightOverride=true, heightOverride=75}}
  Main {widgetClassPath=&Overlay, parent=Root, widget=Main, slot {class=SizeBoxSlot}}
  Image_Background {
   widgetClassPath=&Image
   bIsVariable=true
   parent=Main
   widget=Image_Background
   properties {brush {
     bIsDynamicallyLoaded=false
     drawAs=Image
     imageType=NoImage
     mirroring=NoMirror
     resourceName=None
     resourceObject=&MI_Creative_StartScreen_Button_Hold_Shape_Progress
     tiling=NoTile
     imageSize=<75 75>
     margin=<ltrb 0.495 0 0.495 0>
     outlineSettings {
      bUseBrushTransparency=false
      roundingType=HalfHeightRadius
      width=0
      color {colorUseRule=UseColor_Specified, specifiedColor=<rgba 0 0 0 0>}
      cornerRadii=<0 0 0 0>
     }
     tintColor {colorUseRule=UseColor_Specified, specifiedColor=<rgba 1 1 1 1>}
     uVRegion {bIsValid=false, max=<0 0>, min=<0 0>}
    }}
   slot {class=OverlaySlot, properties {horizontalAlignment=HAlign_Fill, verticalAlignment=VAlign_Fill}}
  }
  Image_Progress {
   widgetClassPath=&Image
   bIsVariable=true
   parent=Main
   widget=Image_Progress
   properties {brush {
     bIsDynamicallyLoaded=false
     drawAs=Image
     imageType=NoImage
     mirroring=NoMirror
     resourceName=None
     resourceObject=&MI_Creative_StartScreen_Button_Hold_Progress
     tiling=NoTile
     imageSize=<75 75>
     margin=<ltrb 0 0 0 0>
     outlineSettings {
      bUseBrushTransparency=false
      roundingType=HalfHeightRadius
      width=0
      color {colorUseRule=UseColor_Specified, specifiedColor=<rgba 0 0 0 0>}
      cornerRadii=<0 0 0 0>
     }
     tintColor {colorUseRule=UseColor_Specified, specifiedColor=<rgba 1 1 1 1>}
     uVRegion {bIsValid=false, max=<0 0>, min=<0 0>}
    }}
   slot {class=OverlaySlot, properties {horizontalAlignment=HAlign_Fill, verticalAlignment=VAlign_Fill}}
  }
  InputActionBox {
   widgetClassPath=&SizeBox
   parent=Main
   widget=InputActionBox
   properties {bOverride_HeightOverride=true, bOverride_WidthOverride=true, heightOverride=40, widthOverride=40}
   slot {class=OverlaySlot, properties {verticalAlignment=VAlign_Center, padding=<ltrb 10 0 10 0>}}
  }
  InputActionWidget {
   widgetClassPath=&CommonActionWidget
   bInherited=true
   bIsVariable=true
   parent=InputActionBox
   widget=InputActionWidget
   properties {
    iconRimBrush {
     bIsDynamicallyLoaded=false
     drawAs=Image
     imageType=NoImage
     mirroring=NoMirror
     resourceName=None
     resourceObject=/FortUILibrary/Foundation/Buttons/Materials/M-Button-Console-WhiteOutline
     tiling=NoTile
     imageSize=<32 32>
     margin=<ltrb 0 0 0 0>
     outlineSettings {
      bUseBrushTransparency=false
      roundingType=HalfHeightRadius
      width=0
      color {colorUseRule=UseColor_Specified, specifiedColor=<rgba 0 0 0 0>}
      cornerRadii=<0 0 0 0>
     }
     tintColor {colorUseRule=UseColor_Specified, specifiedColor=<rgba 1 1 1 1>}
     uVRegion {bIsValid=false, max=<0 0>, min=<0 0>}
    }
    progressMaterialBrush {
     bIsDynamicallyLoaded=false
     drawAs=Image
     imageType=NoImage
     mirroring=NoMirror
     resourceName=None
     resourceObject=null
     tiling=NoTile
     imageSize=<32 32>
     margin=<ltrb 0 0 0 0>
     outlineSettings {
      bUseBrushTransparency=false
      roundingType=HalfHeightRadius
      width=0
      color {colorUseRule=UseColor_Specified, specifiedColor=<rgba 0 0 0 0>}
      cornerRadii=<0 0 0 0>
     }
     tintColor {colorUseRule=UseColor_Specified, specifiedColor=<rgba 1 1 1 0>}
     uVRegion {bIsValid=false, max=<0 0>, min=<0 0>}
    }
   }
   slot {class=SizeBoxSlot}
  }
  Content {
   widgetClassPath=&HorizontalBox
   parent=Main
   widget=Content
   slot {class=OverlaySlot, properties {horizontalAlignment=HAlign_Fill, verticalAlignment=VAlign_Center, padding=<ltrb 50 0 50 0>}}
  }
  VerticalBox_0 {
   widgetClassPath=&VerticalBox
   parent=Content
   widget=VerticalBox_0
   slot {class=HorizontalBoxSlot, properties {verticalAlignment=VAlign_Center, padding=<ltrb 6 0 6 0>, size {sizeRule=Fill, value=1}}}
  }
  Text_Title {
   widgetClassPath=&CommonTextBlock
   bIsVariable=true
   parent=VerticalBox_0
   widget=Text_Title
   properties {
    justification=Center
    scrollStyle=&TextScrollStyle-Base_C
    text=Text Block
    textTransformPolicy=ToUpper
    colorAndOpacity {colorUseRule=UseColor_Specified, specifiedColor=<rgba 0 0 0 1>}
    font {
     bForceMonospaced=false
     bMaterialIsStencil=false
     fontMaterial=&MI_Creative_StartScreen_Button_Hold_StateColors_Title
     fontObject=&Burbank
     letterSpacing=0
     monospacedWidth=1
     size=22
     skewAmount=0
     typefaceFontName=BurbankRegular-Black
     outlineSettings {
      bApplyOutlineToDropShadows=false
      bMiteredCorners=false
      bSeparateFillAlpha=false
      outlineMaterial=null
      outlineSize=0
      outlineColor=<rgba 0 0 0 1>
     }
    }
    renderTransform {angle=0, scale=<1 1>, shear=<-8 0>, translation=<0 0>}
    shadowOffset=<0 0>
   }
   slot {class=VerticalBoxSlot}
  }
  Text_subtitle {
   widgetClassPath=&CommonTextBlock
   bIsVariable=true
   parent=VerticalBox_0
   widget=Text_subtitle
   properties {
    bAutoCollapseWithEmptyText=true
    justification=Center
    lineHeightPercentage=0.85
    scrollStyle=&TextScrollStyle-Base_C
    text=Text Block
    textTransformPolicy=ToUpper
    font {
     bForceMonospaced=false
     bMaterialIsStencil=false
     fontMaterial=&MI_Creative_StartScreen_Button_Hold_StateColors_Subtitle
     fontObject=&BurbankSmall
     letterSpacing=0
     monospacedWidth=1
     size=14
     skewAmount=0
     typefaceFontName=Black
     outlineSettings {
      bApplyOutlineToDropShadows=false
      bMiteredCorners=false
      bSeparateFillAlpha=false
      outlineMaterial=null
      outlineSize=0
      outlineColor=<rgba 0 0 0 1>
     }
    }
    renderTransform {angle=0, scale=<1 1>, shear=<-8 0>, translation=<0 0>}
    shadowOffset=<0 0>
   }
   slot {class=VerticalBoxSlot, properties {verticalAlignment=VAlign_Center, padding=<ltrb 0 -4 0 0>}}
  }
  Spacer_Icon {
   widgetClassPath=&Spacer
   bIsVariable=true
   parent=Content
   widget=Spacer_Icon
   properties {size=<35 1>}
   slot {class=HorizontalBoxSlot}
  }
  Image_Icon {
   widgetClassPath=&Image
   bIsVariable=true
   parent=Main
   widget=Image_Icon
   properties {
    visibility=Collapsed
    brush {
     bIsDynamicallyLoaded=false
     drawAs=Image
     imageType=NoImage
     mirroring=NoMirror
     resourceName=None
     resourceObject=null
     tiling=NoTile
     imageSize=<64 64>
     margin=<ltrb 0 0 0 0>
     outlineSettings {
      bUseBrushTransparency=false
      roundingType=HalfHeightRadius
      width=0
      color {colorUseRule=UseColor_Specified, specifiedColor=<rgba 0 0 0 0>}
      cornerRadii=<0 0 0 0>
     }
     tintColor {colorUseRule=UseColor_Specified, specifiedColor=<rgba 1 1 1 1>}
     uVRegion {bIsValid=false, max=<0 0>, min=<0 0>}
    }
    renderTransform {angle=0, scale=<1.2 1.2>, shear=<0 0>, translation=<-6 1>}
    renderTransformPivot=<1 1>
   }
   slot {class=OverlaySlot, properties {horizontalAlignment=HAlign_Right, verticalAlignment=VAlign_Bottom, padding=<ltrb 20 0 0 0>}}
  }
 ]
}
