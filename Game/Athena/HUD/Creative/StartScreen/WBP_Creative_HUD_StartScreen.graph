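# Creative island start screen HUD widget (FortCreativeStartScreen) showing minigame info and joining status alongside hold-to-Start, Support a Creator and More Info buttons.
# Construct adjusts padding for touch and mobile, then sets Support a Creator visibility from runtime options, banned creator names and playlist, disabling it in competitive.
# Clicking that button async-creates a CreativeAffiliatePanel prefilled with the support code.
# IntroAnim plays once on activation; BoundAnim_StartGame slides and fades content out.
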WidgetBlueprint /Game/Athena/HUD/Creative/StartScreen/WBP_Creative_HUD_StartScreen
animations [
 BoundAnim_StartGame {
  binding_count=4
  length_seconds=0.35
  object=&self:BoundAnim_StartGame
  sequence {
   bindings [
    WBP_Creative_Minigame_Joining {
     id=#1
     possessed_class=WBP_Creative_Minigame_Joining_C
     tracks [
      {
       type=MovieSceneFloatTrack
       display_name=Render Opacity
       sections [{type=MovieSceneFloatSection, end_frame=10, start_frame=0, channels [None {type=MovieSceneScriptingFloatChannel, keys [[3,1],[10.5005,0]]}]}]
      }
      {
       type=MovieScene2DTransformTrack
       sections [{
         type=MovieScene2DTransformSection
         channels [
          Translation.X {type=MovieSceneScriptingFloatChannel, const [0,2]}
          Translation.Y {type=MovieSceneScriptingFloatChannel, keys [[0,0],[1,-10],[3,0],[10.5005,200]]}
          Scale.X {type=MovieSceneScriptingFloatChannel}
          Scale.Y {type=MovieSceneScriptingFloatChannel}
         ]
        }]
      }
     ]
    }
    Image_BG_BlackGradient {
     id=#2
     possessed_class=Image
     tracks [{
       type=MovieSceneFloatTrack
       display_name=Render Opacity
       sections [{type=MovieSceneFloatSection, channels [None {type=MovieSceneScriptingFloatChannel, keys [[6,1],[10.5,0]]}]}]
      }]
    }
    SizeBox_0 {
     id=#3
     possessed_class=SizeBox
     tracks [
      {
       type=MovieSceneFloatTrack
       display_name=Render Opacity
       sections [{type=MovieSceneFloatSection, channels [None {type=MovieSceneScriptingFloatChannel, keys [[3,1],[10.5005,0]]}]}]
      }
      {
       type=MovieScene2DTransformTrack
       sections [{
         type=MovieScene2DTransformSection
         channels [
          Translation.X {type=MovieSceneScriptingFloatChannel}
          Translation.Y {type=MovieSceneScriptingFloatChannel, keys [[3,0],[10.5005,200]]}
          Scale.X {type=MovieSceneScriptingFloatChannel}
          Scale.Y {type=MovieSceneScriptingFloatChannel}
         ]
        }]
      }
     ]
    }
    MinigameInfo {
     id=#4
     possessed_class=WBP_Creative_Minigame_Info_C
     tracks [
      {
       type=MovieSceneFloatTrack
       display_name=Render Opacity
       sections [{type=MovieSceneFloatSection, channels [None {type=MovieSceneScriptingFloatChannel, keys [[0,1],[1,1],[4.5,0]]}]}]
      }
      {
       type=MovieScene2DTransformTrack
       sections [{
         type=MovieScene2DTransformSection
         channels [
          Translation.X {type=MovieSceneScriptingFloatChannel, keys [[0,0],[1,20],[2,20],[10.0005,-360]]}
          Translation.Y {type=MovieSceneScriptingFloatChannel, const [0,1]}
          Scale.X {type=MovieSceneScriptingFloatChannel, const [1,1]}
          Scale.Y {type=MovieSceneScriptingFloatChannel, const [1,1]}
         ]
        }]
      }
     ]
    }
   ]
   display_rate {numerator=30}
  }
 }
 IntroAnim {
  binding_count=3
  length_seconds=0.35
  object=&self:IntroAnim
  sequence {
   bindings [
    WBP_Creative_Minigame_Joining {
     id=#5
     possessed_class=WBP_Creative_Minigame_Joining_C
     tracks [
      {
       type=MovieSceneFloatTrack
       display_name=Render Opacity
       sections [{type=MovieSceneFloatSection, channels [None {type=MovieSceneScriptingFloatChannel, keys [[20,0],[35.0017,1]]}]}]
      }
      {
       type=MovieScene2DTransformTrack
       sections [{
         type=MovieScene2DTransformSection
         channels [
          Translation.X {type=MovieSceneScriptingFloatChannel, const [0,2]}
          Translation.Y {type=MovieSceneScriptingFloatChannel, keys [[14,200],[28.285,0]]}
          Scale.X {type=MovieSceneScriptingFloatChannel}
          Scale.Y {type=MovieSceneScriptingFloatChannel}
         ]
        }]
      }
     ]
    }
    MinigameInfo {
     id=#6
     possessed_class=WBP_Creative_Minigame_Info_C
     tracks [
      {
       type=MovieSceneFloatTrack
       display_name=Render Opacity
       sections [{type=MovieSceneFloatSection, channels [None {type=MovieSceneScriptingFloatChannel, keys [[0,0],[15.6967,1]]}]}]
      }
      {
       type=MovieScene2DTransformTrack
       sections [{
         type=MovieScene2DTransformSection
         channels [
          Translation.X {type=MovieSceneScriptingFloatChannel, keys [[0,-360],[15.6967,0]]}
          Translation.Y {type=MovieSceneScriptingFloatChannel, const [0,1]}
          Angle {type=MovieSceneScriptingFloatChannel, const [0,1]}
          Scale.X {type=MovieSceneScriptingFloatChannel, const [1,1]}
          Scale.Y {type=MovieSceneScriptingFloatChannel, const [1,1]}
          Shear.X {type=MovieSceneScriptingFloatChannel, const [0,1]}
          Shear.Y {type=MovieSceneScriptingFloatChannel, const [0,1]}
         ]
        }]
      }
     ]
    }
    SizeBox_0 {
     id=#7
     possessed_class=SizeBox
     tracks [
      {
       type=MovieSceneFloatTrack
       display_name=Render Opacity
       sections [{type=MovieSceneFloatSection, channels [None {type=MovieSceneScriptingFloatChannel, keys [[20,0],[35.0017,1]]}]}]
      }
      {
       type=MovieScene2DTransformTrack
       sections [{
         type=MovieScene2DTransformSection
         channels [
          Translation.X {type=MovieSceneScriptingFloatChannel, const [0,2]}
          Translation.Y {type=MovieSceneScriptingFloatChannel, keys [[14,200],[28.285,0]]}
          Scale.X {type=MovieSceneScriptingFloatChannel}
          Scale.Y {type=MovieSceneScriptingFloatChannel}
         ]
        }]
      }
     ]
    }
   ]
   display_rate {numerator=100}
  }
 }
]
graphs [EventGraph {nodes [
   "On Button Clicked (Button_SupportACreator)"!() -> (OutputDelegate, then, Button:Common Button Legacy Object Reference) {is=title K2Node_ComponentBoundEvent}
   n1 = GetCreatorInformation!(self:Fort Creative Mini Game Info Object Reference=Get MinigameInfo() -> (MinigameInfo:Fort Creative Mini Game Info Object Reference)) -> (OutCreatorName:String, OutSupportCode:String)
   n2 = Get BannedNames() -> (BannedNames:Array of Strings)
   n3 = Get FortHUDContext!() -> (ReturnValue:Fort HUDContext Object Reference) {is=title K2Node_GetBlueprintContextEx}
   > n4 = Branch(Condition=OR Boolean!(
     A=IsInCursorMode!(self:Fort HUDContext Object Reference=@n3) -> (ReturnValue:bool)
     B=Contains Item(TargetArray:Array of Strings=@n2, ItemToFind:"String (by ref)"=@n1)
     C:bool=IsCreativeRoyalePlaylist!() -> (ReturnValue:bool)
    ) -> (ReturnValue))
   n5 = CreateWidgetAsync!(
    execute:Exec=@n4.else
    UserWidgetSoftClass:User Widget Soft Class Reference=/Game/Athena/UI/InGame/Creative/CreativeAffiliatePanel.CreativeAffiliatePanel_C
    OwningPlayer:Player Controller Object Reference
    bSuspendInputUntilComplete:bool=true
   ) -> (then:Exec, OnComplete:Exec, OnFailure:Exec, UserWidget:User Widget Object Reference) {is=title K2Node_AsyncAction}
   n6 = Cast To CreativeAffiliatePanel!(execute=@n5.OnComplete, Object=@n5) -> (then, CastFailed, AsCreative Affiliate Panel:Creative Affiliate Panel Object Reference) {is=title K2Node_DynamicCast}
   > n7 = Set AffiliatePanel(AffiliatePanel:Creative Affiliate Panel Object Reference=@n6) -> (Output_Get:Creative Affiliate Panel Object Reference)
   > InitFromObject!(execute, self:Creative Affiliate Panel Object Reference=@n7, InitObject:Object Reference=Self-Reference()) -> (then)
   > SetAffiliateName!(
    execute
    self:Creative Affiliate Panel Object Reference=@n7
    AffiliateName:Text="To Text (String)"(InString=@n1.OutSupportCode)
    Can Edit:bool=true
   ) -> (then)
   > SetSourceName!(execute, self:Creative Affiliate Panel Object Reference=@n7, NewSourceName:String=Get Source Name() -> (Source Name:String)) -> (then)
   Event On Activated()
   > Do Once()
   > PlayAnimation(InAnimation=Get IntroAnim() -> (IntroAnim:Widget Animation Object Reference))
   Event Construct()
   > n8 = PlaylistDataReadyAsync!(execute:Exec) -> (
    then:Exec
    OnReady:Exec
    GameState:Fort Game State Athena Object Reference
    Playlist:Fort Playlist Object Reference
    PlaylistContextTags:"Gameplay Tag Container Structure (by ref)"
   ) {is=title K2Node_AsyncAction}
   > n9 = Branch(Condition=IsUsingTouch(PlayerController=GetOwningPlayer()))
   > n10 = SetPadding!(
    execute
    self:Overlay Slot Object Reference=SlotAsOverlaySlot(Widget=Get Main() -> (Main:Horizontal Box Object Reference))
    InPadding_Left:float=300
    InPadding_Top:float=0
    InPadding_Right:float=300
    InPadding_Bottom:float=75
   ) -> (then)
   n11 = Branch(execute=[@n12.then, @n10.then], Condition=IsMobileGame())
   > SetPadding(
    self=SlotAsOverlaySlot(Widget=Get Image_BG_BlackGradient() -> (Image_BG_BlackGradient:Image Object Reference))
    InPadding_Left=-250
    InPadding_Top=-60
    InPadding_Right=-250
    InPadding_Bottom=-120
   )
   SetPadding(
    execute=@n11.else
    self=SlotAsOverlaySlot(Widget=Get Image_BG_BlackGradient())
    InPadding_Left=0
    InPadding_Top=-60
    InPadding_Right=0
    InPadding_Bottom=-70
   )
   n12 = SetPadding(execute=@n9.else, self=SlotAsOverlaySlot(Widget=Get Main()), InPadding_Left=56, InPadding_Top=56, InPadding_Right=56, InPadding_Bottom=56)
   n13 = Contains Item(TargetArray=Get BannedNames(), ItemToFind=GetCreatorInformation(self=Get MinigameInfo()))
   n14 = IsCreativeRoyalePlaylist()
   n15 = GetRuntimeOptions()
   n16 = Get bDisableAffiliateFeature!(self:Fort Runtime Options Object Reference=@n15) -> (bDisableAffiliateFeature:bool)
   n17 = Get bDisableCreativeSACFeature!(self:Fort Runtime Options Object Reference=@n15) -> (bDisableCreativeSACFeature:bool)
   SetVisibility(
    execute=@n8.OnReady
    self=Get Button_SupportACreator() -> (Button_SupportACreator:Icon Text Button Object Reference)
    InVisibility=Select(Option 0=Visible, Option 1=Collapsed, Index=OR Boolean!(A=@n16, B=@n17, C=@n13, D:bool=@n14) -> (ReturnValue))
   )
   > Branch(Condition=IsCompetitivePlaylist!() -> (ReturnValue:bool))
   > SetIsEnabled(self=Get Button_SupportACreator(), bInIsEnabled=false)
  ]}]
variables [
 AffiliatePanel {type=CreativeAffiliatePanel_C, default=null}
 BannedNames {type="array<string>", default [Epic Games, EpicGames]}
 ProgressBarSlateBrushCopy {
  type=SlateBrush
  default {
   bIsDynamicallyLoaded=false
   drawAs=Image
   imageType=NoImage
   mirroring=NoMirror
   resourceName=None
   resourceObject=null
   tiling=NoTile
   imageSize=<32 32>
   margin=<ltrb 0 0 0 0>
   outlineSettings {
    bUseBrushTransparency=false
    roundingType=HalfHeightRadius
    width=0
    color {colorUseRule=UseColor_Specified, specifiedColor=<rgba 0 0 0 0>}
    cornerRadii=<0 0 0 1>
   }
   tintColor {colorUseRule=UseColor_Specified, specifiedColor=<rgba 1 1 1 1>}
   uVRegion {bIsValid=false, max=<0 0>, min=<0 0>}
  }
 }
 Source Name {type=string, default=CreativePregameLobby}
]
widgets {
 info {inheritedWidgetCount=3, parentClass=/Script/FortniteUI.FortCreativeStartScreen, rootWidgetClass=&Overlay, widgetCount=13}
 widgets [
  Root {widgetClassPath=&Overlay, widget=Root}
  Overlay_0 {
   widgetClassPath=&Overlay
   parent=Root
   widget=Overlay_0
   slot {class=OverlaySlot, properties {horizontalAlignment=HAlign_Fill, verticalAlignment=VAlign_Bottom}}
  }
  Image_BG_BlackGradient {
   widgetClassPath=&Image
   bIsVariable=true
   parent=Overlay_0
   widget=Image_BG_BlackGradient
   properties {
    visibility=SelfHitTestInvisible
    brush {
     bIsDynamicallyLoaded=false
     drawAs=Image
     imageType=NoImage
     mirroring=NoMirror
     resourceName=None
     resourceObject=/Game/Athena/HUD/Creative/MaterialInstances/MI_UI_LinearGradient_Creative_StartScreen
     tiling=NoTile
     imageSize=<32 32>
     margin=<ltrb 0 0 0 0>
     outlineSettings {
      bUseBrushTransparency=false
      roundingType=HalfHeightRadius
      width=0
      color {colorUseRule=UseColor_Specified, specifiedColor=<rgba 0 0 0 0>}
      cornerRadii=<0 0 0 0>
     }
     tintColor {colorUseRule=UseColor_Specified, specifiedColor=<rgba 1 1 1 1>}
     uVRegion {bIsValid=false, max=<0 0>, min=<0 0>}
    }
   }
   slot {class=OverlaySlot, properties {horizontalAlignment=HAlign_Fill, verticalAlignment=VAlign_Fill, padding=<ltrb 0 -60 0 -20>}}
  }
  Main {
   widgetClassPath=&HorizontalBox
   bIsVariable=true
   parent=Overlay_0
   widget=Main
   slot {class=OverlaySlot, properties {horizontalAlignment=HAlign_Fill, verticalAlignment=VAlign_Fill, padding=<ltrb 56 56 56 56>}}
  }
  Content {
   widgetClassPath=&Overlay
   bIsVariable=true
   parent=Main
   widget=Content
   slot {class=HorizontalBoxSlot, properties {size {sizeRule=Fill, value=1}}}
  }
  Info {
   widgetClassPath=&VerticalBox
   parent=Content
   widget=Info
   slot {class=OverlaySlot, properties {horizontalAlignment=HAlign_Fill, verticalAlignment=VAlign_Bottom}}
  }
  MinigameInfo {
   widgetClassPath=&/WBP_Creative_Minigame_Info
   bInherited=true
   parent=Info
   widget=MinigameInfo
   properties {bDisplayTags=false, specificIslandCodeColor=<rgba 1 1 1 0.6>, textColor=<rgba 1 1 1 0.6>}
   slot {class=VerticalBoxSlot}
  }
  WBP_Creative_Minigame_Joining {
   widgetClassPath=&/WBP_Creative_Minigame_Joining
   bIsVariable=true
   parent=Info
   widget=WBP_Creative_Minigame_Joining
   properties {playerInfoTextColor=<rgba 1 1 1 0.6>}
   slot {class=VerticalBoxSlot, properties {horizontalAlignment=HAlign_Left, verticalAlignment=VAlign_Bottom, padding=<ltrb -9 0 0 -10>}}
  }
  SizeBox_0 {
   widgetClassPath=&SizeBox
   parent=Main
   widget=SizeBox_0
   properties {bOverride_MinDesiredWidth=true, minDesiredWidth=312}
   slot {class=HorizontalBoxSlot, properties {horizontalAlignment=HAlign_Right, verticalAlignment=VAlign_Bottom}}
  }
  Buttons {widgetClassPath=&VerticalBox, parent=SizeBox_0, widget=Buttons, slot {class=SizeBoxSlot}}
  HoldableButton_Start {
   widgetClassPath=/Game/Athena/HUD/Creative/Widgets/WBP_Creative_Button_Hold
   bInherited=true
   bIsVariable=true
   parent=Buttons
   widget=HoldableButton_Start
   properties {
    bBindHoldToAction=true
    bIsFocusable=true
    bIsHoldableButton=true
    bKeybindingTriggersHoldEvents=true
    holdTime=0.5
    in Height Override=72
    in SubText=(Hold)
    in Text=Start
    in Width Override=312
    inputModeOverride=Game
    qualifyingHoldTime=0
    startingHoldPercentage=0
    visibility=SelfHitTestInvisible
    triggeringInputAction {dataTable=/FortUILibrary/Input/CreativeInputActionDataTable, rowName=IslandStartGame}
   }
   slot {class=VerticalBoxSlot}
  }
  Button_SupportACreator {
   widgetClassPath=&IconTextButton
   bIsVariable=true
   parent=Buttons
   widget=Button_SupportACreator
   properties {
    buttonText=Support a Creator
    inputModeOverride=Game
    onButtonBaseClicked=(&self:WidgetTree.Button_SupportACreator.HandleOnButtonClicked)
    onButtonBaseDoubleClicked=(&self:WidgetTree.Button_SupportACreator.HandleOnButtonDoubleClicked)
    onButtonBaseHovered=(&self:WidgetTree.Button_SupportACreator.HandleOnButtonHovered)
    onButtonBaseUnhovered=(&self:WidgetTree.Button_SupportACreator.HandleOnButtonUnhovered)
    onSelectedChangedBase=(&self:WidgetTree.Button_SupportACreator.HandleOnSelectedChanged)
    style=/FortUILibrary/Foundation/Buttons/ButtonStyles/Athena/Athena_ButtonStyle_AngledBlueMenuButton_Small.Athena_ButtonStyle_AngledBlueMenuButton_Small_C
    visibility=Collapsed
    triggeringInputAction {dataTable=&CreativeInputActionDataTable, rowName=IslandSupportCreator}
   }
   slot {class=VerticalBoxSlot, properties {padding=<ltrb 0 15 0 0>}}
  }
  Button_MoreInfo {
   widgetClassPath=&IconTextButton
   bInherited=true
   bIsVariable=true
   parent=Buttons
   widget=Button_MoreInfo
   properties {
    buttonText=More Info
    inputModeOverride=Game
    onButtonBaseClicked=(&self:WidgetTree.Button_MoreInfo.HandleOnButtonClicked)
    onButtonBaseDoubleClicked=(&self:WidgetTree.Button_MoreInfo.HandleOnButtonDoubleClicked)
    onButtonBaseHovered=(&self:WidgetTree.Button_MoreInfo.HandleOnButtonHovered)
    onButtonBaseUnhovered=(&self:WidgetTree.Button_MoreInfo.HandleOnButtonUnhovered)
    onSelectedChangedBase=(&self:WidgetTree.Button_MoreInfo.HandleOnSelectedChanged)
    style=&Athena_ButtonStyle_AngledBlueMenuButton_Small_C
    triggeringInputAction {dataTable=&CreativeInputActionDataTable, rowName=ToggleMinigameDetails}
   }
   slot {class=VerticalBoxSlot, properties {padding=<ltrb 0 15 0 0>}}
  }
 ]
}
