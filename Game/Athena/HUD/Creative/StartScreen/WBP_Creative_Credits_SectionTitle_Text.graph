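# Widget Blueprint for a section title in the Creative start screen credits, based on the native FortCreativeCreditsSectionTitle class.
# Its only widget is the root CommonTextBlock Text_Title, inherited from the parent, styled in Burbank Small Black with placeholder text.
# The event graph is empty and no variables are declared, so the Blueprint supplies only visual styling.
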WidgetBlueprint /Game/Athena/HUD/Creative/StartScreen/WBP_Creative_Credits_SectionTitle_Text
graphs [EventGraph {links [], nodes []}]
variables []
widgets {
 info {
  inheritedWidgetCount=1
  parentClass=/Script/FortniteUI.FortCreativeCreditsSectionTitle
  rootWidgetClass=&CommonTextBlock
  widgetCount=1
 }
 widgets [Text_Title {
   widgetClassPath=&CommonTextBlock
   bInherited=true
   widget=Text_Title
   properties {
    text=Section Title
    font {
     bForceMonospaced=false
     bMaterialIsStencil=false
     fontMaterial=null
     fontObject=&BurbankSmall
     letterSpacing=0
     monospacedWidth=1
     size=24
     skewAmount=0
     typefaceFontName=Black
     outlineSettings {
      bApplyOutlineToDropShadows=false
      bMiteredCorners=false
      bSeparateFillAlpha=false
      outlineMaterial=null
      outlineSize=0
      outlineColor=<rgba 0 0 0 1>
     }
    }
    shadowOffset=<0 0>
   }
  }]
}
