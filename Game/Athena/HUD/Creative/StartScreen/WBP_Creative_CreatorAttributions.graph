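# Widget Blueprint for the Creative start screen's creator attributions list, built on FortCreativeCreatorAttributions.
# A scroll box with animated focus scrolling holds an inherited vertical dynamic entry box, whose entry class is the credits section title text widget.
# Function graphs named for the creation of attributions, credits and title card buttons, section titles and team member cards each apply a matching padding variable to that widget.
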WidgetBlueprint /Game/Athena/HUD/Creative/StartScreen/WBP_Creative_CreatorAttributions
graphs [
 EventGraph {links [], nodes []}
 OnAttributionsButtonCreated {nodes [
   OnAttributionsButtonCreated() {is=title K2Node_FunctionEntry}
   > SetPadding!(
    execute
    self:User Widget Object Reference=Get Button_CreditsAttributions() -> (Button_CreditsAttributions:Fort Creative Credits Attributions Object Reference)
    InPadding:Margin Structure=Get AttributionsButtonPadding() -> (AttributionsButtonPadding:Margin Structure)
   ) -> (then)
  ]}
 OnCreditsButtonCreated {nodes [
   OnCreditsButtonCreated() {is=title K2Node_FunctionEntry}
   > SetPadding(
    self=Get Button_Credits() -> (Button_Credits:Fort Creative Credits Attributions Object Reference)
    InPadding=Get CreditsButtonPadding() -> (CreditsButtonPadding:Margin Structure)
   )
  ]}
 OnSectionTitleCreated {nodes [
   n1 = OnSectionTitleCreated!() -> (then, SectionTitle:Fort Creative Credits Section Title Object Reference) {is=title K2Node_FunctionEntry}
   > SetPadding(self=@n1, InPadding=Get SectionTitlePadding() -> (SectionTitlePadding:Margin Structure))
  ]}
 OnTeamMemberTitleCardButtonCreated {nodes [
   n1 = OnTeamMemberTitleCardButtonCreated!() -> (then, TeamMemberButton:Fort Creative Credits Team Title Card Object Reference) {is=title K2Node_FunctionEntry}
   > SetPadding(self=@n1, InPadding=Get TeamMemberButtonPadding() -> (TeamMemberButtonPadding:Margin Structure))
  ]}
 OnTitleCardButtonCreated {nodes [
   OnTitleCardButtonCreated() {is=title K2Node_FunctionEntry}
   > SetPadding(
    self=Get Button_CreditsTeamTitleCard() -> (Button_CreditsTeamTitleCard:Fort Creative Credits Team Title Card Object Reference)
    InPadding=Get TitleCardButtonPadding() -> (TitleCardButtonPadding:Margin Structure)
   )
  ]}
]
variables []
widgets {
 info {inheritedWidgetCount=1, parentClass=/Script/FortniteUI.FortCreativeCreatorAttributions, rootWidgetClass=&ScrollBox, widgetCount=2}
 widgets [
  Widget_CreditsContent {
   widgetClassPath=&ScrollBox
   widget=Widget_CreditsContent
   properties {navigationDestination=Center, scrollWhenFocusChanges=AnimatedScroll}
  }
  DynamicEntryBox_CreditsContainer {
   widgetClassPath=/Script/FortniteUI.FortDynamicEntryBox
   bInherited=true
   bIsVariable=true
   parent=Widget_CreditsContent
   widget=DynamicEntryBox_CreditsContainer
   properties {
    entryBoxType=Vertical
    entryWidgetClass=&/WBP_Creative_Credits_SectionTitle_Text.WBP_Creative_Credits_SectionTitle_Text_C
    numDesignerPreviewEntries=1
   }
   slot {class=ScrollBoxSlot}
  }
 ]
}
