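# List entry widget for an island description line in the Creative start screen, built on FortIslandDescriptionListEntry.
# An overlay stacks a material box background behind the inherited, auto-wrapping Text_IslandDescriptionLine, which PreConstruct fills from the ButtonText variable.
# The short Hover animation drives the background material's Select parameter while darkening the text color from light blue.
# Unhover is an empty animation, and this graph plays neither animation.
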WidgetBlueprint /Game/Athena/HUD/Creative/StartScreen/WBP_Creative_Island_Description_ListEntry
animations [
 Hover {
  binding_count=2
  length_seconds=0.15
  object=&self:Hover
  sequence {bindings [
    Image_145 {
     id=#1
     possessed_class=Image
     tracks [{
       type=MovieSceneWidgetMaterialTrack
       sections [{type=MovieSceneParameterSection, channels [Select {type=MovieSceneScriptingFloatChannel, keys [[0,0],[2.9997,1]]}]}]
      }]
    }
    Text_IslandDescriptionLine {
     id=#2
     possessed_class=CommonTextBlock
     tracks [{
       type=MovieSceneColorTrack
       display_name=Color and Opacity
       sections [{
         type=MovieSceneColorSection
         channels [
          Color.R {type=MovieSceneScriptingFloatChannel, keys [[0,0.5271],[2.9997,0]]}
          Color.G {type=MovieSceneScriptingFloatChannel, keys [[0,0.8388],[2.9997,0.004]]}
          Color.B {type=MovieSceneScriptingFloatChannel, keys [[0,1],[2.9997,0.0723]]}
          Color.A {type=MovieSceneScriptingFloatChannel, const [1,2]}
         ]
        }]
      }]
    }
   ]}
 }
 Unhover {
  binding_count=0
  length_seconds=5
  object=&self:Unhover
  sequence {bindings []}
 }
]
graphs [EventGraph {nodes [
   Event PreConstruct()
   > "SetText (Text)"(
    self=Get Text_IslandDescriptionLine() -> (Text_IslandDescriptionLine:Common Text Object Reference)
    InText=Get ButtonText() -> (ButtonText:Text)
   )
   Event Construct()
   Event Tick()
  ]}]
variables [ButtonText {type=string, default="{Text}"}]
widgets {
 info {inheritedWidgetCount=1, parentClass=/Script/FortniteUI.FortIslandDescriptionListEntry, rootWidgetClass=&Overlay, widgetCount=3}
 widgets [
  Overlay_0 {widgetClassPath=&Overlay, widget=Overlay_0}
  Image_145 {
   widgetClassPath=&Image
   parent=Overlay_0
   widget=Image_145
   properties {brush {
     bIsDynamicallyLoaded=false
     drawAs=Box
     imageType=NoImage
     mirroring=NoMirror
     resourceName=None
     resourceObject=/Game/Athena/HUD/MiniMap/Art/M_UI_Overview_Container
     tiling=NoTile
     imageSize=<64 64>
     margin=<ltrb 0.5 0.5 0.5 0.5>
     outlineSettings {
      bUseBrushTransparency=false
      roundingType=HalfHeightRadius
      width=0
      color {colorUseRule=UseColor_Specified, specifiedColor=<rgba 0 0 0 0>}
      cornerRadii=<0 0 0 0>
     }
     tintColor {colorUseRule=UseColor_Specified, specifiedColor=<rgba 1 1 1 1>}
     uVRegion {bIsValid=false, max=<0 0>, min=<0 0>}
    }}
   slot {class=OverlaySlot, properties {horizontalAlignment=HAlign_Fill, verticalAlignment=VAlign_Fill}}
  }
  Text_IslandDescriptionLine {
   widgetClassPath=&CommonTextBlock
   bInherited=true
   bIsVariable=true
   parent=Overlay_0
   widget=Text_IslandDescriptionLine
   properties {
    autoWrapText=true
    text=Text goes here
    colorAndOpacity {colorUseRule=UseColor_Specified, specifiedColor=<rgba 0.5271 0.8388 1 1>}
    font {
     bForceMonospaced=false
     bMaterialIsStencil=false
     fontMaterial=null
     fontObject=&BurbankSmall
     letterSpacing=0
     monospacedWidth=1
     size=18
     skewAmount=0
     typefaceFontName=Bold
     outlineSettings {
      bApplyOutlineToDropShadows=false
      bMiteredCorners=false
      bSeparateFillAlpha=false
      outlineMaterial=null
      outlineSize=0
      outlineColor=<rgba 0 0 0 1>
     }
    }
    shadowOffset=<0 0>
   }
   slot {class=OverlaySlot, properties {horizontalAlignment=HAlign_Fill, verticalAlignment=VAlign_Center, padding=<ltrb 48 32 48 32>}}
  }
 ]
}
