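# Island description widget for the Creative start screen, built on the native FortIslandDescriptionWidget.
# A visibility switcher holds the inherited description-lines list and a standalone list-entry widget that shows a "no description written" fallback message.
# OnDescriptionChanged switches to the fallback when the description is empty and to the list otherwise; the Blueprint has no other wired logic.
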WidgetBlueprint /Game/Athena/HUD/Creative/StartScreen/WBP_Creative_Island_Description
graphs [EventGraph {nodes [
   n1 = Event OnDescriptionChanged!() -> (OutputDelegate, then, bIsEmptyDescription:bool) {is=title K2Node_Event}
   > SetActiveWidgetIndex(
    self:Common Visibility Switcher Object Reference=Get CommonVisibilitySwitcher() -> (CommonVisibilitySwitcher:Common Visibility Switcher Object Reference)
    Index=SelectInt(A=1, B=0, bPickA=@n1.bIsEmptyDescription)
   )
   Event PreConstruct()
   Event Construct()
   Event Tick()
  ]}]
variables []
widgets {
 info {
  inheritedWidgetCount=1
  parentClass=/Script/FortniteUI.FortIslandDescriptionWidget
  rootWidgetClass=&CommonVisibilitySwitcher
  widgetCount=3
 }
 widgets [
  CommonVisibilitySwitcher {widgetClassPath=&CommonVisibilitySwitcher, bIsVariable=true, widget=CommonVisibilitySwitcher}
  ListView_IslandDescriptionLines {
   widgetClassPath=&CommonListView
   bInherited=true
   bIsVariable=true
   parent=CommonVisibilitySwitcher
   widget=ListView_IslandDescriptionLines
   properties {
    entryWidgetClass=&/WBP_Creative_Island_Description_ListEntry.WBP_Creative_Island_Description_ListEntry_C
    horizontalEntrySpacing=24
    verticalEntrySpacing=24
   }
   slot {class=CommonVisibilitySwitcherSlot}
  }
  WBP_Creative_Island_Description_ListEntry {
   widgetClassPath=&/WBP_Creative_Island_Description_ListEntry
   bIsVariable=true
   parent=CommonVisibilitySwitcher
   widget=WBP_Creative_Island_Description_ListEntry
   properties {buttonText=This creator has not written a description for this experience.}
   slot {class=CommonVisibilitySwitcherSlot}
  }
 ]
}
