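# Favorite button widget for Creative island start-screen UI, derived from FortExpressYourSupportFavoriteButton, with icon, label, hint text and key bind.
# Select, deselect, hover and unhover events pick one of eight animations from the selected and hovered states, driving icon and background material parameters, label color and a favorite pulse.
# UpdateSecondaryText shows a favorite or unfavorite hint only while hovered.
# Helper functions stop conflicting animations on unhover and deselect.
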WidgetBlueprint /Game/Athena/HUD/Creative/StartScreen/WBP_Favorite_Button
animations [
 HoveredToggleOn {
  binding_count=2
  length_seconds=0.48
  object=&self:HoveredToggleOn
  sequence {bindings [
    Image_FavoriteIcon {
     id=#1
     possessed_class=Image
     tracks [{
       type=MovieSceneWidgetMaterialTrack
       sections [{
         type=MovieSceneParameterSection
         channels [
          bIsInactive {type=MovieSceneScriptingFloatChannel, keys [[0,1],[4,0]]}
          bOnHover_RGBA {type=MovieSceneScriptingFloatChannel, const [1,1]}
         ]
        }]
      }]
    }
    Image_Pulse {
     id=#2
     possessed_class=Image
     tracks [
      {
       type=MovieSceneFloatTrack
       display_name=Render Opacity
       sections [{type=MovieSceneFloatSection, channels [None {type=MovieSceneScriptingFloatChannel, keys [[1,0],[4,1],[5.6,1],[9.6,0]]}]}]
      }
      {
       type=MovieScene2DTransformTrack
       sections [{
         type=MovieScene2DTransformSection
         channels [
          Scale.X {type=MovieSceneScriptingFloatChannel, keys [[2,1],[7.6,2.5]]}
          Scale.Y {type=MovieSceneScriptingFloatChannel, keys [[2,1],[7.6,2.5]]}
         ]
        }]
      }
     ]
    }
   ]}
 }
 HoveredToggleOff {
  binding_count=1
  length_seconds=0.24
  object=&self:HoveredToggleOff
  sequence {bindings [Image_FavoriteIcon {
     id=#3
     possessed_class=Image
     tracks [
      {
       type=MovieSceneWidgetMaterialTrack
       sections [{
         type=MovieSceneParameterSection
         channels [
          bIsInactive {type=MovieSceneScriptingFloatChannel, keys [[0,0],[4.8,1]]}
          bOnHover_RGBA {type=MovieSceneScriptingFloatChannel, const [1,1]}
         ]
        }]
      }
      {
       type=MovieScene2DTransformTrack
       sections [{
         type=MovieScene2DTransformSection
         channels [
          Translation.X {type=MovieSceneScriptingFloatChannel}
          Translation.Y {type=MovieSceneScriptingFloatChannel}
          Angle {type=MovieSceneScriptingFloatChannel}
          Scale.X {type=MovieSceneScriptingFloatChannel, keys [[0,1],[2.4,0.875],[4.8,1]]}
          Scale.Y {type=MovieSceneScriptingFloatChannel, keys [[0,1],[2.4,0.875],[4.8,1]]}
          Shear.X {type=MovieSceneScriptingFloatChannel}
          Shear.Y {type=MovieSceneScriptingFloatChannel}
         ]
        }]
      }
     ]
    }]}
 }
 HoverToggledOn {
  binding_count=3
  length_seconds=0.2
  object=&self:HoverToggledOn
  sequence {bindings [
    Text_NormalText {
     id=#4
     possessed_class=CommonTextBlock
     tracks [{
       type=MovieSceneColorTrack
       display_name=Color and Opacity
       sections [{
         type=MovieSceneColorSection
         channels [
          Color.R {type=MovieSceneScriptingFloatChannel, keys [[0,1],[4,0.0012]]}
          Color.G {type=MovieSceneScriptingFloatChannel, keys [[0,1],[4,0.004]]}
          Color.B {type=MovieSceneScriptingFloatChannel, keys [[0,1],[4,0.0467]]}
          Color.A {type=MovieSceneScriptingFloatChannel, const [1,2]}
         ]
        }]
      }]
    }
    Border_Background {
     id=#5
     possessed_class=Border
     tracks [{
       type=MovieSceneWidgetMaterialTrack
       sections [{type=MovieSceneParameterSection, channels [Select {type=MovieSceneScriptingFloatChannel, keys [[0,0],[4,1]]}]}]
      }]
    }
    Image_FavoriteIcon {
     id=#6
     possessed_class=Image
     tracks [{
       type=MovieSceneWidgetMaterialTrack
       sections [{
         type=MovieSceneParameterSection
         end_frame=4
         start_frame=0
         channels [
          bIsInactive {type=MovieSceneScriptingFloatChannel, const [0,1]}
          bOnHover_RGBA {type=MovieSceneScriptingFloatChannel, keys [[0,0],[4,1]]}
         ]
        }]
      }]
    }
   ]}
 }
 UnhoverToggledOn {
  binding_count=3
  length_seconds=0.2
  object=&self:UnhoverToggledOn
  sequence {bindings [
    Image_FavoriteIcon {
     id=#7
     possessed_class=Image
     tracks [{
       type=MovieSceneWidgetMaterialTrack
       sections [{
         type=MovieSceneParameterSection
         end_frame=4
         start_frame=0
         channels [
          bIsInactive {type=MovieSceneScriptingFloatChannel, const [0,1]}
          bOnHover_RGBA {type=MovieSceneScriptingFloatChannel, keys [[0,1],[4,0]]}
         ]
        }]
      }]
    }
    Border_Background {
     id=#8
     possessed_class=Border
     tracks [{
       type=MovieSceneWidgetMaterialTrack
       sections [{type=MovieSceneParameterSection, channels [Select {type=MovieSceneScriptingFloatChannel, keys [[0,1],[4,0]]}]}]
      }]
    }
    Text_NormalText {
     id=#9
     possessed_class=CommonTextBlock
     tracks [{
       type=MovieSceneColorTrack
       display_name=Color and Opacity
       sections [{
         type=MovieSceneColorSection
         channels [
          Color.R {type=MovieSceneScriptingFloatChannel, keys [[0,0.0012],[4,1]]}
          Color.G {type=MovieSceneScriptingFloatChannel, keys [[0,0.004],[4,1]]}
          Color.B {type=MovieSceneScriptingFloatChannel, keys [[0,0.0467],[4,1]]}
          Color.A {type=MovieSceneScriptingFloatChannel, const [1,2]}
         ]
        }]
      }]
    }
   ]}
 }
 HoverToggledOff {
  binding_count=5
  length_seconds=0.2
  object=&self:HoverToggledOff
  sequence {bindings [
    Image_FavoriteIcon {
     id=#10
     possessed_class=Image
     tracks [{
       type=MovieSceneWidgetMaterialTrack
       sections [{type=MovieSceneParameterSection, channels [bOnHover_RGBA {type=MovieSceneScriptingFloatChannel, keys [[0,0],[3,1]]}]}]
      }]
    }
    Border_Background {
     id=#11
     possessed_class=Border
     tracks [{
       type=MovieSceneWidgetMaterialTrack
       sections [{type=MovieSceneParameterSection, channels [Select {type=MovieSceneScriptingFloatChannel, keys [[0,0],[4,1]]}]}]
      }]
    }
    Text_NormalText {
     id=#12
     possessed_class=CommonTextBlock
     tracks [{
       type=MovieSceneColorTrack
       display_name=Color and Opacity
       sections [{
         type=MovieSceneColorSection
         channels [
          Color.R {type=MovieSceneScriptingFloatChannel, keys [[0,1],[4,0.0012]]}
          Color.G {type=MovieSceneScriptingFloatChannel, keys [[0,1],[4,0.004]]}
          Color.B {type=MovieSceneScriptingFloatChannel, keys [[0,1],[4,0.0467]]}
          Color.A {type=MovieSceneScriptingFloatChannel, const [1,2]}
         ]
        }]
      }]
    }
   ]}
 }
 UnhoverToggledOff {
  binding_count=3
  length_seconds=0.2
  object=&self:UnhoverToggledOff
  sequence {bindings [
    Image_FavoriteIcon {
     id=#10
     possessed_class=Image
     tracks [{
       type=MovieSceneWidgetMaterialTrack
       sections [{type=MovieSceneParameterSection, channels [bOnHover_RGBA {type=MovieSceneScriptingFloatChannel, keys [[0,1],[3,0]]}]}]
      }]
    }
    Text_NormalText {
     id=#13
     possessed_class=CommonTextBlock
     tracks [{
       type=MovieSceneColorTrack
       display_name=Color and Opacity
       sections [{
         type=MovieSceneColorSection
         channels [
          Color.R {type=MovieSceneScriptingFloatChannel, keys [[0,0.0012],[4,1]]}
          Color.G {type=MovieSceneScriptingFloatChannel, keys [[0,0.004],[4,1]]}
          Color.B {type=MovieSceneScriptingFloatChannel, keys [[0,0.0467],[4,1]]}
          Color.A {type=MovieSceneScriptingFloatChannel, const [1,2]}
         ]
        }]
      }]
    }
    Border_Background {
     id=#14
     possessed_class=Border
     tracks [{
       type=MovieSceneWidgetMaterialTrack
       sections [{type=MovieSceneParameterSection, channels [Select {type=MovieSceneScriptingFloatChannel, keys [[0,1],[4,0]]}]}]
      }]
    }
   ]}
 }
 UnhoveredToggleOn {
  binding_count=2
  length_seconds=0.48
  object=&self:UnhoveredToggleOn
  sequence {bindings [
    Image_FavoriteIcon {
     id=#1
     possessed_class=Image
     tracks [{
       type=MovieSceneWidgetMaterialTrack
       sections [{
         type=MovieSceneParameterSection
         channels [
          bIsInactive {type=MovieSceneScriptingFloatChannel, keys [[0,1],[4,0]]}
          bOnHover_RGBA {type=MovieSceneScriptingFloatChannel, const [0,1]}
         ]
        }]
      }]
    }
    Image_Pulse {
     id=#2
     possessed_class=Image
     tracks [
      {
       type=MovieSceneFloatTrack
       display_name=Render Opacity
       sections [{type=MovieSceneFloatSection, channels [None {type=MovieSceneScriptingFloatChannel, keys [[1,0],[4,1],[5.6,1],[9.6,0]]}]}]
      }
      {
       type=MovieScene2DTransformTrack
       sections [{
         type=MovieScene2DTransformSection
         channels [
          Scale.X {type=MovieSceneScriptingFloatChannel, keys [[2,1],[7.6,2.5]]}
          Scale.Y {type=MovieSceneScriptingFloatChannel, keys [[2,1],[7.6,2.5]]}
         ]
        }]
      }
     ]
    }
   ]}
 }
 UnhoveredToggleOff {
  binding_count=1
  length_seconds=0.24
  object=&self:UnhoveredToggleOff
  sequence {bindings [Image_FavoriteIcon {
     id=#3
     possessed_class=Image
     tracks [
      {
       type=MovieSceneWidgetMaterialTrack
       sections [{
         type=MovieSceneParameterSection
         channels [
          bIsInactive {type=MovieSceneScriptingFloatChannel, keys [[0,0],[4.8,1]]}
          bOnHover_RGBA {type=MovieSceneScriptingFloatChannel, const [0,1]}
         ]
        }]
      }
      {
       type=MovieScene2DTransformTrack
       sections [{
         type=MovieScene2DTransformSection
         channels [
          Translation.X {type=MovieSceneScriptingFloatChannel}
          Translation.Y {type=MovieSceneScriptingFloatChannel}
          Angle {type=MovieSceneScriptingFloatChannel}
          Scale.X {type=MovieSceneScriptingFloatChannel, keys [[0,1],[2.4,0.875],[4.8,1]]}
          Scale.Y {type=MovieSceneScriptingFloatChannel, keys [[0,1],[2.4,0.875],[4.8,1]]}
          Shear.X {type=MovieSceneScriptingFloatChannel}
          Shear.Y {type=MovieSceneScriptingFloatChannel}
         ]
        }]
      }
     ]
    }]}
 }
]
graphs [
 EventGraph {nodes [
   Event On Selected()
   n1 = IsHovered()
   > UpdateSecondaryText!(execute, self, bIsHovered:bool=@n1, bIsSelected:bool=true) -> (then)
   > n2 = Branch(Condition=@n1)
   > PlayAnimation(InAnimation=Get HoveredToggleOn() -> (HoveredToggleOn:Widget Animation Object Reference))
   PlayAnimation(execute=@n2.else, InAnimation=Get UnhoveredToggleOn() -> (UnhoveredToggleOn:Widget Animation Object Reference))
   Event On Hovered()
   n3 = GetSelected!(self:Common Button Base Object Reference) -> (ReturnValue:bool)
   > UpdateSecondaryText(bIsHovered=true, bIsSelected=@n3)
   > n4 = Branch(Condition=@n3)
   > PlayAnimation(InAnimation=Get HoverToggledOn() -> (HoverToggledOn:Widget Animation Object Reference))
   PlayAnimation(execute=@n4.else, InAnimation=Get HoverToggledOff() -> (HoverToggledOff:Widget Animation Object Reference))
   Event On Unhovered()
   > UpdateSecondaryText(bIsHovered=false, bIsSelected=@n3)
   > StopHoveringAnimations()
   > n5 = Branch(Condition=@n3)
   > PlayAnimation(InAnimation=Get UnhoverToggledOn() -> (UnhoverToggledOn:Widget Animation Object Reference), PlaybackSpeed=2)
   PlayAnimation(execute=@n5.else, InAnimation=Get UnhoverToggledOff() -> (UnhoverToggledOff:Widget Animation Object Reference))
   Event On Deselected()
   > UpdateSecondaryText(bIsHovered=@n1, bIsSelected=false)
   > StopTogglingAnimations!(execute, self, bIsHovered:bool=false) -> (then)
   > n6 = Branch(Condition=@n1)
   > PlayAnimation(InAnimation=Get HoveredToggleOff() -> (HoveredToggleOff:Widget Animation Object Reference))
   PlayAnimation(execute=@n6.else, InAnimation=Get UnhoveredToggleOff() -> (UnhoveredToggleOff:Widget Animation Object Reference))
  ]}
 StopHoveringAnimations {nodes [
   StopHoveringAnimations() {is=title K2Node_FunctionEntry}
   > StopAnimation(InAnimation=Get HoverToggledOn())
   > StopAnimation(InAnimation=Get UnhoverToggledOn())
   > StopAnimation(InAnimation=Get HoverToggledOff())
   > StopAnimation(InAnimation=Get UnhoverToggledOff())
   > Return Node()
  ]}
 StopTogglingAnimations {nodes [
   n1 = StopTogglingAnimations!() -> (then, bIsHovered:bool=~) {is=title K2Node_FunctionEntry}
   > StopAnimation(InAnimation=Get HoveredToggleOn())
   > StopAnimation(InAnimation=Get UnhoveredToggleOn())
   > n2 = Branch(Condition=@n1)
   > n3 = StopAnimation(InAnimation=Get UnhoveredToggleOff())
   Return Node(execute=[@n4.then, @n3.then])
   n4 = StopAnimation(execute=@n2.else, InAnimation=Get HoveredToggleOff())
  ]}
 UpdateSecondaryText {nodes [
   n1 = UpdateSecondaryText!() -> (then, bIsHovered:bool=~, bIsSelected:bool=~) {is=title K2Node_FunctionEntry}
   > SetVisibility(
    self=Get Text_Secondary() -> (Text_Secondary:Common Text Object Reference)
    InVisibility=Select(Option 0=Collapsed, Option 1=SelfHitTestInvisible, Index=@n1)
   )
   > "SetText (Text)"(
    self=Get Text_Secondary()
    InText=Select(
     Option 0:Text="LOCTEXT(\"Favorite this island to find it later!\")"
     Option 1:Text="LOCTEXT(\"You already favorited this island. Would you like to unfavorite it?\")"
     Index=@n1.bIsSelected
    ) -> (ReturnValue:Text)
   )
  ]}
]
variables []
widgets {
 info {
  inheritedWidgetCount=2
  parentClass=/Script/FortniteUI.FortExpressYourSupportFavoriteButton
  rootWidgetClass=&Overlay
  widgetCount=12
 }
 widgets [
  Overlay_1 {widgetClassPath=&Overlay, widget=Overlay_1}
  Border_Background {
   widgetClassPath=&Border
   bIsVariable=true
   parent=Overlay_1
   widget=Border_Background
   properties {
    horizontalAlignment=HAlign_Center
    verticalAlignment=VAlign_Top
    visibility=SelfHitTestInvisible
    background {
     bIsDynamicallyLoaded=false
     drawAs=Box
     imageType=NoImage
     mirroring=NoMirror
     resourceName=None
     resourceObject=/Game/Athena/HUD/MiniMap/Art/M_UI_Overview_Container
     tiling=NoTile
     imageSize=<64 64>
     margin=<ltrb 0.5 0.5 0.5 0.5>
     outlineSettings {
      bUseBrushTransparency=false
      roundingType=HalfHeightRadius
      width=0
      color {colorUseRule=UseColor_Specified, specifiedColor=<rgba 0 0 0 0>}
      cornerRadii=<0 0 0 0>
     }
     tintColor {colorUseRule=UseColor_Specified, specifiedColor=<rgba 1 1 1 1>}
     uVRegion {bIsValid=false, max=<0 0>, min=<0 0>}
    }
    padding=<ltrb 0 0 0 0>
   }
   slot {class=OverlaySlot, properties {horizontalAlignment=HAlign_Fill, verticalAlignment=VAlign_Fill}}
  }
  Overlay_0 {
   widgetClassPath=&Overlay
   parent=Overlay_1
   widget=Overlay_0
   slot {class=OverlaySlot, properties {verticalAlignment=VAlign_Center, padding=<ltrb 48 0 0 0>}}
  }
  Image_Pulse {
   widgetClassPath=&Image
   bIsVariable=true
   parent=Overlay_0
   widget=Image_Pulse
   properties {
    renderOpacity=0
    brush {
     bIsDynamicallyLoaded=false
     drawAs=Image
     imageType=NoImage
     mirroring=NoMirror
     resourceName=None
     resourceObject=/FortUILibrary/Foundation/Buttons/Materials/Favorite/MI_UI_Favorite_Pulse_Modal
     tiling=NoTile
     imageSize=<48 48>
     margin=<ltrb 0 0 0 0>
     outlineSettings {
      bUseBrushTransparency=false
      roundingType=HalfHeightRadius
      width=0
      color {colorUseRule=UseColor_Specified, specifiedColor=<rgba 0 0 0 0>}
      cornerRadii=<0 0 0 0>
     }
     tintColor {colorUseRule=UseColor_Specified, specifiedColor=<rgba 1 1 1 1>}
     uVRegion {bIsValid=false, max=<0 0>, min=<0 0>}
    }
   }
   slot {class=OverlaySlot, properties {horizontalAlignment=HAlign_Fill, verticalAlignment=VAlign_Fill}}
  }
  Image_FavoriteIcon {
   widgetClassPath=&Image
   bIsVariable=true
   parent=Overlay_0
   widget=Image_FavoriteIcon
   properties {brush {
     bIsDynamicallyLoaded=false
     drawAs=Image
     imageType=NoImage
     mirroring=NoMirror
     resourceName=None
     resourceObject=/FortUILibrary/Foundation/Buttons/Materials/Favorite/MI_UI_Favorite_ExpressYourSupport
     tiling=NoTile
     imageSize=<60 60>
     margin=<ltrb 0 0 0 0>
     outlineSettings {
      bUseBrushTransparency=false
      roundingType=HalfHeightRadius
      width=0
      color {colorUseRule=UseColor_Specified, specifiedColor=<rgba 0 0 0 0>}
      cornerRadii=<0 0 0 0>
     }
     tintColor {colorUseRule=UseColor_Specified, specifiedColor=<rgba 1 1 1 1>}
     uVRegion {bIsValid=false, max=<0 0>, min=<0 0>}
    }}
   slot {class=OverlaySlot, properties {horizontalAlignment=HAlign_Fill, verticalAlignment=VAlign_Fill}}
  }
  VerticalBox_0 {
   widgetClassPath=&VerticalBox
   parent=Overlay_1
   widget=VerticalBox_0
   slot {class=OverlaySlot, properties {verticalAlignment=VAlign_Center, padding=<ltrb 132 32 48 32>}}
  }
  Text_NormalText {
   widgetClassPath=&CommonTextBlock
   bInherited=true
   parent=VerticalBox_0
   widget=Text_NormalText
   properties {
    mobileFontSizeMultiplier=1.3
    text=FAVORITE
    font {
     bForceMonospaced=false
     bMaterialIsStencil=false
     fontMaterial=null
     fontObject=&BurbankSmall
     letterSpacing=0
     monospacedWidth=1
     size=21
     skewAmount=0
     typefaceFontName=Black
     outlineSettings {
      bApplyOutlineToDropShadows=false
      bMiteredCorners=false
      bSeparateFillAlpha=false
      outlineMaterial=null
      outlineSize=0
      outlineColor=<rgba 0 0 0 1>
     }
    }
    shadowOffset=<0 0>
   }
   slot {class=VerticalBoxSlot, properties {horizontalAlignment=HAlign_Left, padding=<ltrb 0 0 0 8>}}
  }
  Text_Secondary {
   widgetClassPath=&CommonTextBlock
   bIsVariable=true
   parent=VerticalBox_0
   widget=Text_Secondary
   properties {
    mobileFontSizeMultiplier=1.3
    text="Favorite this island to find it later!"
    visibility=Collapsed
    colorAndOpacity {colorUseRule=UseColor_Specified, specifiedColor=<rgba 0.007 0.314 1 1>}
    font {
     bForceMonospaced=false
     bMaterialIsStencil=false
     fontMaterial=null
     fontObject=&BurbankSmall
     letterSpacing=0
     monospacedWidth=1
     size=18
     skewAmount=0
     typefaceFontName=Bold
     outlineSettings {
      bApplyOutlineToDropShadows=false
      bMiteredCorners=false
      bSeparateFillAlpha=false
      outlineMaterial=null
      outlineSize=0
      outlineColor=<rgba 0 0 0 1>
     }
    }
    shadowOffset=<0 0>
   }
   slot {class=VerticalBoxSlot, properties {horizontalAlignment=HAlign_Left}}
  }
  KeyBind {
   widgetClassPath=&Overlay
   bIsVariable=true
   parent=Overlay_1
   widget=KeyBind
   properties {visibility=HitTestInvisible}
   slot {class=OverlaySlot, properties {verticalAlignment=VAlign_Center, padding=<ltrb -64 0 0 0>}}
  }
  Spacer_70 {widgetClassPath=&Spacer, parent=KeyBind, widget=Spacer_70, properties {size=<128 64>}, slot {class=OverlaySlot}}
  Shadow {
   widgetClassPath=&Image
   bIsVariable=true
   parent=KeyBind
   widget=Shadow
   properties {
    visibility=Collapsed
    brush {
     bIsDynamicallyLoaded=false
     drawAs=Image
     imageType=NoImage
     mirroring=NoMirror
     resourceName=None
     resourceObject=/FortUILibrary/Foundation/Art_ToolKit/Tools/M_UI_XP_Everywhere_Backing_Darken
     tiling=NoTile
     imageSize=<32 32>
     margin=<ltrb 0 0 0 0>
     outlineSettings {
      bUseBrushTransparency=false
      roundingType=HalfHeightRadius
      width=0
      color {colorUseRule=UseColor_Specified, specifiedColor=<rgba 0 0 0 0>}
      cornerRadii=<0 0 0 0>
     }
     tintColor {colorUseRule=UseColor_Specified, specifiedColor=<rgba 0 0 0 1>}
     uVRegion {bIsValid=false, max=<0 0>, min=<0 0>}
    }
    colorAndOpacity=<rgba 1 1 1 0.5>
    renderTransform {angle=0, scale=<1.5 1.5>, shear=<0 0>, translation=<0 0>}
   }
   slot {class=OverlaySlot, properties {horizontalAlignment=HAlign_Center, verticalAlignment=VAlign_Center}}
  }
  InputActionWidget {
   widgetClassPath=&CommonActionWidget
   bInherited=true
   bIsVariable=true
   parent=KeyBind
   widget=InputActionWidget
   properties {
    progressMaterialParam=Percentage
    progressMaterialBrush {
     bIsDynamicallyLoaded=false
     drawAs=Image
     imageType=NoImage
     mirroring=NoMirror
     resourceName=None
     resourceObject=/Game/UI/Materials/HollowCircularProgressBar
     tiling=NoTile
     imageSize=<48 48>
     margin=<ltrb 0 0 0 0>
     outlineSettings {
      bUseBrushTransparency=false
      roundingType=HalfHeightRadius
      width=0
      color {colorUseRule=UseColor_Specified, specifiedColor=<rgba 0 0 0 0>}
      cornerRadii=<0 0 0 0>
     }
     tintColor {colorUseRule=UseColor_Specified, specifiedColor=<rgba 0.7913 0.0086 0.1845 1>}
     uVRegion {bIsValid=false, max=<0 0>, min=<0 0>}
    }
   }
   slot {class=OverlaySlot, properties {horizontalAlignment=HAlign_Center, verticalAlignment=VAlign_Center}}
  }
 ]
}
