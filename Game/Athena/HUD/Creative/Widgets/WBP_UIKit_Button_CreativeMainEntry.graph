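# CommonButtonBase widget for a Creative main-entry button, with a material-driven background, text label, and hidden lock icon and dimming overlay.
# PreConstruct applies ButtonText; selection plays Select to grow the button, highlight its material and darken the text, and deselection reverses it.
# Hover plays a Burst material flash.
# Focus changes on a selected button play Off-Focus forward or reversed; the graph never toggles the lock widgets.
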WidgetBlueprint /Game/Athena/HUD/Creative/Widgets/WBP_UIKit_Button_CreativeMainEntry
animations [
 Burst {
  binding_count=1
  length_seconds=3.05
  object=&self:Burst
  sequence {bindings [ButtonImage {
     id=#1
     possessed_class=Image
     tracks [{
       type=MovieSceneWidgetMaterialTrack
       sections [{
         type=MovieSceneParameterSection
         channels [
          Select {type=MovieSceneScriptingFloatChannel}
          Burst {type=MovieSceneScriptingFloatChannel, keys [[-1,0],[2,1],[59.9997,0]]}
          Base_RGB1R {type=MovieSceneScriptingFloatChannel, keys [[-1,0],[2,0.0003],[59.9997,0]]}
          Base_RGB1G {type=MovieSceneScriptingFloatChannel, keys [[-1,0.0116],[2,0.0219],[59.9997,0.0116]]}
          Base_RGB1B {type=MovieSceneScriptingFloatChannel, keys [[-1,0.4342],[2,0.4064],[59.9997,0.4342]]}
          Base_RGB1A {type=MovieSceneScriptingFloatChannel, const [1,3]}
         ]
        }]
      }]
    }]}
 }
 Select {
  binding_count=3
  length_seconds=0.15
  object=&self:Select
  sequence {bindings [
    SizeBox_0 {
     id=#2
     display_name=Anim_SizeBox
     possessed_class=SizeBox
     tracks [
      {
       type=MovieSceneFloatTrack
       display_name=Width Override
       sections [{type=MovieSceneFloatSection, end_frame=3, start_frame=0, channels [None {type=MovieSceneScriptingFloatChannel, keys [[0,456],[3,468]]}]}]
      }
      {
       type=MovieSceneFloatTrack
       display_name=Height Override
       sections [{type=MovieSceneFloatSection, end_frame=3, start_frame=0, channels [None {type=MovieSceneScriptingFloatChannel, keys [[0,56],[3,72]]}]}]
      }
     ]
    }
    ButtonImage {
     id=#3
     possessed_class=Image
     tracks [{
       type=MovieSceneWidgetMaterialTrack
       sections [{type=MovieSceneParameterSection, channels [Select {type=MovieSceneScriptingFloatChannel, keys [[0,0],[3,1]]}]}]
      }]
    }
    ButtonTextBlock {
     id=#4
     possessed_class=CommonTextBlock
     tracks [{
       type=MovieSceneColorTrack
       display_name=Color and Opacity
       sections [{
         type=MovieSceneColorSection
         channels [
          Color.R {type=MovieSceneScriptingFloatChannel, keys [[0,1],[2.9997,0]]}
          Color.G {type=MovieSceneScriptingFloatChannel, keys [[0,1],[2.9997,0.004],[3,0.004]]}
          Color.B {type=MovieSceneScriptingFloatChannel, keys [[0,1],[2.9997,0.0722],[3,0.0722]]}
          Color.A {type=MovieSceneScriptingFloatChannel}
         ]
        }]
      }]
    }
   ]}
 }
 Off-Focus {
  binding_count=2
  length_seconds=0.15
  object=&self:Off-Focus
  sequence {bindings [
    ButtonTextBlock {
     id=#5
     possessed_class=CommonTextBlock
     tracks [{
       type=MovieSceneColorTrack
       display_name=Color and Opacity
       sections [{
         type=MovieSceneColorSection
         end_frame=3
         start_frame=0
         channels [
          Color.R {type=MovieSceneScriptingFloatChannel, keys [[0,0],[3,1]]}
          Color.G {type=MovieSceneScriptingFloatChannel, keys [[0,0.004],[3,1]]}
          Color.B {type=MovieSceneScriptingFloatChannel, keys [[0,0.0722],[3,1]]}
          Color.A {type=MovieSceneScriptingFloatChannel, const [1,2]}
         ]
        }]
      }]
    }
    ButtonImage {
     id=#6
     possessed_class=Image
     tracks [{
       type=MovieSceneWidgetMaterialTrack
       sections [{
         type=MovieSceneParameterSection
         end_frame=3
         start_frame=0
         channels [
          Focus {type=MovieSceneScriptingFloatChannel, keys [[0,1],[3,0]]}
          Select {type=MovieSceneScriptingFloatChannel, const [1,2]}
         ]
        }]
      }]
    }
   ]}
 }
]
graphs [EventGraph {nodes [
   Event On Selected()
   > PlayAnimation(InAnimation=Get Select() -> (Select:Widget Animation Object Reference))
   Event PreConstruct()
   > Set Text(Text=Get ButtonText() -> (ButtonText:Text), self=Get ButtonTextBlock() -> (ButtonTextBlock:Common Text Object Reference))
   Event On Deselected()
   > StopAllAnimations(self:User Widget Object Reference)
   > PlayAnimation(InAnimation=Get Select(), PlayMode=Reverse)
   Event On Hovered()
   > PlayAnimation(InAnimation=Get Burst() -> (Burst:Widget Animation Object Reference))
   Event On Focused() {is=title K2Node_Event}
   > Branch(Condition=GetSelected!(self:Common Button Base Object Reference) -> (ReturnValue:bool))
   n1 = Get Off-Focus() -> (Off-Focus:Widget Animation Object Reference)
   > PlayAnimation(InAnimation=@n1, PlayMode=Reverse)
   Event On Unhovered()
   > n2 = Branch(Condition=GetSelected())
   > PlayAnimation(InAnimation=Get Burst(), PlayMode=Reverse)
   StopAnimation(execute=@n2.else, InAnimation=Get Burst())
   Event On Unfocused() {is=title K2Node_Event}
   > Branch(Condition=GetSelected())
   > PlayAnimation(InAnimation=@n1)
  ]}]
variables [
 ButtonText {type=string, default="{Text}"}
 Parameter Name {type=string, default=Select}
]
widgets {
 info {inheritedWidgetCount=1, parentClass=&CommonButtonBase, rootWidgetClass=&SizeBox, widgetCount=10}
 widgets [
  SizeBox_0 {
   widgetClassPath=&SizeBox
   widget=SizeBox_0
   properties {bOverride_HeightOverride=true, bOverride_WidthOverride=true, heightOverride=72, widthOverride=468}
  }
  Anim_SizeBox {
   widgetClassPath=&SizeBox
   parent=SizeBox_0
   widget=Anim_SizeBox
   properties {bOverride_HeightOverride=true, bOverride_WidthOverride=true, heightOverride=56, widthOverride=456}
   slot {class=SizeBoxSlot, properties {horizontalAlignment=HAlign_Left, verticalAlignment=VAlign_Center}}
  }
  Overlay_0 {widgetClassPath=&Overlay, parent=Anim_SizeBox, widget=Overlay_0, slot {class=SizeBoxSlot}}
  ButtonImage {
   widgetClassPath=&Image
   bIsVariable=true
   parent=Overlay_0
   widget=ButtonImage
   properties {brush {
     bIsDynamicallyLoaded=false
     drawAs=Image
     imageType=NoImage
     mirroring=NoMirror
     resourceName=None
     resourceObject=/Game/Athena/HUD/MiniMap/Art/QuestBundle/M_Ui_QuestBundle_Bckgrnd_inst
     tiling=NoTile
     imageSize=<32 32>
     margin=<ltrb 0 0 0 0>
     outlineSettings {
      bUseBrushTransparency=false
      roundingType=HalfHeightRadius
      width=0
      color {colorUseRule=UseColor_Specified, specifiedColor=<rgba 0 0 0 0>}
      cornerRadii=<0 0 0 0>
     }
     tintColor {colorUseRule=UseColor_Specified, specifiedColor=<rgba 1 1 1 1>}
     uVRegion {bIsValid=false, max=<0 0>, min=<0 0>}
    }}
   slot {class=OverlaySlot, properties {horizontalAlignment=HAlign_Fill, verticalAlignment=VAlign_Fill}}
  }
  HorizontalBox_0 {
   widgetClassPath=&HorizontalBox
   parent=Overlay_0
   widget=HorizontalBox_0
   slot {class=OverlaySlot, properties {verticalAlignment=VAlign_Center, padding=<ltrb 18 0 18 0>}}
  }
  LockContainer {
   widgetClassPath=&SizeBox
   bIsVariable=true
   parent=HorizontalBox_0
   widget=LockContainer
   properties {bOverride_HeightOverride=true, bOverride_WidthOverride=true, heightOverride=28, visibility=Hidden, widthOverride=28}
   slot {class=HorizontalBoxSlot, properties {padding=<ltrb 0 0 12 0>, size {sizeRule=Fill, value=1}}}
  }
  LockImage {
   widgetClassPath=&Image
   bIsVariable=true
   parent=LockContainer
   widget=LockImage
   properties {
    visibility=SelfHitTestInvisible
    brush {
     bIsDynamicallyLoaded=false
     drawAs=Image
     imageType=NoImage
     mirroring=NoMirror
     resourceName=None
     resourceObject=/Game/Athena/UI/InGame/Creative/PropertyEditor/Icon_Picker/Textures/128x128/T_UI_IconLibrary_padlock_128
     tiling=NoTile
     imageSize=<128 24>
     margin=<ltrb 0 0 0 0>
     outlineSettings {
      bUseBrushTransparency=false
      roundingType=HalfHeightRadius
      width=0
      color {colorUseRule=UseColor_Specified, specifiedColor=<rgba 0 0 0 0>}
      cornerRadii=<0 0 0 0>
     }
     tintColor {colorUseRule=UseColor_Specified, specifiedColor=<rgba 0 0.0908 1 1>}
     uVRegion {bIsValid=false, max=<0 0>, min=<0 0>}
    }
   }
   slot {class=SizeBoxSlot}
  }
  ButtonTextBlock {
   widgetClassPath=&CommonTextBlock
   bIsVariable=true
   parent=HorizontalBox_0
   widget=ButtonTextBlock
   properties {
    clipping=ClipToBounds
    text="{Text}"
    font {
     bForceMonospaced=false
     bMaterialIsStencil=false
     fontMaterial=null
     fontObject=&BurbankSmall
     letterSpacing=0
     monospacedWidth=1
     size=18
     skewAmount=0
     typefaceFontName=Black
     outlineSettings {
      bApplyOutlineToDropShadows=false
      bMiteredCorners=false
      bSeparateFillAlpha=false
      outlineMaterial=null
      outlineSize=0
      outlineColor=<rgba 0 0 0 1>
     }
    }
    shadowOffset=<0 0>
   }
   slot {class=HorizontalBoxSlot, properties {verticalAlignment=VAlign_Center}}
  }
  LockOverlayImage {
   widgetClassPath=&Image
   bIsVariable=true
   parent=Overlay_0
   widget=LockOverlayImage
   properties {
    visibility=Hidden
    brush {
     bIsDynamicallyLoaded=false
     drawAs=Image
     imageType=NoImage
     mirroring=NoMirror
     resourceName=None
     resourceObject=null
     tiling=NoTile
     imageSize=<32 32>
     margin=<ltrb 0 0 0 0>
     outlineSettings {
      bUseBrushTransparency=false
      roundingType=HalfHeightRadius
      width=0
      color {colorUseRule=UseColor_Specified, specifiedColor=<rgba 0 0 0 0>}
      cornerRadii=<0 0 0 0>
     }
     tintColor {colorUseRule=UseColor_Specified, specifiedColor=<rgba 0 0.0056 0.1329 0.502>}
     uVRegion {bIsValid=false, max=<0 0>, min=<0 0>}
    }
   }
   slot {class=OverlaySlot, properties {horizontalAlignment=HAlign_Fill, verticalAlignment=VAlign_Fill}}
  }
  InputActionWidget {widgetClassPath=&CommonActionWidget, bInherited=true, widget=null}
 ]
}
