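# Overview screen for Creative islands, built on FortCreativeGameInfoMapTab, with game info header, minigame joining bar and Exit button.
# Sub-tab buttons sit beside a switcher holding island description, recommend/favorite, miscellaneous (respawn, clear saved data) and creator credits.
# Navigation functions target visible content widgets; leaving credits leftward deactivates attributions and returns its tab button.
# Pre-construct sets a tinted top image sized for mobile or not; activation plays a fade-and-slide Intro.
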WidgetBlueprint /Game/Athena/HUD/Creative/StartScreen/WBP_Creative_HUD_StartScreen_Overview
animations [Intro {
  binding_count=4
  length_seconds=0.35
  object=&self:Intro
  sequence {
   bindings [
    SafeZone_1 {
     id=#1
     possessed_class=SafeZone
     tracks [{
       type=MovieSceneFloatTrack
       display_name=Render Opacity
       sections [
        {type=MovieSceneFloatSection, channels [None {type=MovieSceneScriptingFloatChannel, keys [[0,0],[10.501,1]]}]}
        {type=MovieSceneFloatSection, channels [None {type=MovieSceneScriptingFloatChannel, keys [[0,0],[10.501,1]]}]}
       ]
      }]
    }
    HorizontalBox_142 {
     id=#2
     possessed_class=HorizontalBox
     tracks [
      {
       type=MovieSceneFloatTrack
       display_name=Render Opacity
       sections [{type=MovieSceneFloatSection, channels [None {type=MovieSceneScriptingFloatChannel, keys [[2.3545,0],[7.0635,1]]}]}]
      }
      {
       type=MovieScene2DTransformTrack
       sections [{
         type=MovieScene2DTransformSection
         channels [
          Translation.X {type=MovieSceneScriptingFloatChannel, keys [[0,-176],[3.531,0]]}
          Translation.Y {type=MovieSceneScriptingFloatChannel}
         ]
        }]
      }
     ]
    }
    MinigameInfo_HB {
     id=#3
     possessed_class=HorizontalBox
     tracks [
      {
       type=MovieSceneFloatTrack
       display_name=Render Opacity
       sections [{type=MovieSceneFloatSection, channels [None {type=MovieSceneScriptingFloatChannel, thinned_from=3, keys [[0,0],[8,1]]}]}]
      }
      {
       type=MovieScene2DTransformTrack
       sections [{
         type=MovieScene2DTransformSection
         channels [
          Translation.X {type=MovieSceneScriptingFloatChannel, keys [[0,-360],[4,0.95],[6,9],[7,0]]}
          Translation.Y {type=MovieSceneScriptingFloatChannel, const [0,1]}
         ]
        }]
      }
     ]
    }
    HorizontalBox_145 {
     id=#4
     possessed_class=HorizontalBox
     tracks [
      {
       type=MovieSceneFloatTrack
       display_name=Render Opacity
       sections [{type=MovieSceneFloatSection, channels [None {type=MovieSceneScriptingFloatChannel, keys [[6,0],[10.5005,1]]}]}]
      }
      {
       type=MovieScene2DTransformTrack
       sections [{
         type=MovieScene2DTransformSection
         channels [
          Translation.X {type=MovieSceneScriptingFloatChannel, const [0,2]}
          Translation.Y {type=MovieSceneScriptingFloatChannel, keys [[4.2,200],[8.4855,0]]}
          Scale.X {type=MovieSceneScriptingFloatChannel}
          Scale.Y {type=MovieSceneScriptingFloatChannel}
         ]
        }]
      }
     ]
    }
   ]
   display_rate {numerator=30}
  }
 }]
graphs [
 AttributionsSubTabLeftNavigation {nodes [
   n1 = DeactivateWidget(self=Get CreatorAttributions() -> (CreatorAttributions:Fort Creative Creator Attributions Object Reference))
   Return Node!(
    execute=[@n1.then, @n2.then]
    ReturnValue:Widget Object Reference=Get Button_CreditsSubTab() -> (Button_CreditsSubTab:Common Button Base Object Reference)
   )
   n2 = AttributionsSubTabLeftNavigation!() -> (then, Navigation:EUINavigation Enum) {is=title K2Node_FunctionEntry}
  ]}
 AttributionsSubTabRightNavigation {nodes [
   AttributionsSubTabRightNavigation!() -> (then, Navigation:EUINavigation Enum) {is=title K2Node_FunctionEntry}
   > Branch(Condition=IsVisible(self=Get CreatorAttributions()))
   > Return Node!(execute, NewParam:Widget Object Reference=Get CreatorAttributions())
  ]}
 EventGraph {nodes [
   Event PreConstruct()
   > SetBrush!(
    execute
    self=Get Image_145() -> (Image_145:Image Object Reference)
    InBrush_TintColor_SpecifiedColor:Linear Color Structure=(R=0,G=0.0012,B=0.0999,A=0.4)
    InBrush_TintColor_ColorUseRule:ESlateColorStylingMode Enum
    InBrush_DrawAs:ESlateBrushDrawType Enum=Image
    InBrush_Tiling:ESlateBrushTileType Enum=NoTile
    InBrush_Mirroring:ESlateBrushMirrorType Enum=NoMirror
    InBrush_ImageSize:Vector 2D Structure=Select(Option 0:Vector 2D Structure=(X=130,Y=130), Option 1:Vector 2D Structure=(X=130,Y=160), Index=IsMobileGame()) -> (ReturnValue:Vector 2D Structure)
    InBrush_Margin:Margin Structure=(Left=0,Top=0,Right=0,Bottom=0)
    InBrush_ResourceObject:Object Reference
    InBrush_OutlineSettings:Slate Brush Outline Settings Structure=(CornerRadii=(X=0,Y=0,Z=0,W=0),Color=(SpecifiedColor=(R=0,G=0,B=0,A=0),ColorUseRule=UseColor_Specified),Width=0,RoundingType=HalfHeightRadius,bUseBrushTransparency=False)
   ) -> (then)
   Event On Activated()
   > PlayAnimation(InAnimation=Get Intro())
  ]}
 Like and Favorite Subtab Right Navigation {nodes [
   Like and Favorite Subtab Right Navigation!() -> (then, Navigation:EUINavigation Enum) {is=title K2Node_FunctionEntry}
   n1 = Get Button_Favorite() -> (Button_Favorite:Fort Express Your Support Favorite Button Object Reference)
   > n2 = Branch(Condition=IsVisible(self=@n1))
   > Return Node!(execute, ReturnValue:Widget Object Reference=@n1)
   n3 = Get Button_Like() -> (Button_Like:Fort Express Your Support Thumbs Up Button Object Reference)
   n4 = Branch(execute=@n2.else, Condition=IsVisible(self=@n3))
   > Return Node(ReturnValue=@n3)
   Return Node(execute=@n4.else)
  ]}
 Miscellaneous Subtab Right Navigation {nodes [
   Miscellaneous Subtab Right Navigation!() -> (then, Navigation:EUINavigation Enum) {is=title K2Node_FunctionEntry}
   n1 = Get Button_Respawn() -> (Button_Respawn:Common Button Base Object Reference)
   > n2 = Branch(Condition=IsVisible(self=@n1))
   > Return Node(ReturnValue=@n1)
   n3 = Get Button_ClearSavedData() -> (Button_ClearSavedData:Common Button Base Object Reference)
   n4 = Branch(execute=@n2.else, Condition=IsVisible(self=@n3))
   > Return Node(ReturnValue=@n3)
   Return Node(execute=@n4.else)
  ]}
]
variables []
widgets {
 info {inheritedWidgetCount=16, parentClass=/Script/FortniteUI.FortCreativeGameInfoMapTab, rootWidgetClass=&Overlay, widgetCount=30}
 widgets [
  Overlay_0 {widgetClassPath=&Overlay, widget=Overlay_0}
  BackgroundBlur_0 {
   widgetClassPath=&BackgroundBlur
   parent=Overlay_0
   widget=BackgroundBlur_0
   properties {blurStrength=12}
   slot {class=OverlaySlot, properties {horizontalAlignment=HAlign_Fill, verticalAlignment=VAlign_Fill, padding=<ltrb -2 -2 -2 -2>}}
  }
  Border_235 {
   widgetClassPath=&Border
   parent=BackgroundBlur_0
   widget=Border_235
   properties {
    visibility=SelfHitTestInvisible
    background {
     bIsDynamicallyLoaded=false
     drawAs=Image
     imageType=NoImage
     mirroring=NoMirror
     resourceName=None
     resourceObject=/Game/Athena/UI/Frontend/Art/M_UI_RadialGradientMaster_Dark3
     tiling=NoTile
     imageSize=<32 32>
     margin=<ltrb 0 0 0 0>
     outlineSettings {
      bUseBrushTransparency=false
      roundingType=HalfHeightRadius
      width=0
      color {colorUseRule=UseColor_Specified, specifiedColor=<rgba 0 0 0 0>}
      cornerRadii=<0 0 0 0>
     }
     tintColor {colorUseRule=UseColor_Specified, specifiedColor=<rgba 1 1 1 1>}
     uVRegion {bIsValid=false, max=<0 0>, min=<0 0>}
    }
    brushColor=<rgba 1 1 1 0.8>
    padding=<ltrb 0 0 0 0>
   }
   slot {class=BackgroundBlurSlot, properties {padding=<ltrb 0 0 0 0>}}
  }
  SafeZone_1 {
   widgetClassPath=&SafeZone
   bIsVariable=true
   parent=Overlay_0
   widget=SafeZone_1
   properties {padBottom=false}
   slot {class=OverlaySlot, properties {horizontalAlignment=HAlign_Fill}}
  }
  Image_145 {
   widgetClassPath=&Image
   bIsVariable=true
   parent=SafeZone_1
   widget=Image_145
   properties {brush {
     bIsDynamicallyLoaded=false
     drawAs=Image
     imageType=NoImage
     mirroring=NoMirror
     resourceName=None
     resourceObject=null
     tiling=NoTile
     imageSize=<130 130>
     margin=<ltrb 0 0 0 0>
     outlineSettings {
      bUseBrushTransparency=false
      roundingType=HalfHeightRadius
      width=0
      color {colorUseRule=UseColor_Specified, specifiedColor=<rgba 0 0 0 0>}
      cornerRadii=<0 0 0 0>
     }
     tintColor {colorUseRule=UseColor_Specified, specifiedColor=<rgba 0 0.0229 1 0.15>}
     uVRegion {bIsValid=false, max=<0 0>, min=<0 0>}
    }}
   slot {class=SafeZoneSlot, properties {padding=<ltrb -100 -60 -100 0>}}
  }
  Main {
   widgetClassPath=&VerticalBox
   parent=Overlay_0
   widget=Main
   slot {class=OverlaySlot, properties {horizontalAlignment=HAlign_Fill, verticalAlignment=VAlign_Fill, padding=<ltrb 60 80 60 30>}}
  }
  TabsContainer {
   widgetClassPath=&Border
   parent=Main
   widget=TabsContainer
   properties {verticalAlignment=VAlign_Top, brushColor=<rgba 0 0.0116 0.159 0>, padding=<ltrb 0 0 0 0>}
   slot {class=VerticalBoxSlot}
  }
  Spacer_70 {
   widgetClassPath=&Spacer
   parent=TabsContainer
   widget=Spacer_70
   properties {size=<1 70>}
   slot {class=BorderSlot, properties {verticalAlignment=VAlign_Top, padding=<ltrb 0 0 0 0>}}
  }
  MinigameInfo_HB {widgetClassPath=&HorizontalBox, parent=Main, widget=MinigameInfo_HB, slot {class=VerticalBoxSlot}}
  MinigameInfo_Header {
   widgetClassPath=&/WBP_Creative_Minigame_Info
   bInherited=true
   bIsVariable=true
   parent=MinigameInfo_HB
   widget=MinigameInfo_Header
   slot {class=HorizontalBoxSlot}
  }
  HorizontalBox_142 {
   widgetClassPath=&HorizontalBox
   parent=Main
   widget=HorizontalBox_142
   slot {class=VerticalBoxSlot, properties {verticalAlignment=VAlign_Top, padding=<ltrb 0 48 0 40>, size {sizeRule=Fill, value=1}}}
  }
  ContentButtons {
   widgetClassPath=&VerticalBox
   bIsVariable=true
   parent=HorizontalBox_142
   widget=ContentButtons
   slot {class=HorizontalBoxSlot, properties {padding=<ltrb 0 -6 0 0>, size {sizeRule=Automatic, value=0.3}}}
  }
  Button_GameInfoSubTab {
   widgetClassPath=/Game/Athena/HUD/Creative/Widgets/WBP_UIKit_Button_CreativeMainEntry
   bInherited=true
   bIsVariable=true
   parent=ContentButtons
   widget=Button_GameInfoSubTab
   properties {
    bSelectable=true
    bShouldSelectUponReceivingFocus=true
    buttonText=GAME INFO
    navigation=&self:WidgetTree.Button_GameInfoSubTab.WidgetNavigation_0
   }
   slot {class=VerticalBoxSlot, properties {padding=<ltrb 0 0 0 2>}}
  }
  Button_LikeAndFavoriteSubTab {
   widgetClassPath=&WBP_UIKit_Button_CreativeMainEntry
   bInherited=true
   bIsVariable=true
   parent=ContentButtons
   widget=Button_LikeAndFavoriteSubTab
   properties {
    bSelectable=true
    bShouldSelectUponReceivingFocus=true
    buttonText=RECOMMEND AND FAVORITE
    navigation=&self:WidgetTree.Button_LikeAndFavoriteSubTab.WidgetNavigation_0
   }
   slot {class=VerticalBoxSlot, properties {padding=<ltrb 0 0 0 2>}}
  }
  Button_MiscellaneousSubTab {
   widgetClassPath=&WBP_UIKit_Button_CreativeMainEntry
   bInherited=true
   bIsVariable=true
   parent=ContentButtons
   widget=Button_MiscellaneousSubTab
   properties {
    bSelectable=true
    bShouldSelectUponReceivingFocus=true
    buttonText=MISCELLANEOUS
    navigation=&self:WidgetTree.Button_MiscellaneousSubTab.WidgetNavigation_0
   }
   slot {class=VerticalBoxSlot}
  }
  Button_CreditsSubTab {
   widgetClassPath=&WBP_UIKit_Button_CreativeMainEntry
   bInherited=true
   bIsVariable=true
   parent=ContentButtons
   widget=Button_CreditsSubTab
   properties {
    bSelectable=true
    bShouldSelectUponReceivingFocus=true
    buttonText=CREDITS
    navigation=&self:WidgetTree.Button_CreditsSubTab.WidgetNavigation_0
   }
   slot {class=VerticalBoxSlot}
  }
  Switcher_MainContent {
   widgetClassPath=&CommonActivatableWidgetSwitcher
   bInherited=true
   bIsVariable=true
   parent=HorizontalBox_142
   widget=Switcher_MainContent
   properties {transitionDuration=0.1}
   slot {class=HorizontalBoxSlot, properties {padding=<ltrb 72 0 0 0>, size {sizeRule=Fill, value=0.7}}}
  }
  IslandDescription_GameInfoContent {
   widgetClassPath=&/WBP_Creative_Island_Description
   bInherited=true
   bIsVariable=true
   parent=Switcher_MainContent
   widget=IslandDescription_GameInfoContent
   properties {
    clipping=ClipToBounds
    navigation=&self:WidgetTree.IslandDescription_GameInfoContent.WidgetNavigation_0
   }
   slot {class=WidgetSwitcherSlot}
  }
  Widget_LikeAndFavoriteContent {
   widgetClassPath=&VerticalBox
   bInherited=true
   parent=Switcher_MainContent
   widget=Widget_LikeAndFavoriteContent
   properties {navigation=&self:WidgetTree.Widget_LikeAndFavoriteContent.WidgetNavigation_0}
   slot {class=WidgetSwitcherSlot}
  }
  Button_Favorite {
   widgetClassPath=&/WBP_Favorite_Button
   bInherited=true
   bIsVariable=true
   parent=Widget_LikeAndFavoriteContent
   widget=Button_Favorite
   properties {
    navigation=&self:WidgetTree.Button_Favorite.WidgetNavigation_0
    visibility=Visible
   }
   slot {class=VerticalBoxSlot, properties {padding=<ltrb 0 0 0 20>}}
  }
  Button_Like {
   widgetClassPath=&/WBP_Like_Button
   bInherited=true
   bIsVariable=true
   parent=Widget_LikeAndFavoriteContent
   widget=Button_Like
   slot {class=VerticalBoxSlot}
  }
  Widget_MiscellaneousContent {
   widgetClassPath=&VerticalBox
   bInherited=true
   parent=Switcher_MainContent
   widget=Widget_MiscellaneousContent
   properties {navigation=&self:WidgetTree.Widget_MiscellaneousContent.WidgetNavigation_0}
   slot {class=WidgetSwitcherSlot}
  }
  Button_Respawn {
   widgetClassPath=&/WBP_Creative_OverviewScreen_Button
   bInherited=true
   bIsVariable=true
   parent=Widget_MiscellaneousContent
   widget=Button_Respawn
   properties {
    buttonText=RESPAWN
    hoveredButtonText=""
    navigation=&self:WidgetTree.Button_Respawn.WidgetNavigation_0
   }
   slot {class=VerticalBoxSlot, properties {padding=<ltrb 0 0 0 20>}}
  }
  Button_ClearSavedData {
   widgetClassPath=&WBP_Creative_OverviewScreen_Button
   bInherited=true
   bIsVariable=true
   parent=Widget_MiscellaneousContent
   widget=Button_ClearSavedData
   properties {
    buttonText=CLEAR SAVED PERSISTENT DATA
    hoveredButtonText="This action clears all data stored for this content.  "
    navigation=&self:WidgetTree.Button_ClearSavedData.WidgetNavigation_0
   }
   slot {class=VerticalBoxSlot}
  }
  Widget_CreditsContent {
   widgetClassPath=&Overlay
   bInherited=true
   parent=Switcher_MainContent
   widget=Widget_CreditsContent
   properties {navigation=&self:WidgetTree.Widget_CreditsContent.WidgetNavigation_2}
   slot {class=WidgetSwitcherSlot}
  }
  CreatorAttributions {
   widgetClassPath=&/WBP_Creative_CreatorAttributions
   bInherited=true
   bIsVariable=true
   parent=Widget_CreditsContent
   widget=CreatorAttributions
   properties {
    creditsCreatorAttributionsClass=&/WBP_Creative_Credits_Attributions_Button.WBP_Creative_Credits_Attributions_Button_C
    navigation=&self:WidgetTree.CreatorAttributions.WidgetNavigation_1
   }
   slot {class=OverlaySlot, properties {horizontalAlignment=HAlign_Fill, verticalAlignment=VAlign_Fill}}
  }
  SafeZone_12 {
   widgetClassPath=&SafeZone
   bIsVariable=true
   parent=Main
   widget=SafeZone_12
   properties {padTop=false}
   slot {class=VerticalBoxSlot, properties {verticalAlignment=VAlign_Bottom}}
  }
  HorizontalBox_145 {
   widgetClassPath=&HorizontalBox
   parent=SafeZone_12
   widget=HorizontalBox_145
   slot {class=SafeZoneSlot, properties {vAlign=VAlign_Center}}
  }
  WBP_Creative_Minigame_Joining {
   widgetClassPath=&/WBP_Creative_Minigame_Joining
   bIsVariable=true
   parent=HorizontalBox_145
   widget=WBP_Creative_Minigame_Joining
   properties {playerInfoTextColor=<rgba 0.007 0.314 1 0.8>}
   slot {class=HorizontalBoxSlot, properties {size {sizeRule=Fill, value=1}}}
  }
  Button_Exit {
   widgetClassPath=&IconTextButton
   bInherited=true
   parent=HorizontalBox_145
   widget=Button_Exit
   properties {
    bIsFocusable=false
    buttonText=Exit
    controllerInputStyle=/FortUILibrary/Foundation/Buttons/ButtonStyles/ButtonStyle-Empty.ButtonStyle-Empty_C
    minHeight=45
    onButtonBaseClicked=(&self:WidgetTree.Button_Exit.HandleOnButtonClicked)
    onButtonBaseDoubleClicked=(&self:WidgetTree.Button_Exit.HandleOnButtonDoubleClicked)
    onButtonBaseHovered=(&self:WidgetTree.Button_Exit.HandleOnButtonHovered)
    onButtonBaseUnhovered=(&self:WidgetTree.Button_Exit.HandleOnButtonUnhovered)
    onSelectedChangedBase=(&self:WidgetTree.Button_Exit.HandleOnSelectedChanged)
    style=/FortUILibrary/Foundation/Buttons/ButtonStyles/ButtonStyle-BottomBar_Refresh.ButtonStyle-BottomBar_Refresh_C
    triggeringInputAction {dataTable=&InputActionDataTable, rowName=GenericBack}
   }
   slot {class=HorizontalBoxSlot, properties {horizontalAlignment=HAlign_Right, verticalAlignment=VAlign_Bottom}}
  }
 ]
}
