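# HUD widget that shows the Creative minigame info panel while the player is in a pregame minigame on a published island.
# On construct in Creative mode it listens for volume enter/exit router messages, island leave, game start and loading-screen changes.
# Entering a volume binds minigame state changes, which re-evaluate visibility and panel input through SetVisibilityForPublishedIslands.
# A timer retries setup checks until pawn and minigame are ready, and non-Creative modes disable the message.
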WidgetBlueprint /Game/Athena/HUD/CreativeMiniGameMessage
graphs [
 EventGraph {nodes [
   n1 = OnMiniGameStateChanged!() -> (OutputDelegate, then, ChangedMinigame:Fort Minigame Object Reference, MinigameState:EFortMinigameState Enum=PreGame) {is=title K2Node_CustomEvent}
   > UpdateVisibility!(execute, self, MinigameToCheck:Fort Minigame Object Reference=@n1.ChangedMinigame) -> (then)
   n2 = OnLeaveIsland!() -> (OutputDelegate, then, PlayerPawn:Fort Player Pawn Object Reference) {is=title K2Node_CustomEvent}
   > Branch(Condition="Equal (Object)"(A=@n2.PlayerPawn, B=GetOwningPlayerPawn!(self:User Widget Object Reference) -> (ReturnValue:Pawn Object Reference)) {is=title K2Node_CallFunction})
   > SetVisibility(InVisibility=Collapsed)
   > UnBindMiniGameEvents!(execute, self, Minigame:Fort Minigame Object Reference) -> (then)
   > EnablePanelInputs!(
    execute
    self:Creative Mini Game Info Widget Object Reference=Get CreativeMiniGameInfoWidget() -> (CreativeMiniGameInfoWidget:Creative Mini Game Info Widget Object Reference)
    InputEnabled:bool=false
   ) -> (then)
   n3 = UpdateVisibility!() -> (OutputDelegate, then, MinigameToCheck:Fort Minigame Object Reference) {is=title K2Node_CustomEvent}
   > n4 = SetVisibilityForPublishedIslands!(execute, self, MinigameToCheck:Fort Minigame Object Reference=@n3.MinigameToCheck) -> (then, bIsVisible:bool)
   n5 = Get CreativeMiniGameInfoWidget()
   > EnablePanelInputs(self=@n5, InputEnabled=@n4)
   > PlayIntro(self:Creative Mini Game Info Widget Object Reference=@n5)
   n6 = OnExitVolume!() -> (OutputDelegate, then, ExitedMinigame:Fort Minigame Object Reference) {is=title K2Node_CustomEvent}
   > UnBindMiniGameEvents(Minigame=@n6.ExitedMinigame)
   > EnablePanelInputs(self=Get CreativeMiniGameInfoWidget(), InputEnabled=false)
   > SetVisibility(InVisibility=Collapsed)
   n7 = BindMiniGameEvents!() -> (OutputDelegate, then, Minigame:Fort Minigame Object Reference) {is=title K2Node_CustomEvent}
   n8 = Create Event()
   > Bind Event to State Changed Delegate(self:Fort Minigame Object Reference=@n7.Minigame, Delegate=@n8) {is=title K2Node_AddDelegate}
   n9 = UnBindMiniGameEvents!() -> (OutputDelegate, then, Minigame:Fort Minigame Object Reference) {is=title K2Node_CustomEvent}
   > Is Valid(InputObject=@n9.Minigame)
   > Unbind Event from State Changed Delegate(self:Fort Minigame Object Reference=@n9.Minigame, Delegate=@n8) {is=title K2Node_RemoveDelegate}
   DisableMessage() {is=title K2Node_CustomEvent}
   > SetVisibility(InVisibility=Collapsed)
   > Set bShouldShowMessage(bShouldShowMessage:bool=false) -> (Output_Get:bool)
   > UnbindGlobalEvents()
   > UnBindMiniGameEvents() {is=title K2Node_CallFunction}
   > EnablePanelInputs(self=Get CreativeMiniGameInfoWidget(), InputEnabled=false)
   UnbindGlobalEvents() {is=title K2Node_CustomEvent}
   > n10 = Get Player Controller Athena!(execute:Exec) -> (then:Exec, CastFailed:Exec, AsFort Player Controller Athena:Fort Player Controller Athena Object Reference) {is=title K2Node_MacroInstance}
   n11 = Create Event()
   > Unbind Event from On Client Leave Island(self:Fort Player Controller Athena Object Reference=@n10, Delegate=@n11) {is=title K2Node_RemoveDelegate}
   n12 = Get PlayerEnteredHandle() -> (PlayerEnteredHandle:Async Action Start Listening To Event Object Reference)
   > n13 = Is Valid(InputObject=@n12)
   n14 = Get ExitedEventHandle() -> (ExitedEventHandle:Async Action Start Listening To Event Object Reference)
   n15 = Is Valid(exec=[@n21.then, @n13.Is Not Valid], InputObject=@n14)
   > n16 = StopListeningToEvent(self:Async Action Start Listening To Event Object Reference=@n14)
   n17 = Create Event()
   n18 = Get CreativeMiniGameInfoWidget()
   Unbind Event from On Starting Game(execute=[@n16.then, @n15.Is Not Valid], self:Creative Mini Game Info Widget Object Reference=@n18, Delegate=@n17) {is=title K2Node_RemoveDelegate}
   > n19 = Get Player Controller Athena()
   n20 = Create Event()
   > Unbind Event from In Game Load Screen Changed(self:Fort Player Controller Athena Object Reference=@n19, Delegate=@n20) {is=title K2Node_RemoveDelegate}
   n21 = StopListeningToEvent(execute=@n13.Is Valid, self=@n12)
   OnStartingGame() {is=title K2Node_CustomEvent}
   > SetVisibility(InVisibility=Collapsed)
   > EnablePanelInputs(self=Get CreativeMiniGameInfoWidget(), InputEnabled=false)
   BindGlobalEvents() {is=title K2Node_CustomEvent}
   > n22 = Get Player Controller Athena()
   > Bind Event to On Client Leave Island(self:Fort Player Controller Athena Object Reference=@n22, Delegate=@n11) {is=title K2Node_AddDelegate}
   > n23 = Sequence()
   > n24 = StartListeningToEventBackwardCompatible!(
    execute:Exec
    Channel:Event Message Tag Structure=(TagName="EventMessage.Component.Playspace.PlayerEnteredSpatialActor.LegacyMinigame")
    Target:Gameplay Event Router Component Object Reference=GetGlobalEventRouter()
    EventType:Script Struct Object Reference=/Script/FortniteGame.PlayerEnterSpatialActorContextLegacyMinigame
    Context:Object Reference
    EventBubblingRule:EEventBubblingRule Enum=Down
   ) -> (
    then:Exec
    AsyncTaskProxy:Async Action Start Listening To Event Object Reference
    OnEventReceived:Exec
    Payload:Player Enter Spatial Actor Context Legacy Minigame Structure
   ) {is=title K2Node_StartListeningToEventAsyncAction}
   > Set PlayerEnteredHandle(PlayerEnteredHandle:Async Action Start Listening To Event Object Reference=@n24) -> (Output_Get:Async Action Start Listening To Event Object Reference)
   > n25 = StartListeningToEventBackwardCompatible(
    Channel=(TagName="EventMessage.Component.Playspace.PlayerExitedSpatialActor.LegacyMinigame")
    Target=GetGlobalEventRouter()
    EventType=/Script/FortniteGame.PlayerExitSpatialActorContextLegacyMinigame
    EventBubblingRule=Down
   ) -> (Payload:Player Exit Spatial Actor Context Legacy Minigame Structure)
   > Set ExitedEventHandle(ExitedEventHandle:Async Action Start Listening To Event Object Reference=@n25) -> (Output_Get:Async Action Start Listening To Event Object Reference)
   OnExitVolume!(
    execute=@n25.OnEventReceived
    self
    ExitedMinigame:Fort Minigame Object Reference=GetMinigame!(ContextActor:Actor Object Reference=Break Player Exit Spatial Actor Context Legacy Minigame!(PlayerExitSpatialActorContextLegacyMinigame:"Player Exit Spatial Actor Context Legacy Minigame Structure (by ref)"=@n25.Payload) -> (SpatialGameplayInterfaceActor:Actor Object Reference, ExitingPlayerState:Player State Object Reference) {is=title K2Node_BreakStruct}) -> (ReturnValue:Fort Minigame Object Reference)
   ) -> (then) {is=title K2Node_CallFunction}
   OnEnterVolume!(
    execute=@n24.OnEventReceived
    self
    EnteredMinigame:Fort Minigame Object Reference=GetMinigame(ContextActor=Break Player Enter Spatial Actor Context Legacy Minigame!(PlayerEnterSpatialActorContextLegacyMinigame:"Player Enter Spatial Actor Context Legacy Minigame Structure (by ref)"=@n24.Payload) -> (SpatialGameplayInterfaceActor:Actor Object Reference, EnteringPlayerState:Player State Object Reference) {is=title K2Node_BreakStruct})
   ) -> (then)
   Bind Event to On Starting Game(execute=@n23.then_1, self:Creative Mini Game Info Widget Object Reference=@n18, Delegate=@n17) {is=title K2Node_AddDelegate}
   > n26 = Get Player Controller Athena()
   > Bind Event to In Game Load Screen Changed(self:Fort Player Controller Athena Object Reference=@n26, Delegate=@n20) {is=title K2Node_AddDelegate}
   n27 = GetMinigame(ContextActor=@n26)
   > Is Valid(InputObject=@n27)
   > OnEnterVolume(EnteredMinigame=@n27)
   n28 = TrySetup() {is=title K2Node_CustomEvent}
   Branch(
    execute=[@n35.then, @n28.then]
    Condition="integer < integer"!(A:int=Get TryAttempts() -> (TryAttempts:int=0), B:int=30) -> (ReturnValue:bool) {is=title K2Node_CallFunction}
   )
   > n29 = ShouldRetry!(execute, self) -> (then, bShouldRetry:bool)
   > Branch(Condition=@n29)
   > Set Timer by Event(Delegate=@n28, Time=2, bLooping=false)
   n30 = OnEnterVolume!() -> (OutputDelegate, then, EnteredMinigame:Fort Minigame Object Reference) {is=title K2Node_CustomEvent}
   > n31 = SetVisibilityForPublishedIslands(MinigameToCheck=@n30.EnteredMinigame)
   > EnablePanelInputs(self=Get CreativeMiniGameInfoWidget(), InputEnabled=@n31)
   > n32 = Branch(Condition=@n31)
   > BindMiniGameEvents!(execute, self, Minigame:Fort Minigame Object Reference=@n30.EnteredMinigame) -> (then) {is=title K2Node_CallFunction}
   > OnVisible(self:Fort Creative Mini Game Info Object Reference=Get CreativeMiniGameInfoWidget())
   n33 = Get Player Controller Athena(execute=@n32.else)
   > Branch(Condition=IsPlayerInReadOnlyExperience!(self:Fort Player Controller Athena Object Reference=@n33) -> (ReturnValue:bool))
   > BindMiniGameEvents(Minigame=@n30.EnteredMinigame)
   > OnVisible(self=Get CreativeMiniGameInfoWidget())
   Event Construct()
   > n34 = Is Creative Mode!(execute:Exec) -> (True:Exec, False:Exec) {is=title K2Node_MacroInstance}
   DisableMessage(execute=@n34.False) {is=title K2Node_CallFunction}
   n35 = BindGlobalEvents(execute=@n34.True) {is=title K2Node_CallFunction}
   Event Destruct()
   > Is Creative Mode()
   > UnbindGlobalEvents() {is=title K2Node_CallFunction}
  ]}
 GetPlayerControllerAthena {nodes [
   Inputs()
   > n1 = Cast To FortPlayerControllerAthena(Object=GetOwningPlayer())
   > Outputs!(then:Exec, CastFailed:Exec=@n1.CastFailed, AsFort Player Controller Athena:Fort Player Controller Athena Object Reference=@n1)
  ]}
 HandleLoadingScreenChanged {nodes [
   n1 = SetVisibilityForPublishedIslands()
   > EnablePanelInputs(self=Get CreativeMiniGameInfoWidget(), InputEnabled=@n1)
   n2 = HandleLoadingScreenChanged!() -> (then, PlayerController:Fort Player Controller Athena Object Reference, bEnabled:bool=~, HUDReasonText:Text) {is=title K2Node_FunctionEntry}
   > Set bShouldShowMessage(bShouldShowMessage=NOT Boolean(A=@n2.bEnabled))
  ]}
 IsCreativeMode {nodes [
   Inputs()
   > n1 = Cast To FortGameStateAthena!(execute, Object=GetGameState!() -> (ReturnValue:Game State Base Object Reference)) -> (then, CastFailed, AsFort Game State Athena:Fort Game State Athena Object Reference) {is=title K2Node_DynamicCast}
   > n2 = Branch(Condition=IsCreativeOrPlaygroundModePlaylist!(self:Fort Game State Athena Object Reference=@n1) -> (ReturnValue:bool))
   > Outputs!(True:Exec, False:Exec=@n2.else)
  ]}
 SetVisibilityForPublishedIslands {nodes [
   SetVisibilityForPublishedIslands!() -> (then, MinigameToCheck:Fort Minigame Object Reference) {is=title K2Node_FunctionEntry}
   > n1 = Get!(execute:Exec) -> (then:Exec, else:Exec, MinigameToCheck:Fort Minigame Object Reference)
   > n2 = Get Player Controller Athena()
   > n3 = Branch(Condition=IsPlayerInPublishedExperience!(self:Fort Player Controller Athena Object Reference=@n2) -> (ReturnValue:bool))
   > n4 = Branch(Condition=AND Boolean(A=IsPregame!(self:Fort Minigame Object Reference=@n1) -> (ReturnValue:bool), B=Get bShouldShowMessage() -> (bShouldShowMessage:bool)))
   > SetVisibility(InVisibility=SelfHitTestInvisible)
   > ForceRefresh!(execute, self:Fort Creative Mini Game Info Object Reference=Get CreativeMiniGameInfoWidget(), bSetupInputComponent:bool=true) -> (then, ReturnValue:bool)
   > Return Node!(execute, bIsVisible:bool=true)
   SetVisibility(execute=[@n4.else, @n3.else, @n2.CastFailed, @n1.else], InVisibility=Collapsed)
   > Return Node(bIsVisible=false)
  ]}
 ShouldRetry {nodes [
   ShouldRetry() {is=title K2Node_FunctionEntry}
   > n1 = Get Player Controller Athena()
   n2 = IsValid(Object=GetPlayerPawn!(self:Fort Player Controller Object Reference=@n1) -> (ReturnValue:Fort Player Pawn Object Reference))
   n3 = NOT Boolean(A=IsValid(Object=GetMinigame!(self:Fort Player Controller Gameplay Object Reference=@n1) -> (ReturnValue)))
   n4 = IsPlayerInPublishedExperience(self=@n1)
   > Return Node!(execute, bShouldRetry:bool=OR Boolean(A=NOT Boolean(A=@n2), B=AND Boolean(A=@n4, B=@n3)))
   Return Node(execute=@n1.CastFailed, bShouldRetry=false)
  ]}
]
variables [
 bShouldShowMessage {type=boolean, default=true}
 DefaultPadding {type=Margin, default=<ltrb 0 112 0 80>}
 ExitedEventHandle {type=AsyncAction_StartListeningToEvent, default=null}
 MobilePadding {type=Margin, default=<ltrb 0 112 0 160>}
 PlayerEnteredHandle {type=AsyncAction_StartListeningToEvent, default=null}
 TryAttempts {type=integer}
]
widgets {
 info {inheritedWidgetCount=1, parentClass=&FortHUDElementWidget, rootWidgetClass=&SizeBox, widgetCount=4}
 widgets [
  SizeBox_0 {widgetClassPath=&SizeBox, widget=SizeBox_0, properties {bOverride_WidthOverride=true, heightOverride=720, widthOverride=1920}}
  Border_0 {
   widgetClassPath=&Border
   parent=SizeBox_0
   widget=Border_0
   properties {
    visibility=SelfHitTestInvisible
    background {
     bIsDynamicallyLoaded=false
     drawAs=NoDrawType
     imageType=NoImage
     mirroring=NoMirror
     resourceName=None
     resourceObject=/Game/Athena/HUD/Art/Materials/M_UI_CreativeStart_Panel
     tiling=NoTile
     imageSize=<128 512>
     margin=<ltrb 0 0 1 0>
     outlineSettings {
      bUseBrushTransparency=false
      roundingType=HalfHeightRadius
      width=0
      color {colorUseRule=UseColor_Specified, specifiedColor=<rgba 0 0 0 0>}
      cornerRadii=<0 0 0 0>
     }
     tintColor {colorUseRule=UseColor_Specified, specifiedColor=<rgba 1 1 1 1>}
     uVRegion {bIsValid=false, max=<0 0>, min=<0 0>}
    }
    padding=<ltrb 0 112 0 80>
   }
   slot {class=SizeBoxSlot}
  }
  CreativeMiniGameInfoWidget {
   widgetClassPath=&/CreativeMiniGameInfoWidget
   bIsVariable=true
   parent=Border_0
   widget=CreativeMiniGameInfoWidget
   slot {class=BorderSlot, properties {padding=<ltrb 0 112 0 80>}}
  }
  TutorialHighlight {widgetClassPath=&TutorialHighlight}
 ]
}
