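# Container widget for the Creative start screen, built on CommonUserWidget with a single root Overlay.
# On Construct it checks IsNewCreativeStartScreenEnabled and, only when that is false, asynchronously creates CreativeMiniGameMessage from a soft class, suspending input until done.
# A valid result is added to the Overlay with fill alignment on both axes.
# The enabled branch adds nothing here, and Tick is empty.
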WidgetBlueprint /Game/Athena/HUD/Creative/StartScreen/WBP_Creative_StartScreen_Container
graphs [EventGraph {nodes [
   Event Construct()
   > n1 = Branch(Condition=IsNewCreativeStartScreenEnabled!() -> (ReturnValue:bool))
   n2 = CreateWidgetAsync!(
    execute:Exec=@n1.else
    UserWidgetSoftClass:User Widget Soft Class Reference=/Game/Athena/HUD/CreativeMiniGameMessage.CreativeMiniGameMessage_C
    OwningPlayer:Player Controller Object Reference
    bSuspendInputUntilComplete:bool=true
   ) -> (then:Exec, OnComplete:Exec, OnFailure:Exec, UserWidget:User Widget Object Reference) {is=title K2Node_AsyncAction}
   Is Valid(exec=@n2.OnComplete, InputObject=@n2)
   > n3 = AddChildToOverlay!(execute, self:Overlay Object Reference=Get Overlay() -> (Overlay:Overlay Object Reference), Content:Widget Object Reference=@n2) -> (then, ReturnValue:Overlay Slot Object Reference)
   > SetHorizontalAlignment(self=@n3, InHorizontalAlignment=HAlign_Fill)
   > SetVerticalAlignment(self=@n3, InVerticalAlignment=VAlign_Fill)
   Event Tick()
  ]}]
variables []
widgets {
 info {inheritedWidgetCount=0, parentClass=&CommonUserWidget, rootWidgetClass=&Overlay, widgetCount=1}
 widgets [Overlay {widgetClassPath=&Overlay, bIsVariable=true, widget=Overlay}]
}
